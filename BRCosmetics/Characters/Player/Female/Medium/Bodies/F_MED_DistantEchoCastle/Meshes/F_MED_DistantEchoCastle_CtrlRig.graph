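# Post-process style Control Rig for the F_MED_DistantEchoCastle player body that procedurally poses the robe's skirt master bones to follow the legs.
# After optional debug controls, FortDeformRig handles upper-arm and elbow deformation, then VariableSetup remaps thigh, calf and skirt rotations to normalized weights.
# Robe uses those weights to blend thigh/pelvis parent constraints and corrective transform offsets.
# Smoothed state flags add or damp poses for crouching, sliding, skydiving, emotes and front-end idle.
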ControlRigBlueprint /BRCosmetics/Characters/Player/Female/Medium/Bodies/F_MED_DistantEchoCastle/Meshes/F_MED_DistantEchoCastle_CtrlRig
graphs [
 Multiply_2_ContainedGraph {nodes [
   n1 = RigVMFunctionEntryNode()
   Multiply = &RigVMFunction_MathDoubleMul(A=@n1.A, B=@n1.B)
   RigVMFunctionReturnNode(Result=&RigVMFunction_MathDoubleMul(B=@n1.D, A=&RigVMFunction_MathDoubleMul(B=@n1.C, A=@Multiply.Result).Result).Result)
  ]}
 Multiply_2_ContainedGraph {nodes [
   n1 = RigVMFunctionEntryNode()
   n2 = &RigVMFunction_MathDoubleMul(A=@n1.A, B=@n1.B)
   RigVMFunctionReturnNode(Result=&RigVMFunction_MathDoubleMul(B=@n1.D, A=&RigVMFunction_MathDoubleMul(B=@n1.C, A=@n2.Result).Result).Result)
  ]}
 Multiply_4_ContainedGraph {nodes [
   n1 = RigVMFunctionEntryNode()
   RigVMFunctionReturnNode(Result=&RigVMFunction_MathDoubleMul(B=@n1.C, A=&RigVMFunction_MathDoubleMul(A=@n1.A, B=@n1.B).Result).Result)
  ]}
 New Function_ContainedGraph {nodes [
   RigVMFunctionEntryNode()
   > RigVMFunction_Sequence = RigVMAggregateNode()
   &RigUnit_ModifyTransforms(
    ItemToModify="((Item=(Type=Bone,Name=\"dyn_front_skirt_MSTR\"),Transform=(Rotation=(X=-0.1473,Y=0.0418,Z=0.1615,W=-0.9749),Translation=(X=0,Y=0,Z=0),Scale3D=(X=0.98,Y=1,Z=1))),(Item=(Type=Bone,Name=\"dyn_front_skirt_l_MSTR\"),Transform=(Rotation=(X=-0.1123,Y=-0.0162,Z=0.105 ...(truncated)"
    Weight=&RigVMFunction_MathDoubleRemap(SourceMaximum=0.4, SourceMinimum=0, TargetMaximum=1, TargetMinimum=0, Value=RigVMVariableNode() {variable=Thigh_R_FrontRotation}.Value).Result
    ExecutePin=@RigVMFunction_Sequence.C
   )
   > &RigUnit_ModifyTransforms(
    ItemToModify="((Item=(Type=Bone,Name=\"dyn_front_skirt_MSTR\"),Transform=(Rotation=(X=-0.2314,Y=-0.0265,Z=-0.1517,W=0.9606),Translation=(X=0,Y=0,Z=0),Scale3D=(X=0.98,Y=1,Z=1))),(Item=(Type=Bone,Name=\"dyn_front_skirt_r_MSTR\"),Transform=(Rotation=(X=0.2572,Y=0.0422,Z=-0.053 ...(truncated)"
    Weight=&RigVMFunction_MathDoubleRemap(
     SourceMaximum=0.6
     SourceMinimum=0
     TargetMaximum=1
     TargetMinimum=0
     bClamp=False
     Value=RigVMVariableNode() {variable=Thigh_L_FrontRotation}.Value
    ).Result
   )
   RigVMFunctionReturnNode(ExecuteContext=@RigVMFunction_Sequence.H)
   &RigUnit_ParentConstraint(
    AdvancedSettings=(RotationOrderForFilter=XZY)
    Child=(Type=Bone,Name="dyn_side_skirt_r_MSTR")
    Parents=((Item=(Type=Bone,Name="thigh_twist_01_r"),Weight=0.2),(Item=(Type=Bone,Name="pelvis"),Weight=0.8))
    Weight=1
    ExecutePin=@RigVMFunction_Sequence.B
   )
   > &RigUnit_ParentConstraint(
    AdvancedSettings=(RotationOrderForFilter=XZY)
    Child=(Type=Bone,Name="dyn_front_skirt_r_MSTR")
    Filter=(TranslationFilter=(bX=false,bY=false,bZ=false),RotationFilter=(bX=True,bY=True,bZ=True),ScaleFilter=(bX=True,bY=True,bZ=True))
    Parents=((Item=(Type=Bone,Name="thigh_twist_01_r"),Weight=0.25),(Item=(Type=Bone,Name="pelvis"),Weight=0.75))
    Weight=1
   )
   n1 = RigVMVariableNode() {variable=Thigh_R_FrontRotation}
   > &RigUnit_ParentConstraint(
    AdvancedSettings=(RotationOrderForFilter=XZY)
    Child=(Type=Bone,Name="dyn_front_skirt_MSTR")
    Filter=(TranslationFilter=(bX=false,bY=false,bZ=false),RotationFilter=(bX=True,bY=True,bZ=True),ScaleFilter=(bX=True,bY=True,bZ=True))
    Parents=((Item=(Type=Bone,Name="thigh_twist_01_r"),Weight=0.8),(Item=(Type=Bone,Name="pelvis"),Weight=0.2))
    Weight=@n1.Value
   )
   > &RigUnit_ModifyTransforms(
    ItemToModify="((Item=(Type=Bone,Name=\"dyn_front_skirt_r_MSTR\"),Transform=(Rotation=(X=-0.0045,Y=-0.0045,Z=0.5168,W=-0.8561),Translation=(X=9.4782,Y=1.6838,Z=-0.0397),Scale3D=(X=1,Y=1,Z=1))),(Item=(Type=Bone,Name=\"dyn_side_skirt_r_MSTR\"),Transform=(Rotation=(X=-0.2074,Y= ...(truncated)"
    Weight=@n1.Value
   )
   n2 = RigVMVariableNode() {variable=Thigh_R_SideRotation}
   > &RigUnit_ParentConstraint(
    AdvancedSettings=(RotationOrderForFilter=XZY)
    Child=(Type=Bone,Name="dyn_side_skirt_r_MSTR")
    Filter=(TranslationFilter=(bX=false,bY=false,bZ=false),RotationFilter=(bX=True,bY=True,bZ=True),ScaleFilter=(bX=True,bY=True,bZ=True))
    Parents=((Item=(Type=Bone,Name="thigh_twist_01_r"),Weight=0.8),(Item=(Type=Bone,Name="pelvis"),Weight=0.2))
    Weight=@n2.Value
   )
   > &RigUnit_ParentConstraint(
    AdvancedSettings=(RotationOrderForFilter=XZY)
    Child=(Type=Bone,Name="dyn_front_skirt_r_MSTR")
    Filter=(TranslationFilter=(bX=false,bY=false,bZ=false),RotationFilter=(bX=True,bY=True,bZ=True),ScaleFilter=(bX=True,bY=True,bZ=True))
    Parents=((Item=(Type=Bone,Name="thigh_twist_01_r"),Weight=0.7),(Item=(Type=Bone,Name="pelvis"),Weight=0.3))
    Weight=@n2.Value
   )
   > &RigUnit_ModifyTransforms(
    ItemToModify="((Item=(Type=Bone,Name=\"dyn_side_skirt_r_MSTR\"),Transform=(Rotation=(X=-0.0002,Y=-0.0035,Z=0.1573,W=-0.9875),Translation=(X=2.9309,Y=5.0184,Z=-0.0133),Scale3D=(X=0.9215,Y=1,Z=1))),(Item=(Type=Bone,Name=\"dyn_front_skirt_r_MSTR\"),Transform=(Rotation=(X=0.164 ...(truncated)"
    Weight=@n2.Value
   )
   > &RigUnit_ModifyTransforms(
    ItemToModify="((Item=(Type=Bone,Name=\"dyn_rear_skirt_MSTR\"),Transform=(Rotation=(X=0,Y=0,Z=-0.3959,W=0.9183),Translation=(X=0,Y=0,Z=0),Scale3D=(X=1,Y=1,Z=1))),(Item=(Type=Bone,Name=\"dyn_rear_skirt_r_MSTR\"),Transform=(Rotation=(X=0.0053,Y=0.0049,Z=0.4394,W=-0.8983),Trans ...(truncated)"
    WeightMinimum=0
    Weight=RigVMVariableNode() {variable=Thigh_R_BackRotation}.Value
   )
   Multiply_1 = RigVMAggregateNode(
    B=&RigVMFunction_MathDoubleRemap(SourceMaximum=0.7, SourceMinimum=0, TargetMaximum=0, TargetMinimum=1, Value=RigVMVariableNode() {variable=Thigh_R_FrontRotation}.Value).Result
    C=&RigVMFunction_MathDoubleRemap(SourceMaximum=1, SourceMinimum=0, TargetMaximum=0, TargetMinimum=1, Value=RigVMVariableNode() {variable=isKneeSlidingFloat}.Value).Result
    D=&RigVMFunction_MathDoubleRemap(SourceMaximum=1, SourceMinimum=0, TargetMaximum=0.3, TargetMinimum=1, Value=RigVMVariableNode() {variable=isSkydivingFloat}.Value).Result
    A=RigVMVariableNode() {variable=Calf_R_BackRotation}.Value
   )
   > &RigUnit_ModifyTransforms(
    ItemToModify="((Item=(Type=Bone,Name=\"dyn_rear_skirt_r_MSTR\"),Transform=(Rotation=(X=0,Y=0,Z=0.3988,W=-0.917),Translation=(X=0,Y=0,Z=0),Scale3D=(X=1,Y=1,Z=1))),(Item=(Type=Bone,Name=\"dyn_rear_skirt_MSTR\"),Transform=(Rotation=(X=0,Y=0,Z=-0.2164,W=0.9763),Translation=(X=0 ...(truncated)"
    WeightMinimum=0
    Weight=@Multiply_1.Result
   )
   &RigUnit_ParentConstraint(
    AdvancedSettings=(RotationOrderForFilter=XZY)
    Child=(Type=Bone,Name="dyn_side_skirt_l_MSTR")
    Parents=((Item=(Type=Bone,Name="thigh_twist_01_l"),Weight=0.2),(Item=(Type=Bone,Name="pelvis"),Weight=0.8))
    Weight=1
    ExecutePin=@RigVMFunction_Sequence.A
   )
   > &RigUnit_ParentConstraint(
    AdvancedSettings=(RotationOrderForFilter=XZY)
    Child=(Type=Bone,Name="dyn_front_skirt_l_MSTR")
    Filter=(TranslationFilter=(bX=false,bY=false,bZ=false),RotationFilter=(bX=True,bY=True,bZ=True),ScaleFilter=(bX=True,bY=True,bZ=True))
    Parents=((Item=(Type=Bone,Name="thigh_twist_01_l"),Weight=0.25),(Item=(Type=Bone,Name="pelvis"),Weight=0.75))
    Weight=1
   )
   n3 = RigVMVariableNode() {variable=Thigh_L_FrontRotation}
   > &RigUnit_ParentConstraint(
    AdvancedSettings=(RotationOrderForFilter=XZY)
    Child=(Type=Bone,Name="dyn_front_skirt_MSTR")
    Filter=(TranslationFilter=(bX=false,bY=false,bZ=false),RotationFilter=(bX=True,bY=True,bZ=True),ScaleFilter=(bX=True,bY=True,bZ=True))
    Parents=((Item=(Type=Bone,Name="thigh_twist_01_l"),Weight=0.8),(Item=(Type=Bone,Name="pelvis"),Weight=0.2),(Item=(Type=Bone,Name="thigh_twist_01_r"),Weight=0))
    Weight=@n3.Value
   )
   > &RigUnit_ModifyTransforms(
    ItemToModify="((Item=(Type=Bone,Name=\"dyn_front_skirt_l_MSTR\"),Transform=(Rotation=(X=-0.0289,Y=-0.0139,Z=0.4722,W=-0.8809),Translation=(X=-8.1888,Y=-0.9426,Z=-0.0642),Scale3D=(X=1,Y=1,Z=1))),(Item=(Type=Bone,Name=\"dyn_side_skirt_l_MSTR\"),Transform=(Rotation=(X=0.1342,Y ...(truncated)"
    Weight=@n3.Value
   )
   n4 = RigVMVariableNode() {variable=Thigh_L_SideRotation}
   > &RigUnit_ParentConstraint(
    AdvancedSettings=(RotationOrderForFilter=XZY)
    Child=(Type=Bone,Name="dyn_side_skirt_l_MSTR")
    Filter=(TranslationFilter=(bX=false,bY=false,bZ=false),RotationFilter=(bX=True,bY=True,bZ=True),ScaleFilter=(bX=True,bY=True,bZ=True))
    Parents=((Item=(Type=Bone,Name="thigh_twist_01_l"),Weight=0.8),(Item=(Type=Bone,Name="pelvis"),Weight=0.2))
    Weight=@n4.Value
   )
   > &RigUnit_ParentConstraint(
    AdvancedSettings=(RotationOrderForFilter=XZY)
    Child=(Type=Bone,Name="dyn_front_skirt_l_MSTR")
    Filter=(TranslationFilter=(bX=false,bY=false,bZ=false),RotationFilter=(bX=True,bY=True,bZ=True),ScaleFilter=(bX=True,bY=True,bZ=True))
    Parents=((Item=(Type=Bone,Name="thigh_twist_01_l"),Weight=0.7),(Item=(Type=Bone,Name="pelvis"),Weight=0.3))
    Weight=@n4.Value
   )
   > &RigUnit_ModifyTransforms(
    ItemToModify="((Item=(Type=Bone,Name=\"dyn_side_skirt_l_MSTR\"),Transform=(Rotation=(X=-0.0011,Y=-0.0089,Z=0.1403,W=-0.9901),Translation=(X=-1.274,Y=-5.2995,Z=0),Scale3D=(X=0.9215,Y=1,Z=1))),(Item=(Type=Bone,Name=\"dyn_front_skirt_l_MSTR\"),Transform=(Rotation=(X=-0.0072,Y= ...(truncated)"
    Weight=@n4.Value
   )
   > &RigUnit_ModifyTransforms(
    ItemToModify="((Item=(Type=Bone,Name=\"dyn_rear_skirt_MSTR\"),Transform=(Rotation=(X=0,Y=0,Z=-0.4226,W=0.9063),Translation=(X=0,Y=0,Z=0),Scale3D=(X=1,Y=1,Z=1))),(Item=(Type=Bone,Name=\"dyn_rear_skirt_l_MSTR\"),Transform=(Rotation=(X=-0.04,Y=0.0355,Z=-0.4692,W=0.8815),Transl ...(truncated)"
    WeightMinimum=0
    Weight=RigVMVariableNode() {variable=Thigh_L_BackRotation}.Value
   )
   Multiply = RigVMAggregateNode(
    B=&RigVMFunction_MathDoubleRemap(SourceMaximum=0.7, SourceMinimum=0, TargetMaximum=0, TargetMinimum=1, Value=RigVMVariableNode() {variable=Thigh_L_FrontRotation}.Value).Result
    C=&RigVMFunction_MathDoubleRemap(SourceMaximum=1, SourceMinimum=0, TargetMaximum=0, TargetMinimum=1, Value=RigVMVariableNode() {variable=isKneeSlidingFloat}.Value).Result
    D=&RigVMFunction_MathDoubleRemap(SourceMaximum=1, SourceMinimum=0, TargetMaximum=0.3, TargetMinimum=1, Value=RigVMVariableNode() {variable=isSkydivingFloat}.Value).Result
    A=RigVMVariableNode() {variable=Calf_L_BackRotation}.Value
   )
   > &RigUnit_ModifyTransforms(
    ItemToModify="((Item=(Type=Bone,Name=\"dyn_rear_skirt_l_MSTR\"),Transform=(Rotation=(X=0,Y=0,Z=-0.3987,W=0.9171),Translation=(X=0,Y=0,Z=0),Scale3D=(X=1,Y=1,Z=1))),(Item=(Type=Bone,Name=\"dyn_rear_skirt_MSTR\"),Transform=(Rotation=(X=0,Y=0,Z=-0.2164,W=0.9763),Translation=(X= ...(truncated)"
    WeightMinimum=0
    Weight=@Multiply.Result
   )
   n5 = &RigVMFunction_AlphaInterp(
    InterpSpeedDecreasing=10
    InterpSpeedIncreasing=10
    Scale=1
    bClampResult=true
    Value=&RigVMFunction_MathBoolToFloat(Value=RigVMVariableNode() {variable=isCrouching_CR}.Value).Result
   )
   n6 = RigVMVariableNode() {variable=isKneeSlidingFloat}
   n7 = &RigVMFunction_AlphaInterp(
    InterpSpeedDecreasing=10
    InterpSpeedIncreasing=10
    Scale=1
    bClampResult=true
    Value=&RigVMFunction_MathBoolToFloat(Value=RigVMVariableNode() {variable=Is Baseball Sliding_CR}.Value).Result
   )
   Multiply_3 = RigVMAggregateNode(
    C=@n5.Result
    B=&RigVMFunction_MathDoubleRemap(SourceMaximum=1, SourceMinimum=0, TargetMaximum=0.3, TargetMinimum=1, Value=@n6.Value).Result
    A=&RigVMFunction_MathDoubleRemap(SourceMaximum=1, SourceMinimum=0, TargetMaximum=0.3, TargetMinimum=1, Value=@n7.Result).Result
   )
   &RigUnit_ModifyTransforms(
    ItemToModify="((Item=(Type=Bone,Name=\"dyn_side_skirt_l_MSTR\"),Transform=(Rotation=(X=-0.307,Y=0.2613,Z=-0.1048,W=0.9091),Translation=(X=2.8722,Y=0,Z=5.048),Scale3D=(X=1,Y=1,Z=1))),(Item=(Type=Bone,Name=\"dyn_side_skirt_r_MSTR\"),Transform=(Rotation=(X=0.2287,Y=-0.0814,Z=0 ...(truncated)"
    Weight=@Multiply_3.Result
    ExecutePin=@RigVMFunction_Sequence.D
   )
   > &RigUnit_ModifyTransforms(
    ItemToModify="((Item=(Type=Bone,Name=\"dyn_rear_skirt_MSTR\"),Transform=(Rotation=(X=-0.1075,Y=0.104,Z=-0.7098,W=0.6883),Translation=(X=-3.1891,Y=0.7232,Z=1.0186),Scale3D=(X=1,Y=1,Z=1))),(Item=(Type=Bone,Name=\"dyn_rear_skirt_l_MSTR\"),Transform=(Rotation=(X=-0.4516,Y=0.264 ...(truncated)"
    Weight=@n7.Result
   )
   > &RigUnit_ModifyTransforms(
    ItemToModify="((Item=(Type=Bone,Name=\"dyn_rear_skirt_MSTR\"),Transform=(Rotation=(X=0,Y=0,Z=-0.7801,W=0.6256),Translation=(X=0,Y=0,Z=0),Scale3D=(X=1,Y=1,Z=1))),(Item=(Type=Bone,Name=\"dyn_rear_skirt_l_MSTR\"),Transform=(Rotation=(X=-0.2288,Y=-0.0141,Z=-0.7997,W=0.5549),Tra ...(truncated)"
    Weight=@n6.Value
   )
   n8 = &RigVMFunction_AlphaInterp(
    InterpSpeedDecreasing=10
    InterpSpeedIncreasing=10
    Scale=1
    bClampResult=true
    Value=&RigVMFunction_MathBoolToFloat(Value=RigVMVariableNode() {variable=IsPlayingEmoteCR}.Value).Result
   )
   &RigUnit_ModifyTransforms(
    ItemToModify="((Item=(Type=Bone,Name=\"dyn_rear_skirt_MSTR\"),Transform=(Rotation=(X=0,Y=-0.0175,Z=0,W=0.9998),Translation=(X=0,Y=-3,Z=0),Scale3D=(X=1,Y=1,Z=1))),(Item=(Type=Bone,Name=\"dyn_rear_skirt_l_MSTR\"),Transform=(Rotation=(X=0,Y=0,Z=0,W=1),Translation=(X=0,Y=-2,Z=0 ...(truncated)"
    Weight=@n8.Result
    ExecutePin=@RigVMFunction_Sequence.E
   )
   > &RigUnit_ModifyTransforms(
    ItemToModify=((Item=(Type=Bone,Name="dyn_skirt_MSTR"),Transform=(Rotation=(X=0,Y=0,Z=-0.4226,W=0.9063),Translation=(X=0,Y=0,Z=0),Scale3D=(X=1,Y=1,Z=1))))
    Weight=&RigVMFunction_MathDoubleMul(A=@n8.Result, B=RigVMVariableNode() {variable=skirtRotation}.Value).Result
   )
   > &RigUnit_ModifyTransforms(
    ItemToModify="((Item=(Type=Bone,Name=\"dyn_rear_skirt_MSTR\"),Transform=(Rotation=(X=-0.003,Y=-0.0172,Z=-0.1736,W=0.9847),Translation=(X=0,Y=-5,Z=0),Scale3D=(X=1,Y=1,Z=1))),(Item=(Type=Bone,Name=\"dyn_rear_skirt_l_MSTR\"),Transform=(Rotation=(X=0,Y=0,Z=-0.1736,W=0.9848),Tra ...(truncated)"
    Weight=RigVMVariableNode() {variable=isSkydivingFloat}.Value
   )
   n9 = &RigVMFunction_AlphaInterp(
    InterpSpeedDecreasing=10
    InterpSpeedIncreasing=10
    Scale=1
    bClampResult=true
    Value=&RigVMFunction_MathBoolToFloat(Value=RigVMVariableNode() {variable=isFEidle_CR}.Value).Result
   )
   > &RigUnit_ModifyTransforms(
    ItemToModify="((Item=(Type=Bone,Name=\"dyn_rear_skirt_MSTR\"),Transform=(Rotation=(X=-0.0014,Y=0.0869,Z=0.0143,W=0.9961),Translation=(X=-1.1246,Y=-2.885,Z=0.1971),Scale3D=(X=1,Y=1,Z=1))),(Item=(Type=Bone,Name=\"dyn_rear_skirt_l_MSTR\"),Transform=(Rotation=(X=-0.0023,Y=0.072 ...(truncated)"
    Weight=@n9.Result
   )
  ]}
 New Function_ContainedGraph {nodes [
   RigVMFunctionEntryNode()
   > RigVMFunction_Sequence = RigVMAggregateNode()
   Delta Rotation Y = RigVMFunctionReferenceNode(Child=thigh_l, Parent=thigh_l_null, Source Maximum=-50, Source Minimum=0, ExecuteContext=@RigVMFunction_Sequence.A)
   > RigVMVariableNode(Value=@"Delta Rotation Y".Delta Rotation) {variable=Thigh_L_SideRotation}
   > Delta Rotation Y_1 = RigVMFunctionReferenceNode(Child=thigh_r, Parent=thigh_r_null, Source Maximum=-50, Source Minimum=0)
   > RigVMVariableNode(Value=@"Delta Rotation Y_1".Delta Rotation) {variable=Thigh_R_SideRotation}
   > Delta Rotation Z = RigVMFunctionReferenceNode(Child=thigh_l, Parent=thigh_l_null, Source Maximum=60, Source Minimum=0)
   > RigVMVariableNode(Value=@"Delta Rotation Z".Delta Rotation) {variable=Thigh_L_BackRotation}
   > Delta Rotation Z_1 = RigVMFunctionReferenceNode(Child=thigh_r, Parent=thigh_r_null, Source Maximum=60, Source Minimum=0)
   > RigVMVariableNode(Value=@"Delta Rotation Z_1".Delta Rotation) {variable=Thigh_R_BackRotation}
   > Delta Rotation Z_1_1 = RigVMFunctionReferenceNode(Child=calf_l, Parent=calf_l_null, Source Maximum=100, Source Minimum=0)
   > RigVMVariableNode(Value=@"Delta Rotation Z_1_1".Delta Rotation) {variable=Calf_L_BackRotation}
   > Delta Rotation Z_1_1_1 = RigVMFunctionReferenceNode(Child=calf_r, Parent=calf_r_null, Source Maximum=100, Source Minimum=0)
   > RigVMVariableNode(Value=@"Delta Rotation Z_1_1_1".Delta Rotation) {variable=Calf_R_BackRotation}
   > Delta Rotation X = RigVMFunctionReferenceNode(Child=spine_05, Parent=spine_05_null, Source Maximum=-30, Source Minimum=10)
   > RigVMVariableNode(Value=@"Delta Rotation X".Delta Rotation) {variable=spineTwist_R}
   Delta Rotation Z_2 = RigVMFunctionReferenceNode(Child=thigh_l, Parent=thigh_l_null, Source Maximum=-90, Source Minimum=0, ExecuteContext=@RigVMFunction_Sequence.B)
   > RigVMVariableNode(Value=@"Delta Rotation Z_2".Delta Rotation) {variable=Thigh_L_FrontRotation}
   > Delta Rotation Z_3 = RigVMFunctionReferenceNode(Child=thigh_r, Parent=thigh_r_null, Source Maximum=-90, Source Minimum=0)
   > RigVMVariableNode(Value=@"Delta Rotation Z_3".Delta Rotation) {variable=Thigh_R_FrontRotation}
   RigVMFunction_MathBoolToFloat = &RigVMFunction_MathBoolToFloat(Value=RigVMVariableNode() {variable=Is Knee Sliding_CR}.Value)
   n1 = &RigVMFunction_AlphaInterp(InterpSpeedDecreasing=10, InterpSpeedIncreasing=10, Scale=1, bClampResult=true, Value=@RigVMFunction_MathBoolToFloat.Result)
   > RigVMVariableNode(Value=@n1.Result) {variable=isKneeSlidingFloat}
   n2 = &RigVMFunction_AlphaInterp(
    InterpSpeedDecreasing=10
    InterpSpeedIncreasing=10
    Scale=1
    bClampResult=true
    Value=&RigVMFunction_MathBoolToFloat(Value=RigVMVariableNode() {variable=isSkydiving_CR}.Value).Result
   )
   > RigVMVariableNode(Value=@n2.Result) {variable=isSkydivingFloat}
   > Delta Rotation Z_1_2 = RigVMFunctionReferenceNode(Child=dyn_skirt_MSTR, Parent=dyn_skirt_MSTR_null, Source Maximum=40, Source Minimum=5)
   > RigVMVariableNode(Value=@"Delta Rotation Z_1_2".Delta Rotation) {variable=skirtRotation}
   RigVMFunctionReturnNode(ExecuteContext=@RigVMFunction_Sequence.C)
  ]}
 New Function_ContainedGraph {nodes [
   n1 = RigVMFunctionEntryNode()
   RigVMFunction_MathQuaternionToEuler = &RigVMFunction_MathQuaternionToEuler(
    RotationOrder=XYZ
    Value=&RigUnit_GetRelativeTransformForItem(
     Child=(Type=Bone,Name="hand_l")
     Parent=(Type=Null,Name="lowerarm_l")
     bChildInitial=False
     bParentInitial=False
     Child.Name=@n1.Child
     Parent.Name=@n1.Parent
    ).RelativeTransform.Rotation
   )
   Remap = &RigVMFunction_MathDoubleRemap(
    TargetMaximum=1
    TargetMinimum=0
    Value=@RigVMFunction_MathQuaternionToEuler.Result.Y
    SourceMaximum=@n1.Source Maximum
    SourceMinimum=@n1.Source Minimum
   )
   > RigVMFunctionReturnNode(Delta Rotation=@Remap.Result)
  ]}
 New Function_ContainedGraph {nodes [
   n1 = RigVMFunctionEntryNode()
   RigVMFunction_MathQuaternionToEuler = &RigVMFunction_MathQuaternionToEuler(
    RotationOrder=XYZ
    Value=&RigUnit_GetRelativeTransformForItem(
     Child=(Type=Bone,Name="hand_l")
     Parent=(Type=Null,Name="lowerarm_l")
     bChildInitial=False
     bParentInitial=False
     Child.Name=@n1.Child
     Parent.Name=@n1.Parent
    ).RelativeTransform.Rotation
   )
   Remap = &RigVMFunction_MathDoubleRemap(
    TargetMaximum=1
    TargetMinimum=0
    Value=@RigVMFunction_MathQuaternionToEuler.Result.Z
    SourceMaximum=@n1.Source Maximum
    SourceMinimum=@n1.Source Minimum
   )
   > RigVMFunctionReturnNode(Delta Rotation=@Remap.Result)
  ]}
 New Function_ContainedGraph {nodes [
   n1 = RigVMFunctionEntryNode()
   RigVMFunction_MathQuaternionToEuler = &RigVMFunction_MathQuaternionToEuler(
    RotationOrder=XYZ
    Value=&RigUnit_GetRelativeTransformForItem(
     Child=(Type=Bone,Name="hand_l")
     Parent=(Type=Null,Name="lowerarm_l")
     bChildInitial=False
     bParentInitial=False
     Child.Name=@n1.Child
     Parent.Name=@n1.Parent
    ).RelativeTransform.Rotation
   )
   Remap = &RigVMFunction_MathDoubleRemap(
    TargetMaximum=1
    TargetMinimum=0
    Value=@RigVMFunction_MathQuaternionToEuler.Result.X
    SourceMaximum=@n1.Source Maximum
    SourceMinimum=@n1.Source Minimum
   )
   > RigVMFunctionReturnNode(Delta Rotation=@Remap.Result)
  ]}
 New Function_ContainedGraph {nodes [
   n1 = RigVMFunctionEntryNode()
   RigVMFunction_ControlFlowBranch = &RigVMFunction_ControlFlowBranch(Condition=@n1.Debug Controls, ExecuteContext=@n1.ExecuteContext)
   RigVMFunctionReturnNode(ExecuteContext=@RigVMFunction_ControlFlowBranch.Completed)
   RigUnit_GetTransform = &RigUnit_GetTransform(Item=(Type=Control,Name="thigh_r_ctrl"), bInitial=False)
   &RigUnit_SetTransform(
    Item=(Type=Bone,Name="thigh_r")
    Weight=1
    bInitial=False
    Value=@RigUnit_GetTransform.Transform
    ExecutePin=@RigVMFunction_ControlFlowBranch.True
   )
   > &RigUnit_SetTransform(
    Item=(Type=Bone,Name="thigh_l")
    Weight=1
    bInitial=False
    Value=&RigUnit_GetTransform(Item=(Type=Control,Name="thigh_l_ctrl"), bInitial=False).Transform
   )
   > &RigUnit_SetTransform(
    Item=(Type=Bone,Name="calf_l")
    Weight=1
    bInitial=False
    Value=&RigUnit_GetTransform(Item=(Type=Control,Name="calf_l_ctrl"), bInitial=False).Transform
   )
   > &RigUnit_SetTransform(
    Item=(Type=Bone,Name="calf_r")
    Weight=1
    bInitial=False
    Value=&RigUnit_GetTransform(Item=(Type=Control,Name="calf_r_ctrl"), bInitial=False).Transform
   )
   > &RigUnit_SetTransform(
    Item=(Type=Bone,Name="dyn_skirt_MSTR")
    Weight=1
    bInitial=False
    Value=&RigUnit_GetTransform(Item=(Type=Control,Name="dyn_skirt_MSTR_ctrl"), bInitial=False).Transform
   )
  ]}
 RigVMFunctionLibrary {nodes [
   Control Setup = RigVMCollapseNode(Debug Controls=False, ExecuteContext=())
   Delta Rotation X = RigVMCollapseNode(ExecuteContext=())
   Delta Rotation Y = RigVMCollapseNode(ExecuteContext=())
   Delta Rotation Z = RigVMCollapseNode(ExecuteContext=())
   Robe = RigVMCollapseNode(ExecuteContext=())
   VariableSetup = RigVMCollapseNode(ExecuteContext=())
  ]}
 RigVMFunction_Sequence_1_ContainedGraph {nodes [
   n1 = RigVMFunctionEntryNode()
   RigVMFunction_Sequence = &RigVMFunction_Sequence(ExecuteContext=@n1.ExecuteContext)
   n2 = &RigVMFunction_Sequence(ExecuteContext=@RigVMFunction_Sequence.B)
   n3 = &RigVMFunction_Sequence(ExecuteContext=@n2.B)
   n4 = &RigVMFunction_Sequence(ExecuteContext=@n3.B)
   n5 = &RigVMFunction_Sequence(ExecuteContext=@n4.B)
   n6 = &RigVMFunction_Sequence(ExecuteContext=@n5.B)
   n7 = &RigVMFunction_Sequence(ExecuteContext=@n6.B)
   RigVMFunctionReturnNode(A=@RigVMFunction_Sequence.A, B=@n2.A, C=@n3.A, D=@n4.A, E=@n5.A, F=@n6.A, G=@n7.A, H=@n7.B)
  ]}
 RigVMFunction_Sequence_1_ContainedGraph {nodes [
   n1 = RigVMFunctionEntryNode()
   RigVMFunction_Sequence = &RigVMFunction_Sequence(ExecuteContext=@n1.ExecuteContext)
   n2 = &RigVMFunction_Sequence(ExecuteContext=@RigVMFunction_Sequence.B)
   RigVMFunctionReturnNode(A=@RigVMFunction_Sequence.A, B=@n2.A, C=@n2.B)
  ]}
 RigVMModel {nodes [
   RigUnit_BeginExecution = &RigUnit_BeginExecution()
   RigVMFunction_Sequence = &RigVMFunction_Sequence(ExecuteContext=@RigUnit_BeginExecution.ExecutePin)
   Control Setup = RigVMFunctionReferenceNode(Debug Controls=false, ExecuteContext=@RigVMFunction_Sequence.A)
   /Script/FortniteGame.RigUnit_FortDeformRig(
    Bool=(bUpperArm=true,bElbow=true,bWrist=False,bGroin=False,bGlute=False,bKnee=False,bPec=False,bBicep=False,bLat=False)
    Curve="(bLUpperArmUpAlpha=False,bRUpperArmUpAlpha=False,bLUpperArmFrAlpha=False,bRUpperArmFrAlpha=False,bLUpperArmBkAlpha=False,bRUpperArmBkAlpha=False,bLWristUpAlpha=False,bRWristUpAlpha=False,bLWristDnAlpha=False,bRWristDnAlpha=False,bLKneeAlpha=False,bRKneeAlp ...(truncated)"
    Debug=(bUpperArmDebug=False,bWristDebug=False,bLatDebug=False,bGroinGluteDebug=False)
    Value="(UpperArmUp=(Rotation=(X=0.0,Y=-0.0,Z=0.0,W=1.0),Translation=(X=0.0,Y=0.0,Z=8.0),Scale3D=(X=0.70,Y=1.0,Z=1.0)),UpperArmFwdFr=(Rotation=(X=0.0,Y=-0.0,Z=0.0,W=1.0),Translation=(X=0.0,Y=10.0,Z=0.0),Scale3D=(X=0.50,Y=1.0,Z=1.0)),UpperArmFwdBk=(Rotation=(X=0.0, ...(truncated)"
    ExecutePin=@RigVMFunction_Sequence.B
   ) {is=struct RigVMUnitNode}
   > VariableSetup = RigVMFunctionReferenceNode()
   > Robe = RigVMFunctionReferenceNode()
  ]}
]
hierarchy [
 root {loc=<0 0 0>, rot=<0 0 0 1>}
 pelvis {parent=root, loc=<0 -0.5301 98.0571>, rot=<0.0127 0.707 -0.0127 -0.707>}
 spine_01 {parent=pelvis, loc=<6.1208 -0.9328 0>, rot=<0 0 0.034 -0.9994>}
 spine_02 {parent=spine_01, loc=<6.5996 -0.2095 0>, rot=<0 0 0.0238 -0.9997>}
 spine_03 {parent=spine_02, loc=<6.4914 -0.1237 0>, rot=<0 0 0.0149 -0.9999>}
 spine_04 {parent=spine_03, loc=<6.9598 0.0794 0>, rot=<0 0 -0.0132 -0.9999>}
 spine_05 {parent=spine_04, loc=<6.9987 0 0>, rot=<0 0 -0.0015 -1>}
 neck_01 {parent=spine_05, loc=<11.3138 0.1529 0>, rot=<0 0 -0.2003 -0.9797>}
 neck_02 {parent=neck_01, loc=<4.5627 0 0>, rot=<0 0 0.0504 -0.9987>}
 head {parent=neck_02, loc=<5.3914 0 0>, rot=<0 0 0.115 -0.9934>}
 attach_cape {parent=spine_05, loc=<9.7215 -9.4575 0>, rot=<0.1328 -0.6945 -0.1328 -0.6945>}
 cape {parent=attach_cape, type=Null, loc=<0 4.0743 -0.7283>, rot=<-0.0449 0 0 0.999>}
 attach_backpack {parent=spine_05, loc=<-2.006 -14.6941 0>, rot=<0.0353 -0.7062 -0.0353 -0.7062>}
 Backpack_BR {parent=attach_backpack, type=Null, loc=<0 4.5 0>, rot=<0.061 0 0 0.9981>}
 clavicle_l {parent=spine_05, loc=<5.4831 -0.1053 -3.2946>, rot=<0.0465 -0.7289 -0.009 -0.6829>}
 upperarm_l {parent=clavicle_l, loc=<9.8658 -1.2136 -0.4105>, rot=<0.0164 -0.3245 -0.0269 -0.9454>}
 lowerarm_l {parent=upperarm_l, loc=<26.465 -0.8345 0.259>, rot=<0.0036 0.0161 -0.2298 -0.9731>}
 hand_l {parent=lowerarm_l, loc=<22.7355 -0.3203 -0.6105>, rot=<-0.6784 -0.0223 0.0735 -0.7307>}
 thumb_01_l {parent=hand_l, loc=<1.3118 -0.3191 -2.459>, rot=<0.6546 -0.4042 0.0664 -0.6354>}
 thumb_02_l {parent=thumb_01_l, loc=<5.9336 -0.0448 0.0401>, rot=<-0.0037 0.0213 0.298 -0.9543>}
 thumb_03_l {parent=thumb_02_l, loc=<2.2223 0.3682 -0.0907>, rot=<0.0118 -0.013 0.0066 -0.9998>}
 index_metacarpal_l {parent=hand_l, loc=<1.6535 0.0183 -2.96>, rot=<0.0158 -0.0124 -0.0173 -0.9996>}
 index_01_l {parent=index_metacarpal_l, loc=<8.3008 0.0357 -0.0524>, rot=<-0.0044 0.0485 0.0924 -0.9945>}
 index_02_l {parent=index_01_l, loc=<3.3587 -0.3004 0.0452>, rot=<-0.002 0.0027 0.3225 -0.9466>}
 index_03_l {parent=index_02_l, loc=<2.5578 -0.123 -0.0252>, rot=<-0.002 0.0067 0.2558 -0.9667>}
 middle_metacarpal_l {parent=hand_l, loc=<2.1856 -0.1725 -0.5984>, rot=<0.006 0.0168 -0.0329 -0.9993>}
 middle_01_l {parent=middle_metacarpal_l, loc=<7.3275 -0.2447 0.0129>, rot=<-0.0208 0.0042 0.0858 -0.9961>}
 middle_02_l {parent=middle_01_l, loc=<4.0123 0.1401 0.1175>, rot=<-0.0218 0.0258 0.3512 -0.9357>}
 middle_03_l {parent=middle_02_l, loc=<2.6077 0.0534 0.1373>, rot=<0.0082 0.0188 0.1708 -0.9851>}
 ring_metacarpal_l {parent=hand_l, loc=<2.2074 -0.3625 1.3638>, rot=<-0.0284 0.0722 -0.0458 -0.9959>}
 ring_01_l {parent=ring_metacarpal_l, loc=<7.0295 -0.6197 -0.1446>, rot=<-0.0489 -0.0252 0.1892 -0.9804>}
 ring_02_l {parent=ring_01_l, loc=<3.6174 0.4647 -0.1027>, rot=<0 0 0.2806 -0.9598>}
 ring_03_l {parent=ring_02_l, loc=<2.4152 0.114 0.0199>, rot=<0 0 0.1189 -0.9929>}
 pinky_metacarpal_l {parent=hand_l, loc=<1.7616 -0.9753 3.3814>, rot=<-0.0397 0.0814 -0.069 -0.9935>}
 pinky_01_l {parent=pinky_metacarpal_l, loc=<7.1709 -0.4831 -0.1094>, rot=<-0.0654 -0.0035 0.2242 -0.9723>}
 pinky_02_l {parent=pinky_01_l, loc=<2.8274 0.0893 -0.1043>, rot=<0.0037 0.0011 0.2854 -0.9584>}
 pinky_03_l {parent=pinky_02_l, loc=<2.2046 0.0307 -0.0078>, rot=<-0.0002 0.0098 0.1445 -0.9895>}
 hand_attach_l {parent=hand_l, loc=<8.5 -3 0>, rot=<0.7071 0.7071 0 0>}
 weapon_l {parent=hand_l, loc=<-0.1974 -3.4111 0.3811>, rot=<0 0 0 1>}
 lowerarm_twist_02_l {parent=lowerarm_l, loc=<10.6707 0 0>, rot=<0 0 0 1>}
 lowerarm_l_front {parent=lowerarm_twist_02_l, loc=<-10.6707 0.5246 0>, rot=<0 0 0 -1>}
 lowerarm_l_back {parent=lowerarm_twist_02_l, loc=<-10.6707 -0.4766 0>, rot=<0 0 0 -1>}
 lowerarm_twist_01_l {parent=lowerarm_l, loc=<18.316 0 0>, rot=<0 0 0 1>}
 dyn_Cloth_lowerarm_l {parent=lowerarm_twist_01_l, loc=<-13.2236 0.085 2.6897>, rot=<0.2399 -0.6171 0.6392 0.3913>}
 dyn_Cloth_lowerarm_l_A {parent=dyn_Cloth_lowerarm_l, loc=<-5.2302 -9.6435 -1.3163>, rot=<0 0 0 1>}
 dyn_Cloth_lowerarm_l_B01 {parent=dyn_Cloth_lowerarm_l, loc=<-3.5403 0.5991 -4.5567>, rot=<-0.1767 0 0 -0.9843>}
 dyn_Cloth_lowerarm_l_B02 {parent=dyn_Cloth_lowerarm_l_B01, loc=<-5.2759 -8.2336 0>, rot=<0 0 0 1>}
 dyn_Cloth_lowerarm_l_C01 {parent=dyn_Cloth_lowerarm_l, loc=<-9.4954 0.7927 0.3135>, rot=<-0.1767 0 0 -0.9843>}
 dyn_Cloth_lowerarm_l_C02 {parent=dyn_Cloth_lowerarm_l_C01, loc=<-6.9026 -6.9304 1.7741>, rot=<0 0 0 1>}
 dyn_Cloth_lowerarm_l_D01 {parent=dyn_Cloth_lowerarm_l, loc=<-4.0352 -0.5173 3.8454>, rot=<-0.1767 0 0 -0.9843>}
 dyn_Cloth_lowerarm_l_D02 {parent=dyn_Cloth_lowerarm_l_D01, loc=<-5.5161 -7.4602 1.7741>, rot=<0 0 0 1>}
 lowerarm_twist_01_l_scale {parent=lowerarm_twist_01_l, loc=<1 0 0>, rot=<0 0 0 1>}
 deform_elbow_in_base_l {parent=lowerarm_l, loc=<0.1834 3.6062 -0.2572>, rot=<0 0 0 -1>}
 deform_elbow_in_l {parent=deform_elbow_in_base_l, loc=<0 0 0>, rot=<0 0 0 1>}
 deform_elbow_out_base_l {parent=lowerarm_l, loc=<0 -4 0>, rot=<0 0 0 1>}
 deform_elbow_out_l {parent=deform_elbow_out_base_l, loc=<0 0 0>, rot=<0 0 0 1>}
 upperarm_twist_01_l {parent=upperarm_l, loc=<7.2135 -0.1134 0>, rot=<0 0 0 1>}
 flont_uppelalm_l {parent=upperarm_twist_01_l, loc=<-7.2177 1 0.0001>, rot=<0 0 0 1>}
 up_uppelalm_l {parent=upperarm_twist_01_l, loc=<-7.2179 0.113 1>, rot=<0 0 0 1>}
 back_uppelalm_l {parent=upperarm_twist_01_l, loc=<-7.2177 -1 0.0001>, rot=<0 0 0 1>}
 deform_upperarm_bk_base_l {parent=upperarm_twist_01_l, loc=<-6.4053 -4.3746 0.314>, rot=<0 0 0 -1>}
 deform_upperarm_bk_l {parent=deform_upperarm_bk_base_l, loc=<0 0 0>, rot=<0 0 0 1>}
 deform_upperarm_fr_base_l {parent=upperarm_twist_01_l, loc=<-5.6312 5.092 0.3354>, rot=<0 0 0 -1>}
 deform_upperarm_fr_l {parent=deform_upperarm_fr_base_l, loc=<0 0 0>, rot=<0 0 0 1>}
 deform_upperarm_up_base_l {parent=upperarm_twist_01_l, loc=<-8.891 0.4825 3.9517>, rot=<0 0 0 -1>}
 deform_upperarm_up_l {parent=deform_upperarm_up_base_l, loc=<0 0 0>, rot=<0 0 0 1>}
 upperarm_twist_02_l {parent=upperarm_l, loc=<19.7998 -0.1134 0>, rot=<0 0 0 1>}
 clavicle_r {parent=spine_05, loc=<5.4831 -0.1053 3.2946>, rot=<0.7289 0.0465 -0.6829 0.009>}
 upperarm_r {parent=clavicle_r, loc=<-9.8658 1.2136 0.4105>, rot=<0.0164 -0.3245 -0.0269 -0.9454>}
 lowerarm_r {parent=upperarm_r, loc=<-26.465 0.8345 -0.259>, rot=<0.0036 0.0161 -0.2298 -0.9731>}
 hand_r {parent=lowerarm_r, loc=<-22.7355 0.3203 0.6106>, rot=<-0.6784 -0.0223 0.0735 -0.7307>}
 thumb_01_r {parent=hand_r, loc=<-1.3118 0.3191 2.459>, rot=<0.6546 -0.4042 0.0664 -0.6354>}
 thumb_02_r {parent=thumb_01_r, loc=<-5.9336 0.0448 -0.0401>, rot=<-0.0037 0.0213 0.298 -0.9543>}
 thumb_03_r {parent=thumb_02_r, loc=<-2.2223 -0.3682 0.0907>, rot=<0.0118 -0.013 0.0066 -0.9998>}
 index_metacarpal_r {parent=hand_r, loc=<-1.6535 -0.0183 2.96>, rot=<0.0158 -0.0124 -0.0173 -0.9996>}
 index_01_r {parent=index_metacarpal_r, loc=<-8.3008 -0.0357 0.0524>, rot=<-0.0044 0.0485 0.0924 -0.9945>}
 index_02_r {parent=index_01_r, loc=<-3.3587 0.3004 -0.0452>, rot=<-0.002 0.0027 0.3225 -0.9466>}
 index_03_r {parent=index_02_r, loc=<-2.5578 0.123 0.0252>, rot=<-0.002 0.0067 0.2558 -0.9667>}
 middle_metacarpal_r {parent=hand_r, loc=<-2.1856 0.1725 0.5984>, rot=<0.006 0.0168 -0.0329 -0.9993>}
 middle_01_r {parent=middle_metacarpal_r, loc=<-7.3275 0.2447 -0.0129>, rot=<-0.0208 0.0042 0.0858 -0.9961>}
 middle_02_r {parent=middle_01_r, loc=<-4.0123 -0.1401 -0.1175>, rot=<-0.0218 0.0258 0.3512 -0.9357>}
 middle_03_r {parent=middle_02_r, loc=<-2.6077 -0.0534 -0.1373>, rot=<0.0082 0.0188 0.1708 -0.9851>}
 ring_metacarpal_r {parent=hand_r, loc=<-2.2074 0.3625 -1.3638>, rot=<-0.0284 0.0722 -0.0458 -0.9959>}
 ring_01_r {parent=ring_metacarpal_r, loc=<-7.0295 0.6197 0.1446>, rot=<-0.0489 -0.0252 0.1892 -0.9804>}
 ring_02_r {parent=ring_01_r, loc=<-3.6174 -0.4647 0.1027>, rot=<0 0 0.2806 -0.9598>}
 ring_03_r {parent=ring_02_r, loc=<-2.4152 -0.114 -0.0199>, rot=<0 0 0.1189 -0.9929>}
 pinky_metacarpal_r {parent=hand_r, loc=<-1.7616 0.9753 -3.3814>, rot=<-0.0397 0.0814 -0.069 -0.9935>}
 pinky_01_r {parent=pinky_metacarpal_r, loc=<-7.1709 0.4831 0.1094>, rot=<-0.0654 -0.0035 0.2242 -0.9723>}
 pinky_02_r {parent=pinky_01_r, loc=<-2.8274 -0.0893 0.1043>, rot=<0.0037 0.0011 0.2854 -0.9584>}
 pinky_03_r {parent=pinky_02_r, loc=<-2.2046 -0.0307 0.0078>, rot=<-0.0002 0.0098 0.1445 -0.9895>}
 weapon_r {parent=hand_r, loc=<0.1974 3.4111 -0.3811>, rot=<0 0 0 1>}
 hand_attach_r {parent=hand_r, loc=<-8.5001 2.9999 0>, rot=<0 0 -0.7071 -0.7071>}
 lowerarm_twist_02_r {parent=lowerarm_r, loc=<-10.6707 0 0>, rot=<0 0 0 -1>}
 lowerarm_r_front {parent=lowerarm_twist_02_r, loc=<10.6707 -0.4766 -0.0001>, rot=<0 0 0 1>}
 lowerarm_r_back {parent=lowerarm_twist_02_r, loc=<10.6707 0.5246 -0.0001>, rot=<0 0 0 1>}
 lowerarm_twist_01_r {parent=lowerarm_r, loc=<-18.316 0 0>, rot=<0 0 0 -1>}
 dyn_Cloth_lowerarm_r {parent=lowerarm_twist_01_r, loc=<13.2235 -0.0849 -2.6896>, rot=<-0.6245 -0.2199 0.3706 -0.6514>}
 dyn_Cloth_lowerarm_r_A {parent=dyn_Cloth_lowerarm_r, loc=<-5.838 -9.2878 1.3164>, rot=<0 0 0 1>}
 dyn_Cloth_lowerarm_r_B01 {parent=dyn_Cloth_lowerarm_r, loc=<-3.4941 0.8251 4.5567>, rot=<0.1763 -0.0113 0 -0.9843>}
 dyn_Cloth_lowerarm_r_B02 {parent=dyn_Cloth_lowerarm_r_B01, loc=<-5.7945 -7.8778 -0.0001>, rot=<0 0 0 1>}
 dyn_Cloth_lowerarm_r_C01 {parent=dyn_Cloth_lowerarm_r, loc=<-9.425 1.4007 -0.3135>, rot=<0.1763 -0.0113 0 -0.9843>}
 dyn_Cloth_lowerarm_r_C02 {parent=dyn_Cloth_lowerarm_r_C01, loc=<-7.333 -6.4729 -1.7741>, rot=<0 0 0 1>}
 dyn_Cloth_lowerarm_r_D01 {parent=dyn_Cloth_lowerarm_r, loc=<-4.0599 -0.2572 -3.8454>, rot=<0.1763 -0.0113 0 -0.9843>}
 dyn_Cloth_lowerarm_r_D02 {parent=dyn_Cloth_lowerarm_r_D01, loc=<-5.9841 -7.0906 -1.7741>, rot=<0 0 0.0321 -0.9995>}
 lowerarm_twist_01_r_scale {parent=lowerarm_twist_01_r, loc=<-1 0 0>, rot=<0 0 0 1>}
 deform_elbow_out_base_r {parent=lowerarm_r, loc=<0 4 -0.0001>, rot=<0 0 0 -1>}
 deform_elbow_out_r {parent=deform_elbow_out_base_r, loc=<0 0 0>, rot=<0 0 0 1>}
 deform_elbow_in_base_r {parent=lowerarm_r, loc=<-0.1834 -3.6062 0.2572>, rot=<0 0 0 -1>}
 deform_elbow_in_r {parent=deform_elbow_in_base_r, loc=<0 0 0>, rot=<0 0 0 1>}
 upperarm_twist_01_r {parent=upperarm_r, loc=<-7.2135 0.1134 0>, rot=<0 0 0 -1>}
 front_upperarm_r {parent=upperarm_twist_01_r, loc=<7.2136 -1 -0.0001>, rot=<0 0 0 1>}
 up_upperarm_r {parent=upperarm_twist_01_r, loc=<7.2136 -0.113 -1.0001>, rot=<0 0 0 1>}
 back_upperarm_r {parent=upperarm_twist_01_r, loc=<7.2136 1 -0.0001>, rot=<0 0 0 1>}
 deform_upperarm_up_base_r {parent=upperarm_twist_01_r, loc=<8.891 -0.4825 -3.9517>, rot=<0 0 0 -1>}
 deform_upperarm_up_r {parent=deform_upperarm_up_base_r, loc=<0 0 0>, rot=<0 0 0 1>}
 deform_upperarm_fr_base_r {parent=upperarm_twist_01_r, loc=<5.6312 -5.092 -0.3354>, rot=<0 0 0 -1>}
 deform_upperarm_fr_r {parent=deform_upperarm_fr_base_r, loc=<0 0 0>, rot=<0 0 0 1>}
 deform_upperarm_bk_base_r {parent=upperarm_twist_01_r, loc=<6.4053 4.3746 -0.314>, rot=<0 0 0 -1>}
 deform_upperarm_bk_r {parent=deform_upperarm_bk_base_r, loc=<0 0 0>, rot=<0 0 0 1>}
 upperarm_twist_02_r {parent=upperarm_r, loc=<-19.7998 0.1134 0>, rot=<0 0 0 -1>}
 dyn_Hair_BackL_01 {parent=spine_05, loc=<10.5734 -4.0321 -6.3713>, rot=<0.999 0.0401 -0.0008 -0.0201>}
 dyn_Hair_BackL_02 {parent=dyn_Hair_BackL_01, loc=<-8.263 4.2069 -0.1352>, rot=<0 0 0 1>}
 dyn_Hair_BackL_03 {parent=dyn_Hair_BackL_02, loc=<-7.0233 0.8274 -0.0333>, rot=<0 0 0 1>}
 dyn_Hair_BackL_04 {parent=dyn_Hair_BackL_03, loc=<-7.3857 0.1475 -0.0059>, rot=<0 0 0 1>}
 dyn_Hair_BackL_05 {parent=dyn_Hair_BackL_04, loc=<-8.1116 0.6002 -0.0242>, rot=<0 0 0 1>}
 dyn_Hair_BackL_06 {parent=dyn_Hair_BackL_05, loc=<-6.6208 -0.4716 -0.2398>, rot=<0 0 0 1>}
 dyn_Hair_BackM_01 {parent=spine_05, loc=<10.5734 -4.0321 -0.0376>, rot=<0.999 0.0401 -0.0008 -0.0201>}
 dyn_Hair_BackM_02 {parent=dyn_Hair_BackM_01, loc=<-8.263 4.2069 -0.1352>, rot=<0 0 0 1>}
 dyn_Hair_BackM_03 {parent=dyn_Hair_BackM_02, loc=<-7.0233 0.8274 -0.0333>, rot=<0 0 0 1>}
 dyn_Hair_BackM_04 {parent=dyn_Hair_BackM_03, loc=<-7.3857 0.1475 -0.0059>, rot=<0 0 0 1>}
 dyn_Hair_BackM_05 {parent=dyn_Hair_BackM_04, loc=<-8.1116 0.6002 -0.0242>, rot=<0 0 0 1>}
 dyn_Hair_BackM_06 {parent=dyn_Hair_BackM_05, loc=<-6.6208 -0.4716 -0.2398>, rot=<0 0 0 1>}
 dyn_Hair_BackR_01 {parent=spine_05, loc=<10.5734 -4.0321 5.6531>, rot=<0.999 0.0401 -0.0008 -0.0201>}
 dyn_Hair_BackR_02 {parent=dyn_Hair_BackR_01, loc=<-8.263 4.2069 -0.1352>, rot=<0 0 0 1>}
 dyn_Hair_BackR_03 {parent=dyn_Hair_BackR_02, loc=<-7.0233 0.8274 -0.0333>, rot=<0 0 0 1>}
 dyn_Hair_BackR_04 {parent=dyn_Hair_BackR_03, loc=<-7.3857 0.1475 -0.0059>, rot=<0 0 0 1>}
 dyn_Hair_BackR_05 {parent=dyn_Hair_BackR_04, loc=<-8.1116 0.6002 -0.0242>, rot=<0 0 0 1>}
 dyn_Hair_BackR_06 {parent=dyn_Hair_BackR_05, loc=<-6.6208 -0.4716 -0.2398>, rot=<0 0 0 1>}
 dyn_skirt_MSTR {parent=spine_01, loc=<0 0 0>, rot=<0 0 0 1>}
 dyn_front_skirt_l_MSTR {parent=dyn_skirt_MSTR, loc=<-4.3901 7.6198 -11.2383>, rot=<-0.0248 0.0328 0.9992 0.0008>}
 dyn_front_skirt_l_01 {parent=dyn_front_skirt_l_MSTR, loc=<0 0 0>, rot=<0 0 0 1>}
 dyn_front_skirt_l_02 {parent=dyn_front_skirt_l_01, loc=<22 0 0>, rot=<0 0 0 1>}
 dyn_front_skirt_l_03 {parent=dyn_front_skirt_l_02, loc=<16 0 0>, rot=<0 0 0 1>}
 dyn_front_skirt_l_04 {parent=dyn_front_skirt_l_03, loc=<15 0 0>, rot=<0 0 0 1>}
 dyn_front_skirt_l_05 {parent=dyn_front_skirt_l_04, loc=<13 0 0>, rot=<0 0 0 1>}
 dyn_front_skirt_l_06 {parent=dyn_front_skirt_l_05, loc=<14 0 0>, rot=<0 0 0 1>}
 dyn_side_skirt_l_MSTR {parent=dyn_skirt_MSTR, loc=<-4.8705 -1.7711 -14.5383>, rot=<0.0446 -0.7517 -0.6575 -0.0241>}
 dyn_side_skirt_l_01 {parent=dyn_side_skirt_l_MSTR, loc=<0 0 0>, rot=<0 0 0 1>}
 dyn_side_skirt_l_02 {parent=dyn_side_skirt_l_01, loc=<22 0 0>, rot=<0 0 0 1>}
 dyn_side_skirt_l_03 {parent=dyn_side_skirt_l_02, loc=<16 0 0>, rot=<0 0 0.0076 -1>}
 dyn_side_skirt_l_04 {parent=dyn_side_skirt_l_03, loc=<15 0 0>, rot=<0 0 0 1>}
 dyn_side_skirt_l_05 {parent=dyn_side_skirt_l_04, loc=<13 0 0>, rot=<0 0 0 1>}
 dyn_side_skirt_l_06 {parent=dyn_side_skirt_l_05, loc=<14 0 0>, rot=<0 0 0 1>}
 dyn_rear_skirt_l_MSTR {parent=dyn_skirt_MSTR, loc=<-3.9603 -9.2321 -10.8897>, rot=<0.0356 -0.999 0.0067 -0.0281>}
 dyn_rear_skirt_l_01 {parent=dyn_rear_skirt_l_MSTR, loc=<0 0 0>, rot=<0 0 0 1>}
 dyn_rear_skirt_l_02 {parent=dyn_rear_skirt_l_01, loc=<22 0 0>, rot=<0 0 0 1>}
 dyn_rear_skirt_l_03 {parent=dyn_rear_skirt_l_02, loc=<16 0 0>, rot=<0 0 0 1>}
 dyn_rear_skirt_l_04 {parent=dyn_rear_skirt_l_03, loc=<15 0 0>, rot=<0 0 0 1>}
 dyn_rear_skirt_l_05 {parent=dyn_rear_skirt_l_04, loc=<13 0 0>, rot=<0 0 0 1>}
 dyn_rear_skirt_l_06 {parent=dyn_rear_skirt_l_05, loc=<14 0 0>, rot=<0 0 0 1>}
 dyn_rear_skirt_MSTR {parent=dyn_skirt_MSTR, loc=<-4.3819 -11.2881 0>, rot=<0.034 -0.9994 0 0>}
 dyn_rear_skirt_01 {parent=dyn_rear_skirt_MSTR, loc=<0 0 0>, rot=<0 0 0 1>}
 dyn_rear_skirt_02 {parent=dyn_rear_skirt_01, loc=<22 0 0>, rot=<0 0 0 1>}
 dyn_rear_skirt_03 {parent=dyn_rear_skirt_02, loc=<16 0 0>, rot=<0 0 0 1>}
 dyn_rear_skirt_04 {parent=dyn_rear_skirt_03, loc=<15 0 0>, rot=<0 0 0 1>}
 dyn_rear_skirt_05 {parent=dyn_rear_skirt_04, loc=<13 0 0>, rot=<0 0 0 1>}
 dyn_rear_skirt_06 {parent=dyn_rear_skirt_05, loc=<14 0 0>, rot=<0 0 0 1>}
 dyn_front_skirt_MSTR {parent=dyn_skirt_MSTR, loc=<-4.8798 9.5825 0>, rot=<0 0 1 -0.006>}
 dyn_front_skirt_01 {parent=dyn_front_skirt_MSTR, loc=<0 0 0>, rot=<0 0 0 1>}
 dyn_front_skirt_02 {parent=dyn_front_skirt_01, loc=<22 0 0>, rot=<0 0 0 1>}
 dyn_front_skirt_03 {parent=dyn_front_skirt_02, loc=<16 0 0>, rot=<0 0 0 1>}
 dyn_front_skirt_04 {parent=dyn_front_skirt_03, loc=<15 0 0>, rot=<0 0 0 1>}
 dyn_front_skirt_05 {parent=dyn_front_skirt_04, loc=<13 0 0>, rot=<0 0 0 1>}
 dyn_front_skirt_06 {parent=dyn_front_skirt_05, loc=<14 0 0>, rot=<0 0 0 1>}
 dyn_front_skirt_r_MSTR {parent=dyn_skirt_MSTR, loc=<-4.3901 7.6198 11.2383>, rot=<-0.0332 -0.0244 -0.0118 -0.9991>}
 dyn_front_skirt_r_01 {parent=dyn_front_skirt_r_MSTR, loc=<0 0 0>, rot=<0 0 0 1>}
 dyn_front_skirt_r_02 {parent=dyn_front_skirt_r_01, loc=<-22.0003 0 0>, rot=<0 0 0 1>}
 dyn_front_skirt_r_03 {parent=dyn_front_skirt_r_02, loc=<-16 0 0>, rot=<0 0 0 1>}
 dyn_front_skirt_r_04 {parent=dyn_front_skirt_r_03, loc=<-15 0 0>, rot=<0 0 0 1>}
 dyn_front_skirt_r_05 {parent=dyn_front_skirt_r_04, loc=<-13 0 0>, rot=<0 0 0 1>}
 dyn_front_skirt_r_06 {parent=dyn_front_skirt_r_05, loc=<-14.0001 0 0>, rot=<0 0 0 1>}
 dyn_side_skirt_r_MSTR {parent=dyn_skirt_MSTR, loc=<-4.8705 -1.7711 14.5383>, rot=<-0.7516 -0.0467 0.026 -0.6575>}
 dyn_side_skirt_r_01 {parent=dyn_side_skirt_r_MSTR, loc=<0 0 0>, rot=<0 0 0 1>}
 dyn_side_skirt_r_02 {parent=dyn_side_skirt_r_01, loc=<-22 0 0>, rot=<0 0 0 1>}
 dyn_side_skirt_r_03 {parent=dyn_side_skirt_r_02, loc=<-16 -0.0001 0>, rot=<0 0 0.0076 -1>}
 dyn_side_skirt_r_04 {parent=dyn_side_skirt_r_03, loc=<-15 0.0001 0>, rot=<0 0 0 1>}
 dyn_side_skirt_r_05 {parent=dyn_side_skirt_r_04, loc=<-13 0 0>, rot=<0 0 0 1>}
 dyn_side_skirt_r_06 {parent=dyn_side_skirt_r_05, loc=<-14.0001 0 0>, rot=<0 0 0 1>}
 dyn_rear_skirt_r_MSTR {parent=dyn_skirt_MSTR, loc=<-3.9603 -9.2321 10.8897>, rot=<0.999 0.0356 -0.0281 -0.0067>}
 dyn_rear_skirt_r_01 {parent=dyn_rear_skirt_r_MSTR, loc=<0 0 0>, rot=<0 0 0 1>}
 dyn_rear_skirt_r_02 {parent=dyn_rear_skirt_r_01, loc=<-22.0001 0 0>, rot=<0 0 0 1>}
 dyn_rear_skirt_r_03 {parent=dyn_rear_skirt_r_02, loc=<-16 -0.0001 0>, rot=<0 0 0 -1>}
 dyn_rear_skirt_r_04 {parent=dyn_rear_skirt_r_03, loc=<-15 0 0>, rot=<0 0 0 1>}
 dyn_rear_skirt_r_05 {parent=dyn_rear_skirt_r_04, loc=<-13 0 0>, rot=<0 0 0 1>}
 dyn_rear_skirt_r_06 {parent=dyn_rear_skirt_r_05, loc=<-14 0 0>, rot=<0 0 0 1>}
 dyn_skirt_SIMSPACE {parent=spine_01, loc=<0 0 0>, rot=<0 0 0 1>}
 spine_05_null {parent=spine_01, type=Null, loc=<27.0126 -1.456 0>, rot=<0 0 -0.024 0.9997>}
 thigh_l {parent=pelvis, loc=<-1.0855 -0.2109 -8.7053>, rot=<0.0009 0.0355 0.0155 -0.9993>}
 calf_l {parent=thigh_l, loc=<-42.5284 -0.0019 -0.7297>, rot=<0.0006 -0.0161 -0.0391 -0.9991>}
 foot_l {parent=calf_l, loc=<-46.7401 -2.5014 -0.592>, rot=<0.0427 -0.0238 0.0353 -0.9982>}
 ball_l {parent=foot_l, loc=<-3.5658 10.8066 -0.1544>, rot=<0.0001 -0.0001 -0.7063 -0.7079>}
 calf_twist_02_l {parent=calf_l, loc=<-19.1546 -1.7943 -0.2638>, rot=<0.0237 0.0095 -0.0035 -0.9997>}
 calf_twist_01_l {parent=calf_l, loc=<-35.8483 -1.9197 -0.5794>, rot=<0.0237 0.0095 -0.0035 -0.9997>}
 thigh_twist_01_l {parent=thigh_l, loc=<-27.952 0 0>, rot=<0.0013 0 0 -1>}
 thigh_volume_fn_l {parent=thigh_twist_01_l, loc=<27.952 1.5 0.0038>, rot=<0.9998 0 -0.0217 -0.0013>}
 thigh_volume_lt_l1 {parent=thigh_twist_01_l, loc=<27.952 0.0038 -1.5>, rot=<-0.7078 -0.0154 0.0153 -0.706>}
 thigh_volume_bk_l {parent=thigh_twist_01_l, loc=<27.952 -1.5 -0.0038>, rot=<-0.0013 0 0 -1>}
 deform_thigh_notwist_l {parent=thigh_l, loc=<0 0 0>, rot=<0 0.0217 0 -0.9998>}
 calf_l_null {parent=thigh_l, type=Null, loc=<-42.5284 -0.0019 -0.7297>, rot=<0.0006 -0.0161 -0.0391 -0.9991>}
 thigh_r {parent=pelvis, loc=<-1.0855 -0.2109 8.7053>, rot=<0.0355 -0.0009 0.9993 0.0155>}
 calf_r {parent=thigh_r, loc=<42.5284 0.0019 0.7297>, rot=<0.0006 -0.0161 -0.0391 -0.9991>}
 foot_r {parent=calf_r, loc=<46.7401 2.5014 0.592>, rot=<0.0427 -0.0238 0.0353 -0.9982>}
 ball_r {parent=foot_r, loc=<3.5658 -10.8066 0.1544>, rot=<0.0001 -0.0001 -0.7063 -0.7079>}
 calf_twist_02_r {parent=calf_r, loc=<19.1546 1.7943 0.2638>, rot=<0.0237 0.0095 -0.0035 -0.9997>}
 calf_twist_01_r {parent=calf_r, loc=<35.8483 1.9197 0.5794>, rot=<0.0237 0.0095 -0.0035 -0.9997>}
 thigh_twist_01_r {parent=thigh_r, loc=<27.952 0 0>, rot=<0.0013 0 0 -1>}
 thigh_volume_fn_r {parent=thigh_twist_01_r, loc=<-27.9522 -1.5 -0.0038>, rot=<-0.0013 0 0 -1>}
 thigh_volume_rt_r {parent=thigh_twist_01_r, loc=<-27.9522 -0.0038 1.5>, rot=<0.7062 0 0 -0.708>}
 thigh_volume_bk_r {parent=thigh_twist_01_r, loc=<-27.9522 1.5 0.0038>, rot=<1 0 0 -0.0013>}
 deform_thigh_notwist_r {parent=thigh_r, loc=<0 0 0>, rot=<0 0.0217 0 -0.9998>}
 calf_r_null {parent=thigh_r, type=Null, loc=<42.5284 0.0019 0.7297>, rot=<0.0006 -0.0161 -0.0391 -0.9991>}
 cr_thigh_r {parent=pelvis, loc=<-2.5693 -0.0109 7.7501>, rot=<-0.0004 -0.0572 0 0.9984>}
 cr_thigh_no_twist_r {parent=cr_thigh_r, loc=<0 0 0>, rot=<0 0 0 1>}
 dyn_Cloth_volume_fn {parent=pelvis, loc=<-6.5428 8.5934 0.1611>, rot=<0.9998 -0.0179 0 0>}
 dyn_Cloth_volume_fn01 {parent=dyn_Cloth_volume_fn, loc=<-14.7676 -1.5293 0>, rot=<0 0 0 1>}
 dyn_Cloth_volume_fn02 {parent=dyn_Cloth_volume_fn01, loc=<-22.6974 -1.0417 0.0916>, rot=<0.024 0 0 -0.9997>}
 dyn_Cloth_volume_fn03 {parent=dyn_Cloth_volume_fn02, loc=<-16.5606 0 0>, rot=<0 0 0 1>}
 dyn_Cloth_volume_fn04 {parent=dyn_Cloth_volume_fn03, loc=<-17.0687 0 0>, rot=<0 0 0 1>}
 dyn_Cloth_volume_bk {parent=pelvis, loc=<-6.8808 -8.964 0.2721>, rot=<0 0 0.0179 -0.9998>}
 dyn_Cloth_volume_bk01 {parent=dyn_Cloth_volume_bk, loc=<-14.7845 -1.9699 0>, rot=<0 0 0 1>}
 dyn_Cloth_volume_bk02 {parent=dyn_Cloth_volume_bk01, loc=<-22.9717 -1.6111 -0.2026>, rot=<0.024 0 0 -0.9997>}
 dyn_Cloth_volume_bk03 {parent=dyn_Cloth_volume_bk02, loc=<-16.5606 0 0>, rot=<0 0 0 1>}
 dyn_Cloth_volume_bk04 {parent=dyn_Cloth_volume_bk03, loc=<-17.0687 0 0>, rot=<0 0 0 1>}
 dyn_Cloth_volume_l_A {parent=pelvis, loc=<-6.0375 5.718 -9.1939>, rot=<0.9888 -0.0177 -0.0027 0.148>}
 dyn_Cloth_volume_l_A01 {parent=dyn_Cloth_volume_l_A, loc=<-16.8081 -1.2971 0.6872>, rot=<0 0 0 1>}
 dyn_Cloth_volume_l_A02 {parent=dyn_Cloth_volume_l_A01, loc=<-20.3614 -1.2837 0.4445>, rot=<0 0 0 1>}
 dyn_Cloth_volume_l_A03 {parent=dyn_Cloth_volume_l_A02, loc=<-16.8081 0 0>, rot=<0 0 0 1>}
 dyn_Cloth_volume_l_A04 {parent=dyn_Cloth_volume_l_A03, loc=<-17.0247 0 0>, rot=<0 0 0 1>}
 dyn_Cloth_volume_l_B {parent=pelvis, loc=<-6.161 -1.397 -13.2219>, rot=<-0.7102 0.0127 0.0126 -0.7037>}
 dyn_Cloth_volume_l_B01 {parent=dyn_Cloth_volume_l_B, loc=<-17.06 -1.8458 0>, rot=<0 0 0 1>}
 dyn_Cloth_volume_l_B02 {parent=dyn_Cloth_volume_l_B01, loc=<-20.1095 -2.0721 0>, rot=<0 0 0 1>}
 dyn_Cloth_volume_l_B03 {parent=dyn_Cloth_volume_l_B02, loc=<-17.06 -1.4023 0>, rot=<0 0 0 1>}
 dyn_Cloth_volume_l_B04 {parent=dyn_Cloth_volume_l_B03, loc=<-16.9858 -1.239 0>, rot=<0 0 0 1>}
 dyn_Cloth_volume_l_C {parent=pelvis, loc=<-6.4944 -7.029 -8.659>, rot=<-0.1379 0.0025 0.0177 -0.9903>}
 dyn_Cloth_volume_l_C01 {parent=dyn_Cloth_volume_l_C, loc=<-16.8081 -1.5667 -0.6841>, rot=<0 0 0 1>}
 dyn_Cloth_volume_l_C02 {parent=dyn_Cloth_volume_l_C01, loc=<-20.3614 -1.8288 -0.8944>, rot=<0 0 0 1>}
 dyn_Cloth_volume_l_C03 {parent=dyn_Cloth_volume_l_C02, loc=<-16.8081 -0.907 0>, rot=<0 0 0 1>}
 dyn_Cloth_volume_l_C04 {parent=dyn_Cloth_volume_l_C03, loc=<-17.0247 -0.4758 0>, rot=<0 0 0 1>}
 dyn_Cloth_volume_r_A {parent=pelvis, loc=<-6.0375 5.718 9.1939>, rot=<0.9888 -0.0177 0.0027 -0.148>}
 dyn_Cloth_volume_r_A01 {parent=dyn_Cloth_volume_r_A, loc=<-16.8081 -1.2971 -0.6872>, rot=<0 0 0.0286 -0.9996>}
 dyn_Cloth_volume_r_A02 {parent=dyn_Cloth_volume_r_A01, loc=<-20.2545 -2.4471 -0.4444>, rot=<0 0 0 1>}
 dyn_Cloth_volume_r_A03 {parent=dyn_Cloth_volume_r_A02, loc=<-16.7805 -0.9621 0>, rot=<0 0 0 1>}
 dyn_Cloth_volume_r_A04 {parent=dyn_Cloth_volume_r_A03, loc=<-16.9968 -0.9745 0>, rot=<0 0 -0.0286 -0.9996>}
 dyn_Cloth_volume_r_B {parent=pelvis, loc=<-6.161 -1.397 13.2219>, rot=<0.7101 -0.0188 0.0066 -0.7038>}
 dyn_Cloth_volume_r_B01 {parent=dyn_Cloth_volume_r_B, loc=<-17.089 -1.5542 0>, rot=<0 0 0 1>}
 dyn_Cloth_volume_r_B02 {parent=dyn_Cloth_volume_r_B01, loc=<-20.142 -1.7284 0>, rot=<0 0 0 1>}
 dyn_Cloth_volume_r_B03 {parent=dyn_Cloth_volume_r_B02, loc=<-17.0815 -1.1107 0>, rot=<0 0 0 1>}
 dyn_Cloth_volume_r_B04 {parent=dyn_Cloth_volume_r_B03, loc=<-17.0044 -0.9487 0>, rot=<0 0 0.0085 -1>}
 dyn_Cloth_volume_r_C {parent=pelvis, loc=<-6.4944 -7.029 8.6589>, rot=<0.1379 -0.0042 0.005 -0.9904>}
 dyn_Cloth_volume_r_C01 {parent=dyn_Cloth_volume_r_C, loc=<-16.8429 -1.133 0.6841>, rot=<0 0 0 1>}
 dyn_Cloth_volume_r_C02 {parent=dyn_Cloth_volume_r_C01, loc=<-20.4018 -1.3035 0.8944>, rot=<0 0 0 1>}
 dyn_Cloth_volume_r_C03 {parent=dyn_Cloth_volume_r_C02, loc=<-16.8259 -0.4737 0>, rot=<0 0 0 1>}
 dyn_Cloth_volume_r_C04 {parent=dyn_Cloth_volume_r_C03, loc=<-17.0313 -0.037 0>, rot=<0 0 0.0129 -0.9999>}
 thigh_r_null {parent=pelvis, type=Null, loc=<-1.0855 -0.2109 8.7053>, rot=<0.0355 -0.0009 0.9993 0.0155>}
 thigh_l_null {parent=pelvis, type=Null, loc=<-1.0855 -0.2109 -8.7053>, rot=<0.0009 0.0355 0.0155 -0.9993>}
 dyn_skirt_MSTR_null {parent=pelvis, type=Null, loc=<6.1208 -0.9328 0>, rot=<0 0 -0.034 0.9994>}
 dyn_skirt_ground_collision {parent=root, loc=<0 0 0>, rot=<0 0 0 1>}
 collision {parent=root, loc=<0 0 0>, rot=<0 0 0 1>}
 collision_01 {parent=root, loc=<0 0 0>, rot=<0 0 0 1>}
 camera_root {parent=root, loc=<0 0 0>, rot=<0 0 0 1>}
 camera {parent=camera_root, loc=<0 0 0>, rot=<0 0 0 1>}
 CR Rear Knee Bend L {type=Curve, loc=<0 0 0>, rot=<0 0 0 1>}
 CR Rear Knee Bend R {type=Curve, loc=<0 0 0>, rot=<0 0 0 1>}
 CR Foot Near Pelvis L {type=Curve, loc=<0 0 0>, rot=<0 0 0 1>}
 CR Foot Near Pelvis R {type=Curve, loc=<0 0 0>, rot=<0 0 0 1>}
 CR Thigh Raised Front L {type=Curve, loc=<0 0 0>, rot=<0 0 0 1>}
 CR Thigh Raised Front R {type=Curve, loc=<0 0 0>, rot=<0 0 0 1>}
 CR Thigh L Across R {type=Curve, loc=<0 0 0>, rot=<0 0 0 1>}
 CR Thigh R Across L {type=Curve, loc=<0 0 0>, rot=<0 0 0 1>}
 thigh_l_ctrl {controlType=EULER_TRANSFORM, type=Control, loc=<0 0 0>, rot=<0 0 0 1>}
 calf_l_ctrl {controlType=EULER_TRANSFORM, parent=thigh_l_ctrl, type=Control, loc=<0 0 0>, rot=<0 0 0 1>}
 thigh_r_ctrl {controlType=EULER_TRANSFORM, type=Control, loc=<0 0 0>, rot=<0 0 0 1>}
 calf_r_ctrl {controlType=EULER_TRANSFORM, parent=thigh_r_ctrl, type=Control, loc=<0 0 0>, rot=<0 0 0 1>}
 dyn_skirt_MSTR_ctrl {controlType=EULER_TRANSFORM, type=Control, loc=<0 0 0>, rot=<0 0 0 1>}
]
variables [
 Calf_L_BackRotation {type=double, default=0}
 Calf_R_BackRotation {type=double, default=0}
 Cloth Enabled {type=bool, default=False}
 Disable Floor Slope {type=bool, default=False}
 Ground Collision Target {type=FTransform, default=(Rotation=(X=0,Y=0,Z=0,W=1),Translation=(X=0,Y=0,Z=0),Scale3D=(X=1,Y=1,Z=1))}
 Is Baseball Sliding_CR {type=bool, default=False}
 Is Knee Sliding_CR {type=bool, default=False}
 IsPlayingEmoteCR {type=bool, default=False}
 Leg ROM_Result Alpha {type=double, default=1}
 Move Collision to Narnia {type=double, default=0}
 Slope Data {type=FFortAnimInput_SlopeInfo, default=()}
 Slope Quat {type=FQuat, default=(X=0,Y=0,Z=0,W=1)}
 Thigh_L_BackRotation {type=double, default=0}
 Thigh_L_FrontRotation {type=double, default=0}
 Thigh_L_SideRotation {type=double, default=0}
 Thigh_R_BackRotation {type=double, default=0}
 Thigh_R_FrontRotation {type=double, default=0}
 Thigh_R_SideRotation {type=double, default=0}
 isCrouching_CR {type=bool, default=False}
 isFEidle_CR {type=bool, default=False}
 isKneeSlidingFloat {type=double, default=0}
 isSkydivingFloat {type=double, default=0}
 isSkydiving_CR {type=bool, default=False}
 l_arm_front_spr {type=float, default=0}
 l_arm_rear_spr {type=float, default=0}
 l_arm_up_spr {type=float, default=0}
 r_arm_front_spr {type=float, default=0}
 r_arm_rear_spr {type=float, default=0}
 r_arm_up_spr {type=float, default=0}
 skirtRotation {type=double, default=0}
 spineTwist_R {type=double, default=0}
]
