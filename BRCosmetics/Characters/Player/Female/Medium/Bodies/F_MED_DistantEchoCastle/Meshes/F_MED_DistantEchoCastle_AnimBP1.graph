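# Post-process animation Blueprint for the DistantEchoCastle female body cosmetic, adding secondary motion to its hair, sleeves and robe skirt.
# It copies the parent mesh pose into a Control Rig driven by movement-state flags, then layers bone constraints, AnimDynamics hair and sleeve springs, and weapon or hand-shape sleeve offsets through cached poses.
# The skirt blends between two RigidBody setups by skydiving state, with center bones constrained midway between left and right chains.
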AnimBlueprint /BRCosmetics/Characters/Player/Female/Medium/Bodies/F_MED_DistantEchoCastle/Meshes/F_MED_DistantEchoCastle_AnimBP1
target_skeleton=&/F_MED_DistantEchoCastle_Skeleton
graphs [
 AnimGraph {nodes [
   n1 = RigidBody!(
    ComponentPose=Local To Component(LocalPose=Use cached pose 'postSlopePosing'())
    Alpha=0.8
    OverrideWorldGravity:Vector="0,0,-980"
    ExternalForce
    SimSpaceSettings:Sim Space Settings Structure=(WorldAlpha=1,VelocityScaleZ=0.4,DampingAlpha=0,MaxLinearVelocity=400,MaxAngularVelocity=7.5,MaxLinearAcceleration=3000,MaxAngularAcceleration=30,ExternalLinearDragV=(X=0.5,Y=0.5,Z=0.2),ExternalLinearVelocity=(X=0,Y=0,Z=0),ExternalAngularVelocity=(X=0,Y=0,Z=0))
   ) -> (Pose) {
    bOverrideWorldGravity=true
    overridePhysicsAsset=&/F_MED_DistantEchoCastle_Robe_PhysicsFE
    simulationSpace=BaseBoneSpace
    alphaBoolBlend {blendOption=Linear}
    baseBoneRef {boneName=pelvis}
    componentAppliedLinearAccClamp=<0 0 0>
    overrideWorldGravity {z=-980}
    simSpaceSettings {
     dampingAlpha=0
     maxAngularAcceleration=30
     maxAngularVelocity=7.5
     maxLinearAcceleration=3000
     maxLinearVelocity=400
     velocityScaleZ=0.4
     worldAlpha=1
     externalLinearDragV=<0.5 0.5 0.2>
    }
   }
   n2 = Component To Local(ComponentPose=@n1)
   n3 = Get isMovingNotCrouching() -> (isMovingNotCrouching:bool)
   n4 = "Transform (Modify) Bone - Bone: dyn_side_skirt_l_MSTR"!(
    ComponentPose=Local To Component(LocalPose=Use cached pose 'postSlopePosing'())
    bAlphaBoolEnabled:bool=@n3
    Translation="0,0,0"
    Rotation="0,-10,0"
    Scale="1,1,1"
   ) -> (Pose) {
    alphaInputType=Bool
    rotationMode=BMM_Additive
    rotationSpace=BCS_BoneSpace
    alphaBoolBlend {blendInTime=0.3, blendOutTime=0.2}
    boneToModify {boneName=dyn_side_skirt_l_MSTR}
    rotation {yaw=-10}
   }
   n5 = "Transform (Modify) Bone - Bone: dyn_side_skirt_r_MSTR"!(ComponentPose=@n4, bAlphaBoolEnabled:bool=@n3, Translation="0,0,0", Rotation="0,-10,0", Scale="1,1,1") -> (Pose) {
    alphaInputType=Bool
    rotationMode=BMM_Additive
    rotationSpace=BCS_BoneSpace
    alphaBoolBlend {blendInTime=0.3, blendOutTime=0.2}
    boneToModify {boneName=dyn_side_skirt_r_MSTR}
    rotation {yaw=-10}
   }
   n6 = "Transform (Modify) Bone - Bone: dyn_rear_skirt_l_MSTR"!(ComponentPose=@n5, bAlphaBoolEnabled:bool=@n3, Translation="0,0,0", Rotation="0,-15,0", Scale="1,1,1") -> (Pose) {
    alphaInputType=Bool
    rotationMode=BMM_Additive
    rotationSpace=BCS_BoneSpace
    alphaBoolBlend {blendInTime=0.3, blendOutTime=0.2}
    boneToModify {boneName=dyn_rear_skirt_l_MSTR}
    rotation {yaw=-15}
   }
   n7 = "Transform (Modify) Bone - Bone: dyn_rear_skirt_r_MSTR"!(ComponentPose=@n6, bAlphaBoolEnabled:bool=@n3, Translation="0,0,0", Rotation="0,-15,0", Scale="1,1,1") -> (Pose) {
    alphaInputType=Bool
    rotationMode=BMM_Additive
    rotationSpace=BCS_BoneSpace
    alphaBoolBlend {blendInTime=0.3, blendOutTime=0.2}
    boneToModify {boneName=dyn_rear_skirt_r_MSTR}
    rotation {yaw=-15}
   }
   n8 = "Transform (Modify) Bone - Bone: dyn_rear_skirt_MSTR"!(ComponentPose=@n7, bAlphaBoolEnabled:bool=@n3, Translation="0,0,0", Rotation="0,-15,0", Scale="1,1,1") -> (Pose) {
    alphaInputType=Bool
    rotationMode=BMM_Additive
    rotationSpace=BCS_BoneSpace
    alphaBoolBlend {blendInTime=0.3, blendOutTime=0.2}
    boneToModify {boneName=dyn_rear_skirt_MSTR}
    rotation {yaw=-15}
   }
   n9 = RigidBody!(
    ComponentPose=@n8
    OverrideWorldGravity="0,0,-980"
    ExternalForce
    SimSpaceSettings=(WorldAlpha=0.1,VelocityScaleZ=0,DampingAlpha=1,MaxLinearVelocity=50,MaxAngularVelocity=2,MaxLinearAcceleration=150,MaxAngularAcceleration=20,ExternalLinearDragV=(X=0.1,Y=0.1,Z=0.1),ExternalLinearVelocity=(X=0,Y=0,Z=0),ExternalAngularVelocity=(X=0,Y=0,Z=0))
   ) -> (Pose) {
    alpha=0.8
    bOverrideWorldGravity=true
    overridePhysicsAsset=&/F_MED_DistantEchoCastle_RobeRight_Physics
    simulationSpace=BaseBoneSpace
    alphaBoolBlend {blendOption=Linear}
    baseBoneRef {boneName=pelvis}
    componentAppliedLinearAccClamp=<0 0 0>
    overrideWorldGravity {z=-980}
    simSpaceSettings {
     maxAngularAcceleration=20
     maxAngularVelocity=2
     maxLinearAcceleration=150
     maxLinearVelocity=50
     velocityScaleZ=0
     worldAlpha=0.1
     externalLinearDragV=<0.1 0.1 0.1>
    }
   }
   n10 = Get bIsMoving2D()
   n11 = "Transform (Modify) Bone - Bone: dyn_front_skirt_l_MSTR"!(ComponentPose=@n9, bAlphaBoolEnabled:bool=@n10, Translation="0,0,0", Rotation="0,-5,0", Scale="1,1,1") -> (Pose) {
    alphaInputType=Bool
    rotationMode=BMM_Additive
    rotationSpace=BCS_BoneSpace
    alphaBoolBlend {blendInTime=0.3, blendOutTime=0.2}
    boneToModify {boneName=dyn_front_skirt_l_MSTR}
    rotation {yaw=-5}
   }
   n12 = "Transform (Modify) Bone - Bone: dyn_front_skirt_r_MSTR"!(ComponentPose=@n11, bAlphaBoolEnabled:bool=@n10, Translation="0,0,0", Rotation="0,-5,0", Scale="1,1,1") -> (Pose) {
    alphaInputType=Bool
    rotationMode=BMM_Additive
    rotationSpace=BCS_BoneSpace
    alphaBoolBlend {blendInTime=0.3, blendOutTime=0.2}
    boneToModify {boneName=dyn_front_skirt_r_MSTR}
    rotation {yaw=-5}
   }
   n13 = "Transform (Modify) Bone - Bone: dyn_front_skirt_MSTR"!(ComponentPose=@n12, bAlphaBoolEnabled:bool=@n10, Translation="0,0,0", Rotation="0,-5,0", Scale="1,1,1") -> (Pose) {
    alphaInputType=Bool
    rotationMode=BMM_Additive
    rotationSpace=BCS_BoneSpace
    alphaBoolBlend {blendInTime=0.3, blendOutTime=0.2}
    boneToModify {boneName=dyn_front_skirt_MSTR}
    rotation {yaw=-5}
   }
   n14 = "Transform (Modify) Bone - Bone: dyn_side_skirt_r_06"!(ComponentPose=@n13, bAlphaBoolEnabled:bool=True, Translation="0,0,0", Rotation="0.000,11.065,-0.000", Scale="1,1,1") -> (Pose) {
    alphaInputType=Bool
    rotationMode=BMM_Additive
    rotationSpace=BCS_BoneSpace
    alphaBoolBlend {blendInTime=0.05, blendOutTime=0.2}
    boneToModify {boneName=dyn_side_skirt_r_06}
    rotation=<rot 0.0001 11.065 0>
   }
   n15 = Component To Local(ComponentPose=@n14)
   n16 = Get bIsSkyDiving()
   n17 = "Constraint - dyn_front_skirt_02 (PP)"!(
    ComponentPose=Local To Component(LocalPose=Blend Poses by bool(BlendPose_0=@n2, BlendPose_1=@n15, bActiveValue=@n16))
    Alpha=1
    ConstraintWeights_0=0.5
    ConstraintWeights_1:float=0.5
   ) -> (Pose) {
    alphaBoolBlend {blendOption=Linear}
    boneToModify {boneName=dyn_front_skirt_02}
    constraintSetup [{targetBone {boneName=dyn_front_skirt_l_02}}, {targetBone {boneName=dyn_front_skirt_r_02}}]
    constraintWeights [1,1]
   }
   n18 = "Constraint - dyn_front_skirt_03 (PP)"!(ComponentPose=@n17, Alpha=1, ConstraintWeights_0=0.5, ConstraintWeights_1:float=0.5) -> (Pose) {
    alphaBoolBlend {blendOption=Linear}
    boneToModify {boneName=dyn_front_skirt_03}
    constraintSetup [{targetBone {boneName=dyn_front_skirt_l_03}}, {targetBone {boneName=dyn_front_skirt_r_03}}]
    constraintWeights [1,1]
   }
   n19 = "Constraint - dyn_front_skirt_04 (PP)"!(ComponentPose=@n18, Alpha=1, ConstraintWeights_0=0.5, ConstraintWeights_1:float=0.5) -> (Pose) {
    alphaBoolBlend {blendOption=Linear}
    boneToModify {boneName=dyn_front_skirt_04}
    constraintSetup [{targetBone {boneName=dyn_front_skirt_l_04}}, {targetBone {boneName=dyn_front_skirt_r_04}}]
    constraintWeights [1,1]
   }
   n20 = "Constraint - dyn_front_skirt_05 (PP)"!(ComponentPose=@n19, Alpha=1, ConstraintWeights_0=0.5, ConstraintWeights_1:float=0.5) -> (Pose) {
    alphaBoolBlend {blendOption=Linear}
    boneToModify {boneName=dyn_front_skirt_05}
    constraintSetup [{targetBone {boneName=dyn_front_skirt_l_05}}, {targetBone {boneName=dyn_front_skirt_r_05}}]
    constraintWeights [1,1]
   }
   n21 = "Constraint - dyn_front_skirt_06 (PP)"!(ComponentPose=@n20, Alpha=1, ConstraintWeights_0=0.5, ConstraintWeights_1:float=0.5) -> (Pose) {
    alphaBoolBlend {blendOption=Linear}
    boneToModify {boneName=dyn_front_skirt_06}
    constraintSetup [{targetBone {boneName=dyn_front_skirt_l_06}}, {targetBone {boneName=dyn_front_skirt_r_06}}]
    constraintWeights [1,1]
   }
   n22 = "Constraint - dyn_rear_skirt_02 (PP)"!(ComponentPose=@n21, Alpha=1, ConstraintWeights_0=0.5, ConstraintWeights_1:float=0.5) -> (Pose) {
    alphaBoolBlend {blendOption=Linear}
    boneToModify {boneName=dyn_rear_skirt_02}
    constraintSetup [{targetBone {boneName=dyn_rear_skirt_l_02}}, {targetBone {boneName=dyn_rear_skirt_r_02}}]
    constraintWeights [1,1]
   }
   n23 = "Constraint - dyn_rear_skirt_03 (PP)"!(ComponentPose=@n22, Alpha=1, ConstraintWeights_0=0.5, ConstraintWeights_1:float=0.5) -> (Pose) {
    alphaBoolBlend {blendOption=Linear}
    boneToModify {boneName=dyn_rear_skirt_03}
    constraintSetup [{targetBone {boneName=dyn_rear_skirt_l_03}}, {targetBone {boneName=dyn_rear_skirt_r_03}}]
    constraintWeights [1,1]
   }
   n24 = "Constraint - dyn_rear_skirt_04 (PP)"!(ComponentPose=@n23, Alpha=1, ConstraintWeights_0=0.5, ConstraintWeights_1:float=0.5) -> (Pose) {
    alphaBoolBlend {blendOption=Linear}
    boneToModify {boneName=dyn_rear_skirt_04}
    constraintSetup [{targetBone {boneName=dyn_rear_skirt_l_04}}, {targetBone {boneName=dyn_rear_skirt_r_04}}]
    constraintWeights [1,1]
   }
   n25 = "Constraint - dyn_rear_skirt_05 (PP)"!(ComponentPose=@n24, Alpha=1, ConstraintWeights_0=0.5, ConstraintWeights_1:float=0.5) -> (Pose) {
    alphaBoolBlend {blendOption=Linear}
    boneToModify {boneName=dyn_rear_skirt_05}
    constraintSetup [{targetBone {boneName=dyn_rear_skirt_l_05}}, {targetBone {boneName=dyn_rear_skirt_r_05}}]
    constraintWeights [1,1]
   }
   n26 = "Constraint - dyn_rear_skirt_06 (PP)"!(ComponentPose=@n25, Alpha=1, ConstraintWeights_0=0.5, ConstraintWeights_1:float=0.5) -> (Pose) {
    alphaBoolBlend {blendOption=Linear}
    boneToModify {boneName=dyn_rear_skirt_06}
    constraintSetup [{targetBone {boneName=dyn_rear_skirt_l_06}}, {targetBone {boneName=dyn_rear_skirt_r_06}}]
    constraintWeights [1,1]
   }
   Output Pose(Result=Component To Local(ComponentPose=@n26)) {layerGroup=None}
   n27 = Control Rig!(
    Source=Copy Pose From Mesh() {bUseAttachedParent=true}
    Alpha
    isFEidle_CR:bool=Get IsInFrontEndPose() -> (IsInFrontEndPose:bool)
    Is Knee Sliding_CR:bool=Get bIsKneeSliding()
    Is Baseball Sliding_CR:bool=Get bIsBaseballSliding()
    IsPlayingEmoteCR:bool=Get bIsPlayingEmote()
    isCrouching_CR:bool=Get bIsCrouching()
    isSkydiving_CR:bool=Get bIsSkyDiving()
   ) -> (Pose) {
    alphaBoolBlend {blendOption=Linear}
    controlRigAssetReference {blueprintRigClass=&/F_MED_DistantEchoCastle_CtrlRig.F_MED_DistantEchoCastle_CtrlRig_C}
   }
   Save cached pose 'PostDeformationRig'(Pose=@n27) {cacheName=PostDeformationRig}
   n28 = Anim Dynamics - Bone: dyn_Hair_BackL_01(ComponentPose=Local To Component(LocalPose=Use cached pose 'preSimHood'()), Alpha=0.7, ExternalForce="0,0,0") {
    angularSpringConstant=800
    bAngularSpring=true
    bGravityOverrideInSimSpace=true
    bOverrideAngularDamping=true
    bUseGravityOverride=true
    gravityScale=0.3
    simulationSpace=BoneRelative
    alphaBoolBlend {blendOption=Linear}
    boundBone {boneName=dyn_Hair_BackL_01}
    componentAppliedLinearAccClamp=<1200 1200 1200>
    componentLinearAccScale=<0.2 0.2 0.25>
    componentLinearVelScale=<0.2 0.2 0.25>
    gravityOverride {z=-980}
    physicsBodyDefinitions [{
      sphereCollisionRadius=0
      boundBone {boneName=dyn_Hair_BackL_01}
      boxExtents=<50 50 50>
      constraintSetup {angularLimitsMax=<0.1 5 0>, angularLimitsMin=<-0.1 -5 -60>, angularTarget=<0 0 0>}
      localJointOffset=<-5 -0.2 0.1>
     }]
    relativeSpaceBone {boneName=root}
    simSpaceSettings {simSpaceAngularAlpha=0.2}
   }
   n29 = Anim Dynamics - Bone: dyn_Hair_BackR_01(ComponentPose=@n28, Alpha=0.7, ExternalForce="0,0,0") {
    angularSpringConstant=800
    bAngularSpring=true
    bGravityOverrideInSimSpace=true
    bOverrideAngularDamping=true
    bUseGravityOverride=true
    gravityScale=0.3
    simulationSpace=BoneRelative
    alphaBoolBlend {blendOption=Linear}
    boundBone {boneName=dyn_Hair_BackR_01}
    componentAppliedLinearAccClamp=<1200 1200 1200>
    componentLinearAccScale=<0.2 0.2 0.25>
    componentLinearVelScale=<0.2 0.2 0.25>
    gravityOverride {z=-980}
    physicsBodyDefinitions [{
      sphereCollisionRadius=0
      boundBone {boneName=dyn_Hair_BackR_01}
      boxExtents=<50 50 50>
      constraintSetup {angularLimitsMax=<0.1 5 0>, angularLimitsMin=<-0.1 -5 -60>, angularTarget=<0 0 0>}
      localJointOffset=<-5 -0.2 0.1>
     }]
    relativeSpaceBone {boneName=root}
    simSpaceSettings {simSpaceAngularAlpha=0.2}
   }
   n30 = Anim Dynamics - Bone: dyn_Hair_BackM_01(ComponentPose=@n29, Alpha=0.7, ExternalForce="0,0,0") {
    angularSpringConstant=800
    bAngularSpring=true
    bGravityOverrideInSimSpace=true
    bOverrideAngularDamping=true
    bUseGravityOverride=true
    gravityScale=0.3
    simulationSpace=BoneRelative
    alphaBoolBlend {blendOption=Linear}
    boundBone {boneName=dyn_Hair_BackM_01}
    componentAppliedLinearAccClamp=<1200 1200 1200>
    componentLinearAccScale=<0.2 0.2 0.25>
    componentLinearVelScale=<0.2 0.2 0.25>
    gravityOverride {z=-980}
    physicsBodyDefinitions [{
      sphereCollisionRadius=0
      boundBone {boneName=dyn_Hair_BackM_01}
      boxExtents=<50 50 50>
      constraintSetup {angularLimitsMax=<0.1 5 0>, angularLimitsMin=<-0.1 -5 -60>, angularTarget=<0 0 0>}
      localJointOffset=<-5 -0.2 0.1>
     }]
    relativeSpaceBone {boneName=root}
    simSpaceSettings {simSpaceAngularAlpha=0.2}
   }
   n31 = Anim Dynamics - Bone: dyn_Hair_BackM_02(ComponentPose=@n30, Alpha=0.7, ExternalForce="0,0,0") {
    angularSpringConstant=800
    bAngularSpring=true
    bGravityOverrideInSimSpace=true
    bOverrideAngularDamping=true
    bUseGravityOverride=true
    gravityScale=0.3
    simulationSpace=BoneRelative
    alphaBoolBlend {blendOption=Linear}
    boundBone {boneName=dyn_Hair_BackM_02}
    componentAppliedLinearAccClamp=<1200 1200 1200>
    componentLinearAccScale=<0.2 0.2 0.25>
    componentLinearVelScale=<0.2 0.2 0.25>
    gravityOverride {z=-980}
    physicsBodyDefinitions [{
      sphereCollisionRadius=0
      boundBone {boneName=dyn_Hair_BackM_02}
      boxExtents=<45 45 45>
      constraintSetup {angularLimitsMax=<0.1 5 0>, angularLimitsMin=<-0.1 -5 -60>, angularTarget=<0 0 0>}
      localJointOffset=<-5 -0.2 0.1>
     }]
    relativeSpaceBone {boneName=root}
    simSpaceSettings {simSpaceAngularAlpha=0.2}
   }
   n32 = Anim Dynamics - Bone: dyn_Hair_BackM_03(ComponentPose=@n31, Alpha=0.7, ExternalForce="0,0,0") {
    angularSpringConstant=800
    bAngularSpring=true
    bGravityOverrideInSimSpace=true
    bOverrideAngularDamping=true
    bUseGravityOverride=true
    gravityScale=0.3
    simulationSpace=BoneRelative
    alphaBoolBlend {blendOption=Linear}
    boundBone {boneName=dyn_Hair_BackM_03}
    componentAppliedLinearAccClamp=<1200 1200 1200>
    componentLinearAccScale=<0.2 0.2 0.25>
    componentLinearVelScale=<0.2 0.2 0.25>
    gravityOverride {z=-980}
    physicsBodyDefinitions [{
      sphereCollisionRadius=0
      boundBone {boneName=dyn_Hair_BackM_03}
      boxExtents=<30 30 30>
      constraintSetup {angularLimitsMax=<0.1 5 0>, angularLimitsMin=<-0.1 -5 -60>, angularTarget=<0 0 0>}
      localJointOffset=<-5 -0.2 0.1>
     }]
    relativeSpaceBone {boneName=root}
    simSpaceSettings {simSpaceAngularAlpha=0.2}
   }
   n33 = Anim Dynamics - Bone: dyn_Hair_BackM_04(ComponentPose=@n32, Alpha=0.7, ExternalForce="0,0,0") {
    angularSpringConstant=800
    bAngularSpring=true
    bGravityOverrideInSimSpace=true
    bOverrideAngularDamping=true
    bUseGravityOverride=true
    gravityScale=0.3
    simulationSpace=BoneRelative
    alphaBoolBlend {blendOption=Linear}
    boundBone {boneName=dyn_Hair_BackM_04}
    componentAppliedLinearAccClamp=<1200 1200 1200>
    componentLinearAccScale=<0.2 0.2 0.25>
    componentLinearVelScale=<0.2 0.2 0.25>
    gravityOverride {z=-980}
    physicsBodyDefinitions [{
      sphereCollisionRadius=0
      boundBone {boneName=dyn_Hair_BackM_04}
      boxExtents=<25 25 25>
      constraintSetup {angularLimitsMax=<0.1 5 0>, angularLimitsMin=<-0.1 -5 -60>, angularTarget=<0 0 0>}
      localJointOffset=<-5 -0.2 0.1>
     }]
    relativeSpaceBone {boneName=root}
    simSpaceSettings {simSpaceAngularAlpha=0.2}
   }
   n34 = Anim Dynamics - Bone: dyn_Hair_BackM_05(ComponentPose=@n33, Alpha=0.7, ExternalForce="0,0,0") {
    angularSpringConstant=800
    bAngularSpring=true
    bGravityOverrideInSimSpace=true
    bOverrideAngularDamping=true
    bUseGravityOverride=true
    gravityScale=0.3
    simulationSpace=BoneRelative
    alphaBoolBlend {blendOption=Linear}
    boundBone {boneName=dyn_Hair_BackM_05}
    componentAppliedLinearAccClamp=<1200 1200 1200>
    componentLinearAccScale=<0.2 0.2 0.25>
    componentLinearVelScale=<0.2 0.2 0.25>
    gravityOverride {z=-980}
    physicsBodyDefinitions [{
      sphereCollisionRadius=0
      boundBone {boneName=dyn_Hair_BackM_05}
      boxExtents=<15 15 15>
      constraintSetup {angularLimitsMax=<0.1 5 0>, angularLimitsMin=<-0.1 -5 -60>, angularTarget=<0 0 0>}
      localJointOffset=<-5 -0.2 0.1>
     }]
    relativeSpaceBone {boneName=root}
    simSpaceSettings {simSpaceAngularAlpha=0.2}
   }
   n35 = Anim Dynamics - Bone: dyn_Hair_BackM_06(ComponentPose=@n34, Alpha=0.7, ExternalForce="0,0,0") {
    angularSpringConstant=800
    bAngularSpring=true
    bGravityOverrideInSimSpace=true
    bOverrideAngularDamping=true
    bUseGravityOverride=true
    gravityScale=0.3
    simulationSpace=BoneRelative
    alphaBoolBlend {blendOption=Linear}
    boundBone {boneName=dyn_Hair_BackM_06}
    componentAppliedLinearAccClamp=<1200 1200 1200>
    componentLinearAccScale=<0.2 0.2 0.25>
    componentLinearVelScale=<0.2 0.2 0.25>
    gravityOverride {z=-980}
    physicsBodyDefinitions [{
      sphereCollisionRadius=0
      boundBone {boneName=dyn_Hair_BackM_06}
      boxExtents=<10 10 10>
      constraintSetup {angularLimitsMax=<0.1 5 0>, angularLimitsMin=<-0.1 -5 -60>, angularTarget=<0 0 0>}
      localJointOffset=<-5 -0.2 0.1>
     }]
    relativeSpaceBone {boneName=root}
    simSpaceSettings {simSpaceAngularAlpha=0.2}
   }
   n36 = "Transform (Modify) Bone - Bone: dyn_Hair_BackM_01"!(ComponentPose=@n35, bAlphaBoolEnabled:bool=True, Translation="0,3,0", Rotation="-0.003,-9.279,0.011", Scale="1,1,1") -> (Pose) {
    alphaInputType=Bool
    rotationMode=BMM_Additive
    rotationSpace=BCS_BoneSpace
    translationMode=BMM_Additive
    translationSpace=BCS_BoneSpace
    alphaBoolBlend {blendInTime=0.3, blendOutTime=0.2}
    boneToModify {boneName=dyn_Hair_BackM_01}
    rotation=<rot -0.003 -9.279 0.011>
    translation {y=3}
   }
   n37 = Component To Local(ComponentPose=@n36)
   n38 = Use cached pose 'preSimHood'()
   n39 = Get hoodDynamicsOnOff() -> (hoodDynamicsOnOff:bool)
   n40 = "Transform (Modify) Bone - Bone: dyn_Cloth_lowerarm_l_C01"(
    ComponentPose=Local To Component(LocalPose=Blend Poses by bool(BlendPose_0=@n37, BlendPose_1=@n38, bActiveValue=@n39) {blendType=Linear})
    Alpha=Get hand_shape_lt_l() -> (hand_shape_lt_l:double)
    Translation="5,5,0"
    Rotation="0,0,0"
    Scale="1,1,1"
   ) {
    rotationMode=BMM_Additive
    rotationSpace=BCS_BoneSpace
    scaleMode=BMM_Additive
    scaleSpace=BCS_BoneSpace
    translationMode=BMM_Additive
    translationSpace=BCS_BoneSpace
    alphaBoolBlend {blendOption=Linear}
    boneToModify {boneName=dyn_Cloth_lowerarm_l_C01}
    translation=<5 5>
   }
   n41 = Get hand_shape_lt_r() -> (hand_shape_lt_r:double)
   n42 = "Transform (Modify) Bone - Bone: dyn_Cloth_lowerarm_r_B01"(ComponentPose=@n40, Alpha=@n41, Translation="7,7,0", Rotation="0,-10,0", Scale="1,1,1") {
    rotationMode=BMM_Additive
    rotationSpace=BCS_BoneSpace
    scaleMode=BMM_Additive
    scaleSpace=BCS_BoneSpace
    translationMode=BMM_Additive
    translationSpace=BCS_BoneSpace
    alphaBoolBlend {blendOption=Linear}
    boneToModify {boneName=dyn_Cloth_lowerarm_r_B01}
    rotation {yaw=-10}
    translation=<7 7>
   }
   n43 = "Transform (Modify) Bone - Bone: dyn_Cloth_lowerarm_r_C01"(ComponentPose=@n42, Alpha=@n41, Translation="7,7,0", Rotation="-10,-10,10", Scale="1,1,1") {
    rotationMode=BMM_Additive
    rotationSpace=BCS_BoneSpace
    scaleMode=BMM_Additive
    scaleSpace=BCS_BoneSpace
    translationMode=BMM_Additive
    translationSpace=BCS_BoneSpace
    alphaBoolBlend {blendOption=Linear}
    boneToModify {boneName=dyn_Cloth_lowerarm_r_C01}
    rotation=<rot -10 -10 10>
    translation=<7 7>
   }
   n44 = Get offset_weapon() -> (offset_weapon:double)
   n45 = "Transform (Modify) Bone - Bone: dyn_Cloth_lowerarm_l"(ComponentPose=@n43, Alpha=@n44, Translation="3,3,0", Rotation="0,0,0", Scale="0.8,1,1") {
    rotationMode=BMM_Additive
    rotationSpace=BCS_BoneSpace
    scaleMode=BMM_Additive
    scaleSpace=BCS_BoneSpace
    translationMode=BMM_Additive
    translationSpace=BCS_BoneSpace
    alphaBoolBlend {blendOption=Linear}
    boneToModify {boneName=dyn_Cloth_lowerarm_l}
    scale {x=0.8}
    translation=<3 3>
   }
   n46 = "Transform (Modify) Bone - Bone: dyn_Cloth_lowerarm_r"(ComponentPose=@n45, Alpha=@n44, Translation="3,3,0", Rotation="0,0,0", Scale="0.8,1,1") {
    rotationMode=BMM_Additive
    rotationSpace=BCS_BoneSpace
    scaleMode=BMM_Additive
    scaleSpace=BCS_BoneSpace
    translationMode=BMM_Additive
    translationSpace=BCS_BoneSpace
    alphaBoolBlend {blendOption=Linear}
    boneToModify {boneName=dyn_Cloth_lowerarm_r}
    scale {x=0.8}
    translation=<3 3>
   }
   n47 = "Transform (Modify) Bone - Bone: dyn_Cloth_lowerarm_l_C01"(ComponentPose=@n46, Alpha=@n44, Translation="0,0,2", Rotation="0,5,0", Scale="0.8,1,1") {
    rotationMode=BMM_Additive
    rotationSpace=BCS_BoneSpace
    scaleMode=BMM_Additive
    scaleSpace=BCS_BoneSpace
    translationMode=BMM_Additive
    translationSpace=BCS_BoneSpace
    alphaBoolBlend {blendOption=Linear}
    boneToModify {boneName=dyn_Cloth_lowerarm_l_C01}
    rotation {yaw=5}
    scale {x=0.8}
    translation {z=2}
   }
   n48 = "Transform (Modify) Bone - Bone: dyn_Cloth_lowerarm_r_C01"(ComponentPose=@n47, Alpha=@n44, Translation="0,0,2", Rotation="0,10,0", Scale="0.8,1,1") {
    rotationMode=BMM_Additive
    rotationSpace=BCS_BoneSpace
    scaleMode=BMM_Additive
    scaleSpace=BCS_BoneSpace
    translationMode=BMM_Additive
    translationSpace=BCS_BoneSpace
    alphaBoolBlend {blendOption=Linear}
    boneToModify {boneName=dyn_Cloth_lowerarm_r_C01}
    rotation {yaw=10}
    scale {x=0.8}
    translation {z=2}
   }
   n49 = "Transform (Modify) Bone - Bone: dyn_Cloth_lowerarm_l"(ComponentPose=@n48, Alpha=1, Translation="0,0,0", Rotation="0,5,0", Scale="1,1,1") {
    rotationMode=BMM_Additive
    rotationSpace=BCS_BoneSpace
    scaleMode=BMM_Additive
    scaleSpace=BCS_BoneSpace
    translationMode=BMM_Additive
    translationSpace=BCS_BoneSpace
    alphaBoolBlend {blendOption=Linear}
    boneToModify {boneName=dyn_Cloth_lowerarm_l}
    rotation {yaw=5}
   }
   n50 = Anim Dynamics - Bone: dyn_Cloth_lowerarm_l(ComponentPose=@n49, Alpha=0.2, ExternalForce="0,0,0") {
    angularSpringConstant=800
    bAngularSpring=true
    bGravityOverrideInSimSpace=true
    bOverrideAngularDamping=true
    bUseGravityOverride=true
    gravityScale=0.3
    simulationSpace=BoneRelative
    alphaBoolBlend {blendOption=Linear}
    boundBone {boneName=dyn_Cloth_lowerarm_l}
    componentAppliedLinearAccClamp=<2500 2500 2500>
    componentLinearAccScale=<0.2 0.2 0.25>
    componentLinearVelScale=<0.2 0.2 0.25>
    gravityOverride {z=-980}
    physicsBodyDefinitions [{
      sphereCollisionRadius=0
      boundBone {boneName=dyn_Cloth_lowerarm_l}
      boxExtents=<50 50 50>
      constraintSetup {twistAxis=AxisY, angularLimitsMax=<5 10 10>, angularLimitsMin=<-5 -10 -10>, angularTarget=<0 0 0>}
      localJointOffset=<0 -10 0>
     }]
    relativeSpaceBone {boneName=root}
   }
   n51 = Anim Dynamics - Bone: dyn_Cloth_lowerarm_l_A(ComponentPose=@n50, Alpha=0.2, ExternalForce="0,0,0") {
    angularSpringConstant=800
    bAngularSpring=true
    bGravityOverrideInSimSpace=true
    bOverrideAngularDamping=true
    bUseGravityOverride=true
    gravityScale=0.3
    simulationSpace=BoneRelative
    alphaBoolBlend {blendOption=Linear}
    boundBone {boneName=dyn_Cloth_lowerarm_l_A}
    componentAppliedLinearAccClamp=<2500 2500 2500>
    componentLinearAccScale=<0.1 0.1 0.1>
    componentLinearVelScale=<0.1 0.1 0.1>
    gravityOverride {z=-980}
    physicsBodyDefinitions [{
      sphereCollisionRadius=0
      boundBone {boneName=dyn_Cloth_lowerarm_l_A}
      boxExtents=<30 30 30>
      constraintSetup {twistAxis=AxisY, angularLimitsMax=<5 10 10>, angularLimitsMin=<-5 -10 -10>, angularTarget=<0 0 0>}
      localJointOffset=<-10 -1 -1>
     }]
    relativeSpaceBone {boneName=root}
   }
   n52 = Anim Dynamics - Bone: dyn_Cloth_lowerarm_l_B01(ComponentPose=@n51, Alpha=0.2, ExternalForce="0,0,0") {
    angularSpringConstant=800
    bAngularSpring=true
    bGravityOverrideInSimSpace=true
    bOverrideAngularDamping=true
    bUseGravityOverride=true
    gravityScale=0.3
    simulationSpace=BoneRelative
    alphaBoolBlend {blendOption=Linear}
    boundBone {boneName=dyn_Cloth_lowerarm_l_B01}
    componentAppliedLinearAccClamp=<2500 2500 2500>
    componentLinearAccScale=<0.1 0.1 0.1>
    componentLinearVelScale=<0.1 0.1 0.1>
    gravityOverride {z=-980}
    physicsBodyDefinitions [{
      sphereCollisionRadius=0
      boundBone {boneName=dyn_Cloth_lowerarm_l_B01}
      boxExtents=<30 30 30>
      constraintSetup {twistAxis=AxisY, angularLimitsMax=<0 10 10>, angularLimitsMin=<-5 -10 -10>, angularTarget=<0 0 0>}
      localJointOffset=<-10 -1 -1>
     }]
    relativeSpaceBone {boneName=root}
   }
   n53 = Anim Dynamics - Bone: dyn_Cloth_lowerarm_l_B02(ComponentPose=@n52, Alpha=0.2, ExternalForce="0,0,0") {
    angularSpringConstant=800
    bAngularSpring=true
    bGravityOverrideInSimSpace=true
    bOverrideAngularDamping=true
    bUseGravityOverride=true
    gravityScale=0.3
    simulationSpace=BoneRelative
    alphaBoolBlend {blendOption=Linear}
    boundBone {boneName=dyn_Cloth_lowerarm_l_B02}
    componentAppliedLinearAccClamp=<2500 2500 2500>
    componentLinearAccScale=<0.1 0.1 0.1>
    componentLinearVelScale=<0.1 0.1 0.1>
    gravityOverride {z=-980}
    physicsBodyDefinitions [{
      sphereCollisionRadius=0
      boundBone {boneName=dyn_Cloth_lowerarm_l_B02}
      boxExtents=<30 30 30>
      constraintSetup {twistAxis=AxisY, angularLimitsMax=<0 10 10>, angularLimitsMin=<-5 -10 -10>, angularTarget=<0 0 0>}
      localJointOffset=<-10 -1 -1>
     }]
    relativeSpaceBone {boneName=root}
   }
   n54 = Anim Dynamics - Bone: dyn_Cloth_lowerarm_l_C01(ComponentPose=@n53, Alpha=0.2, ExternalForce="0,0,0") {
    angularSpringConstant=800
    bAngularSpring=true
    bGravityOverrideInSimSpace=true
    bOverrideAngularDamping=true
    bUseGravityOverride=true
    gravityScale=0.3
    simulationSpace=BoneRelative
    alphaBoolBlend {blendOption=Linear}
    boundBone {boneName=dyn_Cloth_lowerarm_l_C01}
    componentAppliedLinearAccClamp=<2500 2500 2500>
    componentLinearAccScale=<0.1 0.1 0.1>
    componentLinearVelScale=<0.1 0.1 0.1>
    gravityOverride {z=-980}
    physicsBodyDefinitions [{
      sphereCollisionRadius=0
      boundBone {boneName=dyn_Cloth_lowerarm_l_C01}
      boxExtents=<30 30 30>
      constraintSetup {twistAxis=AxisY, angularLimitsMax=<5 10 10>, angularLimitsMin=<-5 -10 -10>, angularTarget=<0 0 0>}
      localJointOffset=<-10 -1 -1>
     }]
    relativeSpaceBone {boneName=root}
   }
   n55 = Anim Dynamics - Bone: dyn_Cloth_lowerarm_l_C02(ComponentPose=@n54, Alpha=0.2, ExternalForce="0,0,0") {
    angularSpringConstant=800
    bAngularSpring=true
    bGravityOverrideInSimSpace=true
    bOverrideAngularDamping=true
    bUseGravityOverride=true
    gravityScale=0.3
    simulationSpace=BoneRelative
    alphaBoolBlend {blendOption=Linear}
    boundBone {boneName=dyn_Cloth_lowerarm_l_C02}
    componentAppliedLinearAccClamp=<2500 2500 2500>
    componentLinearAccScale=<0.1 0.1 0.1>
    componentLinearVelScale=<0.1 0.1 0.1>
    gravityOverride {z=-980}
    physicsBodyDefinitions [{
      sphereCollisionRadius=0
      boundBone {boneName=dyn_Cloth_lowerarm_l_C02}
      boxExtents=<30 30 30>
      constraintSetup {twistAxis=AxisY, angularLimitsMax=<5 10 10>, angularLimitsMin=<-5 -10 -10>, angularTarget=<0 0 0>}
      localJointOffset=<-10 -1 -1>
     }]
    relativeSpaceBone {boneName=root}
   }
   n56 = Anim Dynamics - Bone: dyn_Cloth_lowerarm_l_D01(ComponentPose=@n55, Alpha=0.2, ExternalForce="0,0,0") {
    angularSpringConstant=800
    bAngularSpring=true
    bGravityOverrideInSimSpace=true
    bOverrideAngularDamping=true
    bUseGravityOverride=true
    gravityScale=0.3
    simulationSpace=BoneRelative
    alphaBoolBlend {blendOption=Linear}
    boundBone {boneName=dyn_Cloth_lowerarm_l_D01}
    componentAppliedLinearAccClamp=<2500 2500 2500>
    componentLinearAccScale=<0.1 0.1 0.1>
    componentLinearVelScale=<0.1 0.1 0.1>
    gravityOverride {z=-980}
    physicsBodyDefinitions [{
      sphereCollisionRadius=0
      boundBone {boneName=dyn_Cloth_lowerarm_l_D01}
      boxExtents=<30 30 30>
      constraintSetup {twistAxis=AxisY, angularLimitsMax=<0 8 8>, angularLimitsMin=<-15 -8 -8>, angularTarget=<0 0 0>}
      localJointOffset=<-15 -3 -1>
     }]
    relativeSpaceBone {boneName=root}
   }
   n57 = Anim Dynamics - Bone: dyn_Cloth_lowerarm_l_D02(ComponentPose=@n56, Alpha=0.2, ExternalForce="0,0,0") {
    angularSpringConstant=800
    bAngularSpring=true
    bGravityOverrideInSimSpace=true
    bOverrideAngularDamping=true
    bUseGravityOverride=true
    gravityScale=0.3
    simulationSpace=BoneRelative
    alphaBoolBlend {blendOption=Linear}
    boundBone {boneName=dyn_Cloth_lowerarm_l_D02}
    componentAppliedLinearAccClamp=<2500 2500 2500>
    componentLinearAccScale=<0.1 0.1 0.1>
    componentLinearVelScale=<0.1 0.1 0.1>
    gravityOverride {z=-980}
    physicsBodyDefinitions [{
      sphereCollisionRadius=0
      boundBone {boneName=dyn_Cloth_lowerarm_l_D02}
      boxExtents=<30 30 30>
      constraintSetup {twistAxis=AxisY, angularLimitsMax=<0 8 8>, angularLimitsMin=<-15 -8 -8>, angularTarget=<0 0 0>}
      localJointOffset=<-15 -3 -1>
     }]
    relativeSpaceBone {boneName=root}
   }
   n58 = "Transform (Modify) Bone - Bone: dyn_Cloth_lowerarm_r"(ComponentPose=@n57, Alpha=1, Translation="0,0,0", Rotation="0,5,0", Scale="1,1,1") {
    rotationMode=BMM_Additive
    rotationSpace=BCS_BoneSpace
    scaleMode=BMM_Additive
    scaleSpace=BCS_BoneSpace
    translationMode=BMM_Additive
    translationSpace=BCS_BoneSpace
    alphaBoolBlend {blendOption=Linear}
    boneToModify {boneName=dyn_Cloth_lowerarm_r}
    rotation {yaw=5}
   }
   n59 = Anim Dynamics - Bone: dyn_Cloth_lowerarm_r(ComponentPose=@n58, Alpha=0.2, ExternalForce="0,0,0") {
    angularSpringConstant=800
    bAngularSpring=true
    bGravityOverrideInSimSpace=true
    bOverrideAngularDamping=true
    bUseGravityOverride=true
    gravityScale=0.3
    simulationSpace=BoneRelative
    alphaBoolBlend {blendOption=Linear}
    boundBone {boneName=dyn_Cloth_lowerarm_r}
    componentAppliedLinearAccClamp=<2500 2500 2500>
    componentLinearAccScale=<0.2 0.2 0.25>
    componentLinearVelScale=<0.2 0.2 0.25>
    gravityOverride {z=-980}
    physicsBodyDefinitions [{
      sphereCollisionRadius=0
      boundBone {boneName=dyn_Cloth_lowerarm_r}
      boxExtents=<50 50 50>
      constraintSetup {twistAxis=AxisY, angularLimitsMax=<5 10 10>, angularLimitsMin=<-5 -10 -10>, angularTarget=<0 0 0>}
      localJointOffset=<0 -10 0>
     }]
    relativeSpaceBone {boneName=root}
   }
   n60 = Anim Dynamics - Bone: dyn_Cloth_lowerarm_r_A(ComponentPose=@n59, Alpha=0.2, ExternalForce="0,0,0") {
    angularSpringConstant=800
    bAngularSpring=true
    bGravityOverrideInSimSpace=true
    bOverrideAngularDamping=true
    bUseGravityOverride=true
    gravityScale=0.3
    simulationSpace=BoneRelative
    alphaBoolBlend {blendOption=Linear}
    boundBone {boneName=dyn_Cloth_lowerarm_r_A}
    componentAppliedLinearAccClamp=<2500 2500 2500>
    componentLinearAccScale=<0.1 0.1 0.1>
    componentLinearVelScale=<0.1 0.1 0.1>
    gravityOverride {z=-980}
    physicsBodyDefinitions [{
      sphereCollisionRadius=0
      boundBone {boneName=dyn_Cloth_lowerarm_r_A}
      boxExtents=<30 30 30>
      constraintSetup {twistAxis=AxisY, angularLimitsMax=<5 10 10>, angularLimitsMin=<-5 -10 -10>, angularTarget=<0 0 0>}
      localJointOffset=<-10 -1 -1>
     }]
    relativeSpaceBone {boneName=root}
   }
   n61 = Anim Dynamics - Bone: dyn_Cloth_lowerarm_r_B01(ComponentPose=@n60, Alpha=0.2, ExternalForce="0,0,0") {
    angularSpringConstant=800
    bAngularSpring=true
    bGravityOverrideInSimSpace=true
    bOverrideAngularDamping=true
    bUseGravityOverride=true
    gravityScale=0.3
    simulationSpace=BoneRelative
    alphaBoolBlend {blendOption=Linear}
    boundBone {boneName=dyn_Cloth_lowerarm_r_B01}
    componentAppliedLinearAccClamp=<2500 2500 2500>
    componentLinearAccScale=<0.1 0.1 0.1>
    componentLinearVelScale=<0.1 0.1 0.1>
    gravityOverride {z=-980}
    physicsBodyDefinitions [{
      sphereCollisionRadius=0
      boundBone {boneName=dyn_Cloth_lowerarm_r_B01}
      boxExtents=<30 30 30>
      constraintSetup {twistAxis=AxisY, angularLimitsMax=<0 10 10>, angularLimitsMin=<-5 -10 -10>, angularTarget=<0 0 0>}
      localJointOffset=<-10 -1 -1>
     }]
    relativeSpaceBone {boneName=root}
   }
   n62 = Anim Dynamics - Bone: dyn_Cloth_lowerarm_r_B02(ComponentPose=@n61, Alpha=0.2, ExternalForce="0,0,0") {
    angularSpringConstant=800
    bAngularSpring=true
    bGravityOverrideInSimSpace=true
    bOverrideAngularDamping=true
    bUseGravityOverride=true
    gravityScale=0.3
    simulationSpace=BoneRelative
    alphaBoolBlend {blendOption=Linear}
    boundBone {boneName=dyn_Cloth_lowerarm_r_B02}
    componentAppliedLinearAccClamp=<2500 2500 2500>
    componentLinearAccScale=<0.1 0.1 0.1>
    componentLinearVelScale=<0.1 0.1 0.1>
    gravityOverride {z=-980}
    physicsBodyDefinitions [{
      sphereCollisionRadius=0
      boundBone {boneName=dyn_Cloth_lowerarm_r_B02}
      boxExtents=<30 30 30>
      constraintSetup {twistAxis=AxisY, angularLimitsMax=<0 10 10>, angularLimitsMin=<-5 -10 -10>, angularTarget=<0 0 0>}
      localJointOffset=<-10 -1 -1>
     }]
    relativeSpaceBone {boneName=root}
   }
   n63 = Anim Dynamics - Bone: dyn_Cloth_lowerarm_r_C01(ComponentPose=@n62, Alpha=0.2, ExternalForce="0,0,0") {
    angularSpringConstant=800
    bAngularSpring=true
    bGravityOverrideInSimSpace=true
    bOverrideAngularDamping=true
    bUseGravityOverride=true
    gravityScale=0.3
    simulationSpace=BoneRelative
    alphaBoolBlend {blendOption=Linear}
    boundBone {boneName=dyn_Cloth_lowerarm_r_C01}
    componentAppliedLinearAccClamp=<2500 2500 2500>
    componentLinearAccScale=<0.1 0.1 0.1>
    componentLinearVelScale=<0.1 0.1 0.1>
    gravityOverride {z=-980}
    physicsBodyDefinitions [{
      sphereCollisionRadius=0
      boundBone {boneName=dyn_Cloth_lowerarm_r_C01}
      boxExtents=<30 30 30>
      constraintSetup {twistAxis=AxisY, angularLimitsMax=<5 10 10>, angularLimitsMin=<-5 -10 -10>, angularTarget=<0 0 0>}
      localJointOffset=<-10 -1 -1>
     }]
    relativeSpaceBone {boneName=root}
   }
   n64 = Anim Dynamics - Bone: dyn_Cloth_lowerarm_r_C02(ComponentPose=@n63, Alpha=0.2, ExternalForce="0,0,0") {
    angularSpringConstant=800
    bAngularSpring=true
    bGravityOverrideInSimSpace=true
    bOverrideAngularDamping=true
    bUseGravityOverride=true
    gravityScale=0.3
    simulationSpace=BoneRelative
    alphaBoolBlend {blendOption=Linear}
    boundBone {boneName=dyn_Cloth_lowerarm_r_C02}
    componentAppliedLinearAccClamp=<2500 2500 2500>
    componentLinearAccScale=<0.1 0.1 0.1>
    componentLinearVelScale=<0.1 0.1 0.1>
    gravityOverride {z=-980}
    physicsBodyDefinitions [{
      sphereCollisionRadius=0
      boundBone {boneName=dyn_Cloth_lowerarm_r_C02}
      boxExtents=<30 30 30>
      constraintSetup {twistAxis=AxisY, angularLimitsMax=<5 10 10>, angularLimitsMin=<-5 -10 -10>, angularTarget=<0 0 0>}
      localJointOffset=<-10 -1 -1>
     }]
    relativeSpaceBone {boneName=root}
   }
   n65 = Anim Dynamics - Bone: dyn_Cloth_lowerarm_r_D01(ComponentPose=@n64, Alpha=0.2, ExternalForce="0,0,0") {
    angularSpringConstant=800
    bAngularSpring=true
    bGravityOverrideInSimSpace=true
    bOverrideAngularDamping=true
    bUseGravityOverride=true
    gravityScale=0.3
    simulationSpace=BoneRelative
    alphaBoolBlend {blendOption=Linear}
    boundBone {boneName=dyn_Cloth_lowerarm_r_D01}
    componentAppliedLinearAccClamp=<2500 2500 2500>
    componentLinearAccScale=<0.1 0.1 0.1>
    componentLinearVelScale=<0.1 0.1 0.1>
    gravityOverride {z=-980}
    physicsBodyDefinitions [{
      sphereCollisionRadius=0
      boundBone {boneName=dyn_Cloth_lowerarm_r_D01}
      boxExtents=<30 30 30>
      constraintSetup {twistAxis=AxisY, angularLimitsMax=<0 8 8>, angularLimitsMin=<-15 -8 -8>, angularTarget=<0 0 0>}
      localJointOffset=<-15 -3 -1>
     }]
    relativeSpaceBone {boneName=root}
   }
   n66 = Anim Dynamics - Bone: dyn_Cloth_lowerarm_r_D02(ComponentPose=@n65, Alpha=0.2, ExternalForce="0,0,0") {
    angularSpringConstant=800
    bAngularSpring=true
    bGravityOverrideInSimSpace=true
    bOverrideAngularDamping=true
    bUseGravityOverride=true
    gravityScale=0.3
    simulationSpace=BoneRelative
    alphaBoolBlend {blendOption=Linear}
    boundBone {boneName=dyn_Cloth_lowerarm_r_D02}
    componentAppliedLinearAccClamp=<2500 2500 2500>
    componentLinearAccScale=<0.1 0.1 0.1>
    componentLinearVelScale=<0.1 0.1 0.1>
    gravityOverride {z=-980}
    physicsBodyDefinitions [{
      sphereCollisionRadius=0
      boundBone {boneName=dyn_Cloth_lowerarm_r_D02}
      boxExtents=<30 30 30>
      constraintSetup {twistAxis=AxisY, angularLimitsMax=<0 8 8>, angularLimitsMin=<-15 -8 -8>, angularTarget=<0 0 0>}
      localJointOffset=<-15 -3 -1>
     }]
    relativeSpaceBone {boneName=root}
   }
   Save cached pose 'sleevesDyn'(Pose=Component To Local(ComponentPose=@n66)) {cacheName=sleevesDyn}
   n67 = "Transform (Modify) Bone - Bone: dyn_skirt_ground_collision"!(
    ComponentPose=Local To Component(LocalPose=Use cached pose 'sleevesDyn'())
    bAlphaBoolEnabled:bool=True
    Translation="0,0,0"
    Rotation="0,0,0"
    Scale="1,1,1"
   ) -> (Pose) {
    alphaInputType=Bool
    rotationMode=BMM_Replace
    rotationSpace=BCS_WorldSpace
    translationMode=BMM_Replace
    translationSpace=BCS_BoneSpace
    alphaBoolBlend {blendOption=Linear}
    boneToModify {boneName=dyn_skirt_ground_collision}
   }
   Save cached pose 'postSlopePosing'(Pose=Component To Local(ComponentPose=@n67)) {cacheName=postSlopePosing}
   n68 = "Constraint - dyn_Hair_BackL_01 (PPPPP)"!(
    ComponentPose=Local To Component(LocalPose=Use cached pose 'PostDeformationRig'())
    Alpha=1
    ConstraintWeights_0=0.4883
    ConstraintWeights_1:float=0.3875
    ConstraintWeights_2:float=0.1016
    ConstraintWeights_3:float=0.0225
    ConstraintWeights_4:float=0.0001
   ) -> (Pose) {
    alphaBoolBlend {blendOption=Linear}
    boneToModify {boneName=dyn_Hair_BackL_01}
    constraintSetup [
     {targetBone {boneName=spine_05}}
     {targetBone {boneName=clavicle_l}}
     {targetBone {boneName=neck_01}}
     {targetBone {boneName=neck_02}}
     {targetBone {boneName=upperarm_twist_01_l}}
    ]
    constraintWeights [1,1,1,1,1]
   }
   n69 = "Constraint - dyn_Hair_BackR_01 (PPPPP)"!(
    ComponentPose=@n68
    Alpha=1
    ConstraintWeights_0=0.5259
    ConstraintWeights_1:float=0.4651
    ConstraintWeights_2:float=0.0073
    ConstraintWeights_3:float=0.0017
    ConstraintWeights_4:float=0
   ) -> (Pose) {
    alphaBoolBlend {blendOption=Linear}
    boneToModify {boneName=dyn_Hair_BackR_01}
    constraintSetup [
     {targetBone {boneName=spine_05}}
     {targetBone {boneName=clavicle_r}}
     {targetBone {boneName=neck_01}}
     {targetBone {boneName=neck_02}}
     {targetBone {boneName=upperarm_twist_01_r}}
    ]
    constraintWeights [1,1,1,1,1]
   }
   n70 = "Constraint - dyn_Hair_BackM_01 (PPPPP)"!(
    ComponentPose=@n69
    Alpha=1
    ConstraintWeights_0=0.6674
    ConstraintWeights_1:float=0.2513
    ConstraintWeights_2:float=0.0804
    ConstraintWeights_3:float=0.0006
    ConstraintWeights_4:float=0.0003
   ) -> (Pose) {
    alphaBoolBlend {blendOption=Linear}
    boneToModify {boneName=dyn_Hair_BackM_01}
    constraintSetup [
     {targetBone {boneName=spine_05}}
     {targetBone {boneName=neck_01}}
     {targetBone {boneName=neck_02}}
     {targetBone {boneName=clavicle_r}}
     {targetBone {boneName=clavicle_l}}
    ]
    constraintWeights [1,1,1,1,1]
   }
   n71 = "Constraint - dyn_Hair_BackM_02 (P)"!(ComponentPose=@n70, Alpha=0.7, ConstraintWeights_0=1, ConstraintWeights_1:float=0) -> (Pose) {
    alphaBoolBlend {blendOption=Linear}
    boneToModify {boneName=dyn_Hair_BackM_02}
    constraintSetup [{targetBone {boneName=spine_05}}]
    constraintWeights [1,0]
   }
   n72 = "Constraint - dyn_Hair_BackM_03 (PP)"!(ComponentPose=@n71, Alpha=0.5, ConstraintWeights_0=0.9282, ConstraintWeights_1:float=0.0718) -> (Pose) {
    alphaBoolBlend {blendOption=Linear}
    boneToModify {boneName=dyn_Hair_BackM_03}
    constraintSetup [{targetBone {boneName=spine_05}}, {targetBone {boneName=spine_04}}]
    constraintWeights [1,1]
   }
   n73 = "Constraint - dyn_Hair_BackM_04 (PPPPPP)"!(
    ComponentPose=@n72
    Alpha=0.5
    ConstraintWeights_0=0.2853
    ConstraintWeights_1:float=0.2238
    ConstraintWeights_2:float=0.2136
    ConstraintWeights_3:float=0.2065
    ConstraintWeights_4:float=0.0692
    ConstraintWeights_5:float=0.0016
   ) -> (Pose) {
    alphaBoolBlend {blendOption=Linear}
    boneToModify {boneName=dyn_Hair_BackM_04}
    constraintSetup [
     {targetBone {boneName=spine_04}}
     {targetBone {boneName=spine_02}}
     {targetBone {boneName=spine_03}}
     {targetBone {boneName=spine_05}}
     {targetBone {boneName=spine_01}}
     {targetBone {boneName=pelvis}}
    ]
    constraintWeights [1,1,1,1,1,1]
   }
   n74 = "Constraint - dyn_Hair_BackM_05 (PPPP)"!(
    ComponentPose=@n73
    Alpha=0.3
    ConstraintWeights_0=0.4545
    ConstraintWeights_1:float=0.3037
    ConstraintWeights_2:float=0.2321
    ConstraintWeights_3:float=0.0096
   ) -> (Pose) {
    alphaBoolBlend {blendOption=Linear}
    boneToModify {boneName=dyn_Hair_BackM_05}
    constraintSetup [{targetBone {boneName=pelvis}}, {targetBone {boneName=spine_01}}, {targetBone {boneName=spine_02}}, {targetBone {boneName=spine_03}}]
    constraintWeights [1,1,1,1]
   }
   n75 = "Constraint - dyn_Hair_BackM_06 (PPP)"!(ComponentPose=@n74, Alpha=0.3, ConstraintWeights_0=0.7771, ConstraintWeights_1:float=0.2147, ConstraintWeights_2:float=0.0082) -> (Pose) {
    alphaBoolBlend {blendOption=Linear}
    boneToModify {boneName=dyn_Hair_BackM_06}
    constraintSetup [{targetBone {boneName=pelvis}}, {targetBone {boneName=spine_01}}, {targetBone {boneName=spine_02}}]
    constraintWeights [1,1,1]
   }
   n76 = "Transform (Modify) Bone - Bone: dyn_Hair_BackM_05"!(ComponentPose=@n75, bAlphaBoolEnabled:bool=Get bIsSkyDiving(), Translation="0,0,0", Rotation="0,-20,0", Scale="1,1,1") -> (Pose) {
    alphaInputType=Bool
    rotationMode=BMM_Additive
    rotationSpace=BCS_BoneSpace
    scaleMode=BMM_Additive
    scaleSpace=BCS_BoneSpace
    translationMode=BMM_Additive
    translationSpace=BCS_BoneSpace
    alphaBoolBlend {blendOption=Linear}
    boneToModify {boneName=dyn_Hair_BackM_05}
    rotation {yaw=-20}
   }
   Save cached pose 'preSimHood'(Pose=Component To Local(ComponentPose=@n76)) {cacheName=preSimHood}
  ]}
 EventGraph {nodes [
   Event OnCharacterCustomizationLoaded()
   > n1 = Sequence!(execute) -> (then_0, then_1, then_2:Exec)
   n2 = CheckForBackbling()
   > Set isWearingPackBool(isWearingPackBool:bool=OR Boolean!(A=@n2, B=@n2.bIsBackblingACape, C:bool=@n2.bIsBackblingLong) -> (ReturnValue)) -> (Output_Get:bool)
   Set IsClawpadGlider(
    execute=@n1.then_1
    IsClawpadGlider:bool=GliderItemDefinitionHasAnyMetaTag!(Tags:Array of Gameplay Tag Structures=Get Tags() -> (Tags:Array of Gameplay Tag Structures), bRequiresOpened:bool=false) -> (ReturnValue:bool)
   ) -> (Output_Get:bool)
   n3 = Event BlueprintUpdateAnimation()
   > n4 = Sequence!(execute) -> (
    then_0
    then_1
    then_2
    then_3:Exec
    then_4:Exec
    then_5:Exec
    then_6:Exec
    then_7:Exec
    then_8:Exec
    then_9:Exec
    then_10:Exec
    then_11:Exec
   )
   n5 = Branch(execute=@n4.then_1, Condition=Get bIsSkydiveDiveMode())
   > n6 = Set Diving Timer(Diving Timer:double=float + float(A=Get Diving Timer() -> (Diving Timer:double), B=1) {is=title K2Node_CommutativeAssociativeBinaryOperator}) -> (Output_Get:double)
   > Set IsSkyDiving(IsSkyDiving:double=MapRangeClamped(Value=@n6, InRangeA=0, InRangeB=30, OutRangeA=0.3, OutRangeB=0)) -> (Output_Get:double)
   > Set IsUse_RigidBoy(IsUse_RigidBoy:double=0) -> (Output_Get:double)
   n7 = Branch(execute=@n5.else, Condition=Get bIsParachuteOpened())
   > n8 = Set Diving Timer(Diving Timer=float + float(A=Get Diving Timer(), B=1))
   > Set IsSkyDiving(IsSkyDiving=MapRangeClamped(Value=@n8, InRangeA=0, InRangeB=250, OutRangeA=0.1, OutRangeB=0.7))
   Set IsSkyDiving(execute=@n7.else, IsSkyDiving=0.1)
   > Set IsUse_RigidBoy(IsUse_RigidBoy=0.85)
   > Set Diving Timer(Diving Timer=0)
   Handle Slope Warp!(execute=@n4.then_5, self, DeltaTime:double=@n3.DeltaTimeX) -> (then)
   n9 = "Gravity Override State Function (Parameter Asset) (Nativized)"(
    execute=@n4.then_0
    CharacterPartAnimInstance=Self-Reference()
    GameTimeInSeconds=GetGameTimeInSeconds()
    Parameters=&/F_MED_DistantEchoCastle_Parameters
   )
   > Set Gravity Override(Gravity Override=@n9)
   > Set RBAN Settings Out(RBAN Settings Out=@n9.OutRigidBodyAnimNodeParameters)
   n10 = Get isWearingPackBool() -> (isWearingPackBool:bool)
   Set hoodDynamicsOnOff(execute=@n4.then_3, hoodDynamicsOnOff:bool=AND Boolean(A=NOT Boolean(A=@n10), B=Get bAnimDynamicsEnabled() -> (bAnimDynamicsEnabled:bool))) -> (Output_Get:bool)
   n11 = Get bIsMoving2D()
   Set isCrouchingNotMoving(execute=@n4.then_4, isCrouchingNotMoving:bool=AND Boolean(A=Get bIsCrouching(), B=NOT Boolean(A=@n11))) -> (Output_Get:bool)
   n12 = Get bIsCrouching()
   > Set isMovingNotCrouching(isMovingNotCrouching:bool=AND Boolean(A=Get bIsMoving2D(), B=NOT Boolean(A=@n12))) -> (Output_Get:bool)
   Branch(execute=@n4.then_2, Condition=Get bIsInFrontEnd())
   > n13 = Branch(Condition=Get bIsPlayingEmote())
   > Set IsInFrontEndPose(IsInFrontEndPose:bool=false) -> (Output_Get:bool)
   Set IsInFrontEndPose(execute=@n13.else, IsInFrontEndPose=true)
  ]}
 Handle Slope Warp {nodes [
   n1 = Handle Slope Warp!() -> (then, DeltaTime:double=~) {is=title K2Node_FunctionEntry}
   > n2 = Sequence!(execute) -> (then_0, then_1)
   n3 = Get bIsOnGround() -> (bIsOnGround:bool)
   > n4 = Branch(Condition=OR Boolean(A=NOT Boolean(A=@n3), B=Get bIsPlayingEmote(), C=Get bIsInVehicle() -> (bIsInVehicle:bool)))
   > Set Target Info Input Root Slope Translation(Target Info Input Root Slope Translation:Vector="0,0,-100") -> (Output_Get:Vector)
   Set Slope Info Input Slope Rotation(
    execute=@n4.else
    Slope Info Input Slope Rotation:Rotator=RInterpTo!(
     Current:Rotator=Get Slope Info Input Slope Rotation() -> (Slope Info Input Slope Rotation:Rotator)
     Target:Rotator=Get!() -> (
      SlopeInfoInput_RootSlopeTranslation:Vector=~
      SlopeInfoInput_LeftFootSlopeTranslation:Vector=~
      SlopeInfoInput_RightFootSlopeTranslation:Vector=~
      SlopeInfoInput_SlopeRotation:Rotator=~
     ).SlopeInfoInput_SlopeRotation
     DeltaTime:float=@n1
     InterpSpeed:float=7
    ) -> (ReturnValue:Rotator)
   ) -> (Output_Get:Rotator)
   > Set Target Info Input Root Slope Translation(Target Info Input Root Slope Translation="0,0,5")
   Set Slope Info Input Root Slope Translation(
    execute=@n2.then_1
    Slope Info Input Root Slope Translation:Vector=VInterpTo!(
     Current:Vector=Get Slope Info Input Root Slope Translation() -> (Slope Info Input Root Slope Translation:Vector)
     Target:Vector=Get Target Info Input Root Slope Translation() -> (Target Info Input Root Slope Translation:Vector)
     DeltaTime:float=@n1
     InterpSpeed:float=10
    ) -> (ReturnValue:Vector)
   ) -> (Output_Get:Vector)
  ]}
]
variables [
 Diving Timer {type=number}
 Gravity Override {type=Vector}
 hand_shape_lt_l {type=number}
 hand_shape_lt_r {type=number}
 hoodDynamicsOnOff {type=boolean}
 IsClawpadGlider {type=boolean}
 isCrouchingNotMoving {type=boolean}
 IsInFrontEndPose {type=boolean}
 isMovingNotCrouching {type=boolean}
 IsSkyDiving {type=number}
 IsUse_RigidBoy {type=number}
 isWearingPackBool {type=boolean}
 NewVar {type=boolean}
 offset_weapon {type=number}
 RBAN Settings Out {type=FortRigidBodyAnimNodeParameters}
 Slope Info Input Root Slope Translation {type=Vector}
 Slope Info Input Slope Rotation {type=Rotator}
 Tags {type="array<GameplayTag>", default [{tagName=Cosmetics.MetaTags.ClawPadHost}]}
 Target Info Input Root Slope Translation {type=Vector}
]
