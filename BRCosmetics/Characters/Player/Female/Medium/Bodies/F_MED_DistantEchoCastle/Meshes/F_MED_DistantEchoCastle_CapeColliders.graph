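# Physics asset of cape colliders for the DistantEchoCastle medium female player body, apparently for cape simulation to collide against.
# Capsules cover head, clavicles, spine, pelvis and feet, with extra crosswise capsules on the middle spine.
# Center, left and right rear-skirt chains each carry six capsules that grow toward the end, all reusing one copied shape name.
# Arm, thigh, calf and neck bodies are empty, and no constraints exist.
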PhysicsAsset /BRCosmetics/Characters/Player/Female/Medium/Bodies/F_MED_DistantEchoCastle/Meshes/F_MED_DistantEchoCastle_CapeColliders
bodies {
 calf_l []
 calf_r []
 clavicle_l [{shapeType=Capsule, length=7.6964, radius=7.6353, center=<4.0147 -0.8354 -1.7419>, rotation=<rot -77.5566 177.3915 7.7314>}]
 clavicle_r [{shapeType=Capsule, length=7.6964, radius=8.5457, center=<-4.0147 0.8354 1.7419>, rotation=<rot -77.5565 177.3915 -172.2685>}]
 dyn_rear_skirt_01 [{
   shapeType=Capsule
   length=8.6718
   radius=11.9493
   shapeName=dyn_rear_skirt_r_01_capsule
   center=<15.9466 3.8958 -1.1399>
   rotation=<rot 0 83.7525 -2.6422>
  }]
 dyn_rear_skirt_02 [{
   shapeType=Capsule
   length=8.6718
   radius=11.9493
   shapeName=dyn_rear_skirt_r_01_capsule
   center=<9.2615 4.6276 -0.8295>
   rotation=<rot 0 83.7525 -2.6422>
  }]
 dyn_rear_skirt_03 [{
   shapeType=Capsule
   length=8.6718
   radius=11.9493
   shapeName=dyn_rear_skirt_r_01_capsule
   center=<11.0138 4.4358 -0.9109>
   rotation=<rot 0 83.7525 -2.6422>
  }]
 dyn_rear_skirt_04 [{
   shapeType=Capsule
   length=11.8225
   radius=11.9493
   shapeName=dyn_rear_skirt_r_01_capsule
   center=<11.0138 4.4358 -0.9109>
   rotation=<rot 0 83.7525 -2.6422>
  }]
 dyn_rear_skirt_05 [{
   shapeType=Capsule
   length=17.1754
   radius=11.9493
   shapeName=dyn_rear_skirt_r_01_capsule
   center=<11.0138 4.4358 -0.9109>
   rotation=<rot 0 83.7525 -2.6422>
  }]
 dyn_rear_skirt_06 [{
   shapeType=Capsule
   length=22.7333
   radius=14.2272
   shapeName=dyn_rear_skirt_r_01_capsule
   center=<11.0138 4.4358 -0.9109>
   rotation=<rot 0 83.7525 -2.6422>
  }]
 dyn_rear_skirt_l_01 [{
   shapeType=Capsule
   length=8.6718
   radius=11.9493
   shapeName=dyn_rear_skirt_r_01_capsule
   center=<15.9466 3.8958 -1.1399>
   rotation=<rot 0 83.7525 -2.6422>
  }]
 dyn_rear_skirt_l_02 [{
   shapeType=Capsule
   length=8.6718
   radius=11.9493
   shapeName=dyn_rear_skirt_r_01_capsule
   center=<9.2615 4.6276 -0.8295>
   rotation=<rot 0 83.7525 -2.6422>
  }]
 dyn_rear_skirt_l_03 [{
   shapeType=Capsule
   length=8.6718
   radius=11.9493
   shapeName=dyn_rear_skirt_r_01_capsule
   center=<11.0138 4.4358 -0.9109>
   rotation=<rot 0 83.7525 -2.6422>
  }]
 dyn_rear_skirt_l_04 [{
   shapeType=Capsule
   length=11.8225
   radius=11.9493
   shapeName=dyn_rear_skirt_r_01_capsule
   center=<11.0138 4.4358 -0.9109>
   rotation=<rot 0 83.7525 -2.6422>
  }]
 dyn_rear_skirt_l_05 [{
   shapeType=Capsule
   length=17.1754
   radius=11.9493
   shapeName=dyn_rear_skirt_r_01_capsule
   center=<11.0138 4.4358 -0.9109>
   rotation=<rot 0 83.7525 -2.6422>
  }]
 dyn_rear_skirt_l_06 [{
   shapeType=Capsule
   length=22.7333
   radius=14.2272
   shapeName=dyn_rear_skirt_r_01_capsule
   center=<11.0138 4.4358 -0.9109>
   rotation=<rot 0 83.7525 -2.6422>
  }]
 dyn_rear_skirt_r_01 [{
   shapeType=Capsule
   length=8.6718
   radius=11.9493
   shapeName=dyn_rear_skirt_r_01_capsule
   center=<-16.2809 -7.8758 0.2794>
   rotation=<rot 0 83.7525 -2.6422>
  }]
 dyn_rear_skirt_r_02 [{
   shapeType=Capsule
   length=8.6718
   radius=11.9493
   shapeName=dyn_rear_skirt_r_01_capsule
   center=<-11.1457 -8.438 0.041>
   rotation=<rot 0 83.7525 -2.6422>
  }]
 dyn_rear_skirt_r_03 [{
   shapeType=Capsule
   length=8.6718
   radius=11.9493
   shapeName=dyn_rear_skirt_r_01_capsule
   center=<-12.1749 -8.3253 0.0888>
   rotation=<rot 0 83.7525 -2.6422>
  }]
 dyn_rear_skirt_r_04 [{
   shapeType=Capsule
   length=11.8225
   radius=11.9493
   shapeName=dyn_rear_skirt_r_01_capsule
   center=<-12.1749 -8.3253 0.0888>
   rotation=<rot 0 83.7525 -2.6422>
  }]
 dyn_rear_skirt_r_05 [{
   shapeType=Capsule
   length=17.1754
   radius=11.9493
   shapeName=dyn_rear_skirt_r_01_capsule
   center=<-12.1749 -8.3253 0.0888>
   rotation=<rot 0 83.7525 -2.6422>
  }]
 dyn_rear_skirt_r_06 [{
   shapeType=Capsule
   length=22.7333
   radius=14.2272
   shapeName=dyn_rear_skirt_r_01_capsule
   center=<-12.1749 -8.3253 0.0888>
   rotation=<rot 0 83.7525 -2.6422>
  }]
 foot_l [{shapeType=Capsule, length=12.5982, radius=6.5127, center=<-2.2072 6.0534 -0.3367>, rotation=<rot 0.0388 1.4701 -89.3982>}]
 foot_r [{shapeType=Capsule, length=12.5982, radius=6.5127, center=<2.2072 -6.0534 0.3367>, rotation=<rot 0.0388 1.4701 90.6018>}]
 head [{shapeType=Capsule, length=0, radius=18, center=<9.0697 0.0746 0>, rotation=<rot 0 -90 0>}]
 lowerarm_l []
 lowerarm_r []
 neck_02 []
 pelvis [{shapeType=Capsule, length=9.4028, radius=10.6596, center=<-1.9544 -3.5609 -0.1981>}]
 spine_01 [{shapeType=Capsule, length=3.75, radius=15.4854, center=<-0.6936 -0.1504 0>}]
 spine_02 [
  {shapeType=Capsule, length=5, radius=14.2354, center=<1.179 -1.1244 0>}
  {shapeType=Capsule, length=20.6875, radius=9.3125, center=<-1.8317 2.606 0>, rotation=<rot 90 0 0>}
 ]
 spine_03 [
  {shapeType=Capsule, length=2.75, radius=13.4854, center=<-0.0747 -2.4082 0>}
  {shapeType=Capsule, length=20.7188, radius=9.2812, center=<-1.2326 2.6059 0>, rotation=<rot 90 0 0>}
 ]
 spine_04 [
  {shapeType=Capsule, length=2, radius=13.1416, center=<4.1675 -1.2195 0>}
  {shapeType=Capsule, length=10.0312, radius=9.2812, center=<-1.2326 1.308 0>, rotation=<rot 90 0 0>}
 ]
 spine_05 [{shapeType=Capsule, length=7.7359, radius=10.5791, center=<5.6377 2.1778 0>}]
 thigh_l []
 thigh_r []
 upperarm_l []
 upperarm_r []
}
constraints []
