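# Physics asset for the F_MED_DistantEchoCastle female medium player body cosmetic, providing simple collision shapes for the legs only.
# Four bodies each hold one capsule: thicker ones on the left and right thigh_twist_01 bones, thin, longer ones on both calves.
# No constraints link the bodies, so the asset defines standalone leg colliders rather than a jointed ragdoll.
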PhysicsAsset /BRCosmetics/Characters/Player/Female/Medium/Bodies/F_MED_DistantEchoCastle/Meshes/F_MED_DistantEchoCastle_Physics
bodies {
 calf_l [{shapeType=Capsule, length=49.4448, radius=1.5972, center=<-27.3723 0.6601 -0.5435>, rotation=<rot -0.0752 85.4988 -89.1413>}]
 calf_r [{shapeType=Capsule, length=49.4448, radius=1.5972, center=<27.3723 -0.6598 0.5179>, rotation=<rot -0.0745 85.4983 -89.1488>}]
 thigh_twist_01_l [{shapeType=Capsule, length=29, radius=5.25, shapeName=thigh_twist_01_l_capsule, rotation=<rot 90 87.3911 87.3913>}]
 thigh_twist_01_r [{shapeType=Capsule, length=29, radius=5.25, shapeName=thigh_twist_01_r_capsule, rotation=<rot 90 87.3911 87.3913>}]
}
constraints []
