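# Post-simulation Control Rig for the DistantEchoCastle female body, correcting its simulated skirt, sleeve and front panel bones.
# Skirt offsets respond to Kalman-filtered thigh rotation deltas, animation curves, speed, lobby idle and baseball sliding, with spherical pose readers on thighs and calves.
# Middle skirt chains blend between left and right chains through parent constraints.
# When IsHostFull is set, front panels are constrained to the skirt, then posed by curves and movement.
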ControlRigBlueprint /BRCosmetics/Characters/Player/Female/Medium/Bodies/F_MED_DistantEchoCastle/Meshes/F_MED_DistantEchoCastle_PostSim_CtrlRig
graphs [
 New Function_ContainedGraph {nodes [
   RigVMFunctionEntryNode()
   > RigVMFunction_Sequence = RigVMAggregateNode()
   For_Each_2 = RigVMDispatchNode(ExecuteContext=@RigVMFunction_Sequence.B, Array=RigVMVariableNode() {variable=rear_skirt mid}.Value) {template="DISPATCH_RigVMDispatch_ArrayIterator(in Array,out Element,out Index,out Count,out Ratio)"}
   > &RigUnit_ParentConstraint(
    AdvancedSettings=(InterpolationType=Average,RotationOrderForFilter=XZY)
    Parents=((Item=(Type=Bone,Name="None"),Weight=0.5),(Item=(Type=Bone,Name="None"),Weight=0.5))
    Weight=1
    Parents.0.Item=RigVMDispatchNode(Index=@For_Each_2.Index, Array=RigVMVariableNode() {variable=rear_skirt_l}.Value) {template="DISPATCH_RigVMDispatch_ArrayGetAtIndex(in Array,in Index,out Element)"}.Element
    Parents.1.Item=RigVMDispatchNode(Index=@For_Each_2.Index, Array=RigVMVariableNode() {variable=rear_skirt_r}.Value) {template="DISPATCH_RigVMDispatch_ArrayGetAtIndex(in Array,in Index,out Element)"}.Element
    Child=@For_Each_2.Element
   )
   RigVMFunctionReturnNode(ExecuteContext=@RigVMFunction_Sequence.D)
   Remap = &RigVMFunction_MathDoubleRemap(
    SourceMaximum=1
    SourceMinimum=0
    TargetMaximum=0.7
    TargetMinimum=0.5
    Value=/Script/ControlRig.RigUnit_GetCurveValue(Curve=CR Thigh Raised Front L) {is=struct RigVMUnitNode}.Value
   )
   Subtract = &RigVMFunction_MathDoubleSub(A=1, B=@Remap.Result)
   &RigUnit_ParentConstraint(
    AdvancedSettings=(InterpolationType=Average,RotationOrderForFilter=XZY)
    Child=(Type=Bone,Name="dyn_front_skirt_MSTR")
    Filter=(TranslationFilter=(bX=True,bY=True,bZ=True),RotationFilter=(bX=true,bY=true,bZ=true),ScaleFilter=(bX=True,bY=True,bZ=True))
    Parents=((Item=(Type=Bone,Name="dyn_front_skirt_l_MSTR"),Weight=0.5),(Item=(Type=Bone,Name="dyn_front_skirt_r_MSTR"),Weight=0.5))
    Weight=1
    ExecutePin=@RigVMFunction_Sequence.C
    Parents.0.Weight=@Remap.Result
    Parents.1.Weight=@Subtract.Result
   )
   > &RigUnit_ParentConstraint(
    AdvancedSettings=(InterpolationType=Average,RotationOrderForFilter=XZY)
    Child=(Type=Bone,Name="dyn_front_skirt_01")
    Filter=(TranslationFilter=(bX=True,bY=True,bZ=True),RotationFilter=(bX=true,bY=true,bZ=true),ScaleFilter=(bX=True,bY=True,bZ=True))
    Parents=((Item=(Type=Bone,Name="dyn_front_skirt_l_01"),Weight=0.5),(Item=(Type=Bone,Name="dyn_front_skirt_r_01"),Weight=0.5))
    Weight=1
    Parents.0.Weight=@Remap.Result
    Parents.1.Weight=@Subtract.Result
   )
   > &RigUnit_ParentConstraint(
    AdvancedSettings=(InterpolationType=Average,RotationOrderForFilter=XZY)
    Child=(Type=Bone,Name="dyn_front_skirt_02")
    Filter=(TranslationFilter=(bX=True,bY=True,bZ=True),RotationFilter=(bX=true,bY=true,bZ=true),ScaleFilter=(bX=True,bY=True,bZ=True))
    Parents=((Item=(Type=Bone,Name="dyn_front_skirt_l_02"),Weight=0.5),(Item=(Type=Bone,Name="dyn_front_skirt_r_02"),Weight=0.5))
    Weight=1
    Parents.0.Weight=@Remap.Result
    Parents.1.Weight=@Subtract.Result
   )
   Add = &RigVMFunction_MathDoubleAdd(
    A=&RigUnit_GetCurveValue(Curve=CR Thigh Raised Front L).Value
    B=&RigUnit_GetCurveValue(Curve=CR Thigh Raised Front R).Value
   )
   Multiply = &RigVMFunction_MathDoubleMul(B=0.5, A=@Add.Result)
   Clamp = &RigVMFunction_MathDoubleClamp(Value=@Multiply.Result)
   > &RigUnit_ModifyTransforms(
    ItemToModify=((Item=(Type=Bone,Name="dyn_front_skirt_MSTR"),Transform=(Rotation=(X=0,Y=0,Z=-0.1736,W=0.9848),Translation=(X=0,Y=0,Z=0),Scale3D=(X=1,Y=1,Z=1))))
    WeightMinimum=0
    Weight=@Clamp.Result
   )
   > &RigUnit_ModifyTransforms(
    ItemToModify=((Item=(Type=Bone,Name="dyn_front_skirt_MSTR"),Transform=(Rotation=(X=-0.0868,Y=0.0868,Z=0.0076,W=0.9924),Translation=(X=0,Y=0,Z=0),Scale3D=(X=1,Y=1,Z=1))))
    WeightMinimum=0
    Weight=&RigUnit_GetCurveValue(Curve=CR Thigh Raised Front L).Value
   )
   > &RigUnit_ModifyTransforms(
    ItemToModify=((Item=(Type=Bone,Name="dyn_front_skirt_MSTR"),Transform=(Rotation=(X=0.0868,Y=-0.0868,Z=0.0076,W=0.9924),Translation=(X=0,Y=0,Z=0),Scale3D=(X=1,Y=1,Z=1))))
    WeightMinimum=0
    Weight=&RigUnit_GetCurveValue(Curve=CR Thigh Raised Front R).Value
   )
   > &RigUnit_ParentConstraint(
    AdvancedSettings=(InterpolationType=Average,RotationOrderForFilter=XZY)
    Child=(Type=Bone,Name="dyn_front_skirt_03")
    Filter=(TranslationFilter=(bX=True,bY=True,bZ=True),RotationFilter=(bX=true,bY=true,bZ=true),ScaleFilter=(bX=True,bY=True,bZ=True))
    Parents=((Item=(Type=Bone,Name="dyn_front_skirt_l_03"),Weight=0.5),(Item=(Type=Bone,Name="dyn_front_skirt_r_03"),Weight=0.5))
    Weight=1
    Parents.1.Weight=@Subtract.Result
    Parents.0.Weight=@Remap.Result
   )
   > &RigUnit_ParentConstraint(
    AdvancedSettings=(InterpolationType=Average,RotationOrderForFilter=XZY)
    Child=(Type=Bone,Name="dyn_front_skirt_04")
    Filter=(TranslationFilter=(bX=True,bY=True,bZ=True),RotationFilter=(bX=true,bY=true,bZ=true),ScaleFilter=(bX=True,bY=True,bZ=True))
    Parents=((Item=(Type=Bone,Name="dyn_front_skirt_l_04"),Weight=0.5),(Item=(Type=Bone,Name="dyn_front_skirt_r_04"),Weight=0.5))
    Weight=1
    Parents.1.Weight=@Subtract.Result
    Parents.0.Weight=@Remap.Result
   )
   > &RigUnit_ParentConstraint(
    AdvancedSettings=(InterpolationType=Average,RotationOrderForFilter=XZY)
    Child=(Type=Bone,Name="dyn_front_skirt_05")
    Filter=(TranslationFilter=(bX=True,bY=True,bZ=True),RotationFilter=(bX=true,bY=true,bZ=true),ScaleFilter=(bX=True,bY=True,bZ=True))
    Parents=((Item=(Type=Bone,Name="dyn_front_skirt_l_05"),Weight=0.5),(Item=(Type=Bone,Name="dyn_front_skirt_r_05"),Weight=0.5))
    Weight=1
    Parents.1.Weight=@Subtract.Result
    Parents.0.Weight=@Remap.Result
   )
   > &RigUnit_ParentConstraint(
    AdvancedSettings=(InterpolationType=Average,RotationOrderForFilter=XZY)
    Child=(Type=Bone,Name="dyn_front_skirt_06")
    Filter=(TranslationFilter=(bX=True,bY=True,bZ=True),RotationFilter=(bX=true,bY=true,bZ=true),ScaleFilter=(bX=True,bY=True,bZ=True))
    Parents=((Item=(Type=Bone,Name="dyn_front_skirt_l_06"),Weight=0.5),(Item=(Type=Bone,Name="dyn_front_skirt_r_06"),Weight=0.5))
    Weight=1
    Parents.1.Weight=@Subtract.Result
    Parents.0.Weight=@Remap.Result
   )
   For_Each_1 = RigVMDispatchNode(ExecuteContext=(), Array=RigVMVariableNode() {variable=front skirt mid}.Value) {template="DISPATCH_RigVMDispatch_ArrayIterator(in Array,out Element,out Index,out Count,out Ratio)"}
   > &RigUnit_ParentConstraint(
    AdvancedSettings=(InterpolationType=Average,RotationOrderForFilter=XZY)
    Filter=(TranslationFilter=(bX=True,bY=True,bZ=True),RotationFilter=(bX=true,bY=true,bZ=true),ScaleFilter=(bX=True,bY=True,bZ=True))
    Parents=((Item=(Type=Bone,Name="None"),Weight=0.5),(Item=(Type=Bone,Name="None"),Weight=0.5))
    Weight=1
    Parents.0.Item=RigVMDispatchNode(Index=@For_Each_1.Index, Array=RigVMVariableNode() {variable=front skirt l}.Value) {template="DISPATCH_RigVMDispatch_ArrayGetAtIndex(in Array,in Index,out Element)"}.Element
    Parents.1.Item=RigVMDispatchNode(Index=@For_Each_1.Index, Array=RigVMVariableNode() {variable=front skirt r}.Value) {template="DISPATCH_RigVMDispatch_ArrayGetAtIndex(in Array,in Index,out Element)"}.Element
    Child=@For_Each_1.Element
   )
   &RigVMFunction_MathDoubleRemap(
    SourceMaximum=1
    SourceMinimum=0
    TargetMaximum=1
    TargetMinimum=0
    Value=&RigUnit_GetCurveValue(Curve=CR Thigh Raised Front R).Value
   )
   RigVMRerouteNode(Value=@RigVMFunction_Sequence.A)
  ]}
 New Function_ContainedGraph {nodes [
   RigVMFunctionEntryNode()
   n1 = &RigUnit_GetTransform(Item=(Type=Bone,Name="thigh_r"), Space=LocalSpace, bInitial=False)
   n2 = /Script/RigVM.RigVMFunction_MathQuaternionToAxisAndAngle(Value=/Script/RigVM.RigVMFunction_MathQuaternionMul(
     B=/Script/RigVM.RigVMFunction_MathQuaternionInverse(Value=@n1.Transform.Rotation) {is=struct RigVMUnitNode, template="Inverse::Execute(in Value,out Result)"}.Result
     A=RigVMVariableNode() {variable=prev thigh r quaternion}.Value
    ) {is=struct RigVMUnitNode, template="Multiply::Execute(in A,in B,out Result)"}.Result) {is=struct RigVMUnitNode, template="ToAxisAndAngle::Execute(in Value,out Axis,out Angle)"}
   n3 = &RigVMFunction_KalmanFloat(
    BufferSize=4
    Value=&RigVMFunction_MathDoubleRemap(
     SourceMaximum=15
     SourceMinimum=0
     TargetMaximum=1
     TargetMinimum=0
     Value=/Script/RigVM.RigVMFunction_MathDoubleDeg(Value=@n2.Angle) {is=struct RigVMUnitNode, template="Degrees::Execute(in Value,out Result)"}.Result
    ).Result
   ) {template="KalmanFilter::Execute(in Value,in BufferSize,out Result)"}
   RigUnit_GetTransform = &RigUnit_GetTransform(Item=(Type=Bone,Name="thigh_l"), Space=LocalSpace, bInitial=False)
   Inverse = &RigVMFunction_MathQuaternionInverse(Value=@RigUnit_GetTransform.Transform.Rotation) {template="Inverse::Execute(in Value,out Result)"}
   RigVMFunction_MathQuaternionToAxisAndAngle = &RigVMFunction_MathQuaternionToAxisAndAngle(Value=&RigVMFunction_MathQuaternionMul(B=@Inverse.Result, A=RigVMVariableNode() {variable=prev thigh l quaternion}.Value) {template="Multiply::Execute(in A,in B,out Result)"}.Result) {template="ToAxisAndAngle::Execute(in Value,out Axis,out Angle)"}
   Degrees = &RigVMFunction_MathDoubleDeg(Value=@RigVMFunction_MathQuaternionToAxisAndAngle.Angle) {template="Degrees::Execute(in Value,out Result)"}
   KalmanFilter = &RigVMFunction_KalmanFloat(
    BufferSize=4
    Value=&RigVMFunction_MathDoubleRemap(SourceMaximum=-15, SourceMinimum=0, TargetMaximum=-1, TargetMinimum=0, Value=@Degrees.Result).Result
   ) {template="KalmanFilter::Execute(in Value,in BufferSize,out Result)"}
   > Modify Transforms with Mirror = RigVMFunctionReferenceNode(
    Base Modify Data="((Item=(Type=Bone,Name=\"dyn_front_skirt_l_MSTR\"),Transform=(Rotation=(X=0.0076,Y=0.0868,Z=-0.0868,W=0.9924),Translation=(X=0,Y=-5,Z=-5),Scale3D=(X=1,Y=1,Z=1))),(Item=(Type=Bone,Name=\"dyn_side_skirt_l_MSTR\"),Transform=(Rotation=(X=0.0114,Y=0.0864,Z=-0.13,W= ...(truncated)"
    Mirror Modify Data="((Item=(Type=Bone,Name=\"dyn_front_skirt_r_MSTR\"),Transform=(Rotation=(X=0,Y=0,Z=0,W=1),Translation=(X=0,Y=0,Z=0),Scale3D=(X=1,Y=1,Z=1))),(Item=(Type=Bone,Name=\"dyn_side_skirt_r_MSTR\"),Transform=(Rotation=(X=0,Y=0,Z=0,W=1),Translation=(X=0,Y=0,Z=0),Scale3D= ...(truncated)"
    "Mirror Weight "=@n3.Result
    Weight=@KalmanFilter.Result
   )
   Add = &RigVMFunction_MathDoubleAdd(B=@n3.Result, A=@KalmanFilter.Result)
   Clamp = &RigVMFunction_MathDoubleClamp(Maximum=0, Value=&RigVMFunction_MathDoubleMul(B=0.5, A=@Add.Result).Result)
   > &RigUnit_ModifyTransforms(
    ItemToModify=((Item=(Type=Bone,Name="dyn_front_skirt_MSTR"),Transform=(Rotation=(X=0,Y=0,Z=-0.0872,W=0.9962),Translation=(X=0,Y=0,Z=0),Scale3D=(X=1,Y=1,Z=1))))
    WeightMinimum=0
    Weight=@Clamp.Result
   )
   > RigVMVariableNode(Value=@RigUnit_GetTransform.Transform.Rotation) {variable=prev thigh l quaternion}
   > RigVMVariableNode(Value=@n1.Transform.Rotation) {variable=prev thigh r quaternion}
   Remap = &RigVMFunction_MathDoubleRemap(SourceMaximum=770, SourceMinimum=220, TargetMaximum=1, TargetMinimum=0, Value=RigVMVariableNode() {variable=player speed}.Value)
   Multiply = &RigVMFunction_MathDoubleMul(A=&RigUnit_GetCurveValue(Curve=CR Rear Knee Bend L).Value, B=@Remap.Result)
   > Modify Transforms with Mirror_3 = RigVMFunctionReferenceNode(
    Base Modify Data=((Item=(Type=Bone,Name="dyn_rear_skirt_l_01"),Transform=(Rotation=(X=0,Y=-0.1305,Z=0,W=0.9914),Translation=(X=0,Y=0,Z=0),Scale3D=(X=1,Y=1,Z=1))))
    Mirror Modify Data=((Item=(Type=Bone,Name="dyn_rear_skirt_r_01"),Transform=(Rotation=(X=0,Y=0,Z=0,W=1),Translation=(X=0,Y=0,Z=0),Scale3D=(X=1,Y=1,Z=1))))
    Weight=@Multiply.Result
    "Mirror Weight "=&RigVMFunction_MathDoubleMul(B=&RigUnit_GetCurveValue(Curve=CR Rear Knee Bend R).Value, A=@Remap.Result).Result
   )
   > Modify Transforms with Mirror_4 = RigVMFunctionReferenceNode(
    Base Modify Data=((Item=(Type=Bone,Name="dyn_rear_skirt_l_01"),Transform=(Rotation=(X=0,Y=-0.2164,Z=0,W=0.9763),Translation=(X=-5,Y=-3,Z=0),Scale3D=(X=1,Y=1,Z=1))))
    Mirror Modify Data=((Item=(Type=Bone,Name="dyn_rear_skirt_r_01"),Transform=(Rotation=(X=0,Y=0,Z=0,W=1),Translation=(X=0,Y=0,Z=0),Scale3D=(X=1,Y=1,Z=1))))
    Weight=&RigUnit_GetCurveValue(Curve=CR Foot Near Pelvis L).Value
    "Mirror Weight "=&RigUnit_GetCurveValue(Curve=CR Foot Near Pelvis R).Value
   )
   n4 = RigVMVariableNode() {variable=IsCustomLobbyIdle}
   > RigUnit_SetRelativeTransformForItem = /Script/ControlRig.RigUnit_SetRelativeTransformForItem(
    Child=(Type=Bone,Name="dyn_front_skirt_l_02")
    Parent=(Type=Bone,Name="dyn_front_skirt_l_02")
    Value=(Rotation=(X=0,Y=0,Z=-0.0698,W=0.9976),Translation=(X=0,Y=0,Z=0),Scale3D=(X=1,Y=1,Z=1))
    bParentInitial=False
    bPropagateToChildren=false
    Weight=@n4.Value
   ) {
    is=struct RigVMUnitNode
    template="Set Relative Transform::Execute(in Child,in Parent,in bParentInitial,in Value,in Weight,in bPropagateToChildren)"
   }
   > &RigUnit_SetRelativeTransformForItem(
    Child=(Type=Bone,Name="dyn_front_skirt_l_03")
    Parent=(Type=Bone,Name="dyn_front_skirt_l_03")
    Value=(Rotation=(X=0,Y=0,Z=0,W=1),Translation=(X=0,Y=-1.2,Z=0),Scale3D=(X=1,Y=1,Z=1))
    bParentInitial=False
    bPropagateToChildren=false
    Weight=@n4.Value
   ) {template="Set Relative Transform::Execute(in Child,in Parent,in bParentInitial,in Value,in Weight,in bPropagateToChildren)"}
   > &RigUnit_SetRelativeTransformForItem(
    Child=(Type=Bone,Name="dyn_front_skirt_02")
    Parent=(Type=Bone,Name="dyn_front_skirt_02")
    Value=(Rotation=(X=0,Y=0,Z=-0.0698,W=0.9976),Translation=(X=0,Y=0,Z=0),Scale3D=(X=1,Y=1,Z=1))
    bParentInitial=False
    bPropagateToChildren=false
    Weight=@n4.Value
   ) {template="Set Relative Transform::Execute(in Child,in Parent,in bParentInitial,in Value,in Weight,in bPropagateToChildren)"}
   > &RigUnit_SetRelativeTransformForItem(
    Child=(Type=Bone,Name="dyn_front_skirt_03")
    Parent=(Type=Bone,Name="dyn_front_skirt_03")
    Value=(Rotation=(X=0,Y=0,Z=0,W=1),Translation=(X=0,Y=1.2,Z=0),Scale3D=(X=1,Y=1,Z=1))
    bParentInitial=False
    bPropagateToChildren=false
    Weight=@n4.Value
   ) {template="Set Relative Transform::Execute(in Child,in Parent,in bParentInitial,in Value,in Weight,in bPropagateToChildren)"}
   > &RigUnit_SetRelativeTransformForItem(
    Child=(Type=Bone,Name="dyn_side_skirt_l_01")
    Parent=(Type=Bone,Name="dyn_side_skirt_l_01")
    Value=(Rotation=(X=0,Y=0,Z=0.0175,W=0.9998),Translation=(X=0,Y=2,Z=0),Scale3D=(X=1,Y=1,Z=1))
    bParentInitial=False
    bPropagateToChildren=true
    Weight=@n4.Value
   ) {template="Set Relative Transform::Execute(in Child,in Parent,in bParentInitial,in Value,in Weight,in bPropagateToChildren)"}
   > &RigUnit_SetRelativeTransformForItem(
    Child=(Type=Bone,Name="dyn_side_skirt_r_02")
    Parent=(Type=Bone,Name="dyn_side_skirt_r_02")
    Value=(Rotation=(X=0,Y=0,Z=0.0175,W=0.9998),Translation=(X=0,Y=-2,Z=0),Scale3D=(X=1,Y=1,Z=1))
    bParentInitial=False
    bPropagateToChildren=True
    Weight=@n4.Value
   ) {template="Set Relative Transform::Execute(in Child,in Parent,in bParentInitial,in Value,in Weight,in bPropagateToChildren)"}
   > &RigUnit_SetRelativeTransformForItem(
    Child=(Type=Bone,Name="dyn_rear_skirt_l_01")
    Parent=(Type=Bone,Name="dyn_rear_skirt_l_01")
    Value=(Rotation=(X=0,Y=0,Z=0.061,W=0.9981),Translation=(X=0,Y=0,Z=0),Scale3D=(X=1,Y=1,Z=1))
    bParentInitial=False
    bPropagateToChildren=True
    Weight=@n4.Value
   ) {template="Set Relative Transform::Execute(in Child,in Parent,in bParentInitial,in Value,in Weight,in bPropagateToChildren)"}
   > &RigUnit_SetRelativeTransformForItem(
    Child=(Type=Bone,Name="dyn_rear_skirt_r_01")
    Parent=(Type=Bone,Name="dyn_rear_skirt_r_01")
    Value=(Rotation=(X=0,Y=0,Z=0.0349,W=0.9994),Translation=(X=0,Y=0,Z=0),Scale3D=(X=1,Y=1,Z=1))
    bParentInitial=False
    bPropagateToChildren=True
    Weight=@n4.Value
   ) {template="Set Relative Transform::Execute(in Child,in Parent,in bParentInitial,in Value,in Weight,in bPropagateToChildren)"}
   > RigVMFunctionReturnNode()
  ]}
 New Function_ContainedGraph {nodes [
   RigVMFunctionEntryNode()
   > RigVMFunction_Sequence = RigVMAggregateNode()
   For_Each_1 = RigVMDispatchNode(ExecuteContext=@RigVMFunction_Sequence.B, Array=RigVMVariableNode() {variable=front panel l}.Value) {template="DISPATCH_RigVMDispatch_ArrayIterator(in Array,out Element,out Index,out Count,out Ratio)"}
   > &RigUnit_AimConstraintLocalSpaceOffset(
    AdvancedSettings=(DebugSettings=(bEnabled=true,Scale=10,WorldOffset=(Rotation=(X=0,Y=0,Z=0,W=1),Translation=(X=0,Y=0,Z=0),Scale3D=(X=1,Y=1,Z=1))),RotationOrderForFilter=XZY)
    Filter=(bX=True,bY=True,bZ=True)
    Parents=((Item=(Type=Bone,Name="None"),Weight=1))
    UpAxis=(X=0,Y=-1,Z=0)
    Weight=1
    WorldUp=(Target=(X=0,Y=1,Z=0),Kind=Location,Space=(Type=None,Name="None"))
    bMaintainOffset=true
    Parents.0.Item=RigVMDispatchNode(Index=@For_Each_1.Index, Array=RigVMVariableNode() {variable=front skirt l}.Value) {template="DISPATCH_RigVMDispatch_ArrayGetAtIndex(in Array,in Index,out Element)"}.Element
    Child=@For_Each_1.Element
    WorldUp.Space=@For_Each_1.Element
   )
   For_Each_2 = RigVMDispatchNode(ExecuteContext=@RigVMFunction_Sequence.C, Array=RigVMVariableNode() {variable=front panel r}.Value) {template="DISPATCH_RigVMDispatch_ArrayIterator(in Array,out Element,out Index,out Count,out Ratio)"}
   > &RigUnit_AimConstraintLocalSpaceOffset(
    AimAxis=(X=0,Y=0,Z=1)
    Filter=(bX=True,bY=True,bZ=True)
    Parents=((Item=(Type=Bone,Name="None"),Weight=1))
    UpAxis=(X=0,Y=1,Z=0)
    Weight=1
    WorldUp=(Target=(X=0,Y=-1,Z=0),Kind=Location,Space=(Type=None,Name="None"))
    Parents.0.Item=RigVMDispatchNode(Index=@For_Each_2.Index, Array=RigVMVariableNode() {variable=front skirt r}.Value) {template="DISPATCH_RigVMDispatch_ArrayGetAtIndex(in Array,in Index,out Element)"}.Element
    Child=@For_Each_2.Element
    WorldUp.Space=@For_Each_2.Element
   )
   RigVMFunctionReturnNode(ExecuteContext=@RigVMFunction_Sequence.D)
   For_Each = RigVMDispatchNode(Array=RigVMVariableNode() {variable=front panel mid}.Value, ExecuteContext=@RigVMFunction_Sequence.A) {template="DISPATCH_RigVMDispatch_ArrayIterator(in Array,out Element,out Index,out Count,out Ratio)"}
   > &RigUnit_ParentConstraint(
    AdvancedSettings=(InterpolationType=Average,RotationOrderForFilter=XZY)
    Parents=((Item=(Type=Bone,Name="None"),Weight=1))
    Weight=1
    Parents.0.Item=RigVMDispatchNode(Index=@For_Each.Index, Array=RigVMVariableNode() {variable=front skirt mid}.Value) {template="DISPATCH_RigVMDispatch_ArrayGetAtIndex(in Array,in Index,out Element)"}.Element
    Child=@For_Each.Element
   )
  ]}
 New Function_ContainedGraph {nodes [
   RigVMFunctionEntryNode()
   > RigVMFunction_Sequence = RigVMAggregateNode()
   RigVMFunctionReturnNode(ExecuteContext=@RigVMFunction_Sequence.C)
   Maximum = &RigVMFunction_MathDoubleMax(
    A=&RigUnit_GetCurveValue(Curve=CR Thigh Raised Front L).Value
    B=&RigUnit_GetCurveValue(Curve=CR Thigh Raised Front R).Value
   )
   KalmanFilter = &RigVMFunction_KalmanFloat(BufferSize=8, Value=@Maximum.Result) {template="KalmanFilter::Execute(in Value,in BufferSize,out Result)"}
   &RigUnit_ModifyTransforms(
    ItemToModify="((Item=(Type=Bone,Name=\"dyn_front_panel_02\"),Transform=(Rotation=(X=0,Y=0,Z=-0.0175,W=0.9998),Translation=(X=0,Y=0,Z=0),Scale3D=(X=1,Y=1,Z=1))),(Item=(Type=Bone,Name=\"dyn_front_panel_MSTR\"),Transform=(Rotation=(X=0,Y=0,Z=-0.0175,W=0.9998),Translation=(X=0, ...(truncated)"
    WeightMinimum=0
    Weight=@KalmanFilter.Result
    ExecutePin=@RigVMFunction_Sequence.B
   )
   > &RigUnit_ModifyTransforms(
    ItemToModify=((Item=(Type=Bone,Name="dyn_front_panel_02"),Transform=(Rotation=(X=-0.1171,Y=-0.2446,Z=-0.148,W=0.9511),Translation=(X=0,Y=0,Z=0),Scale3D=(X=1,Y=1,Z=1))))
    WeightMinimum=0
    Weight=&RigUnit_GetCurveValue(Curve=CR Thigh L Across R).Value
   )
   > &RigUnit_ModifyTransforms(
    ItemToModify=((Item=(Type=Bone,Name="dyn_front_panel_02"),Transform=(Rotation=(X=0.1171,Y=0.2446,Z=-0.148,W=0.9511),Translation=(X=0,Y=0,Z=0),Scale3D=(X=1,Y=1,Z=1))))
    WeightMinimum=0
    Weight=&RigUnit_GetCurveValue(Curve=CR Thigh R Across L).Value
   )
   > Modify Transforms with Mirror_2 = RigVMFunctionReferenceNode(
    Base Modify Data="((Item=(Type=Bone,Name=\"dyn_front_panel_l_01\"),Transform=(Rotation=(X=0,Y=0,Z=0,W=1),Translation=(X=0,Y=-3,Z=0),Scale3D=(X=1,Y=1,Z=1))),(Item=(Type=Bone,Name=\"dyn_front_panel_l_02\"),Transform=(Rotation=(X=0,Y=0,Z=0,W=1),Translation=(X=0,Y=-2.5,Z=0),Scale3D ...(truncated)"
    Mirror Modify Data="((Item=(Type=Bone,Name=\"dyn_front_panel_r_01\"),Transform=(Rotation=(X=0,Y=0,Z=0,W=1),Translation=(X=0,Y=0,Z=0),Scale3D=(X=1,Y=1,Z=1))),(Item=(Type=Bone,Name=\"dyn_front_panel_r_02\"),Transform=(Rotation=(X=0,Y=0,Z=0,W=1),Translation=(X=0,Y=0,Z=0),Scale3D=(X= ...(truncated)"
    Weight=&RigUnit_GetCurveValue(Curve=CR Thigh Raised Front L).Value
    "Mirror Weight "=&RigUnit_GetCurveValue(Curve=CR Thigh Raised Front R).Value
   )
   RigVMFunction_ControlFlowBranch = &RigVMFunction_ControlFlowBranch(ExecuteContext=@"Modify Transforms with Mirror_2".ExecuteContext, Condition=RigVMVariableNode() {variable=IsPlayingEmote}.Value)
   Remap = &RigVMFunction_MathDoubleRemap(SourceMaximum=770, SourceMinimum=0, TargetMaximum=1, TargetMinimum=0, Value=RigVMVariableNode() {variable=player speed}.Value)
   n1 = &RigUnit_ModifyTransforms(
    ItemToModify=((Item=(Type=Bone,Name="dyn_front_panel_MSTR"),Transform=(Rotation=(X=0,Y=0,Z=-0.0436,W=0.999),Translation=(X=0,Y=-10,Z=0),Scale3D=(X=1,Y=1,Z=1))))
    WeightMinimum=0
    Weight=@Remap.Result
    ExecutePin=@RigVMFunction_ControlFlowBranch.False
   )
   n2 = &RigVMFunction_ControlFlowBranch(ExecuteContext=@n1.ExecutePin, Condition=RigVMVariableNode() {variable=IsInFrontend}.Value)
   Absolute = &RigVMFunction_MathDoubleAbs(Value=RigVMVariableNode() {variable=rotation speed}.Value)
   &RigUnit_ModifyTransforms(
    ItemToModify=((Item=(Type=Bone,Name="dyn_front_panel_MSTR"),Transform=(Rotation=(X=0,Y=0,Z=-0.3827,W=0.9239),Translation=(X=0,Y=0,Z=0),Scale3D=(X=1,Y=1,Z=1))))
    WeightMinimum=0
    Weight=&RigVMFunction_MathDoubleRemap(SourceMaximum=7, SourceMinimum=0, TargetMaximum=1, TargetMinimum=0, Value=@Absolute.Result).Result
    ExecutePin=@n2.False
   )
   &RigUnit_ModifyTransforms(
    ItemToModify=((Item=(Type=Bone,Name="dyn_front_panel_MSTR"),Transform=(Rotation=(X=0,Y=0,Z=-0.0175,W=0.9998),Translation=(X=0,Y=-0.5,Z=0),Scale3D=(X=1,Y=1,Z=1))))
    Weight=0.5
    WeightMinimum=0
    ExecutePin=@RigVMFunction_Sequence.A
   )
   > Modify Transforms with Mirror_1 = RigVMFunctionReferenceNode(
    Base Modify Data="((Item=(Type=Bone,Name=\"dyn_front_panel_l_01\"),Transform=(Rotation=(X=-0.0436,Y=0,Z=0,W=0.999),Translation=(X=0,Y=-0.5,Z=0),Scale3D=(X=1,Y=1,Z=1))),(Item=(Type=Bone,Name=\"dyn_front_panel_l_02\"),Transform=(Rotation=(X=-0.0436,Y=0,Z=0,W=0.999),Translation=(X ...(truncated)"
    Mirror Modify Data="((Item=(Type=Bone,Name=\"dyn_front_panel_r_01\"),Transform=(Rotation=(X=0,Y=0,Z=0,W=1),Translation=(X=0,Y=0,Z=0),Scale3D=(X=1,Y=1,Z=1))),(Item=(Type=Bone,Name=\"dyn_front_panel_r_02\"),Transform=(Rotation=(X=0,Y=0,Z=0,W=1),Translation=(X=0,Y=0,Z=0),Scale3D=(X= ...(truncated)"
    "Mirror Weight "=1
    Weight=1
   )
  ]}
 New Function_ContainedGraph {nodes [
   RigVMFunctionEntryNode()
   > RigVMFunctionReturnNode()
  ]}
 RigVMFunctionLibrary {nodes [
   Constrain Front Panel = RigVMCollapseNode(ExecuteContext=())
   Constrain Mid Sim Chains = RigVMCollapseNode(ExecuteContext=())
   Deform Post Sim Skirt Curves = RigVMCollapseNode(ExecuteContext=())
   Deform Sleeves = RigVMCollapseNode(ExecuteContext=())
   Pose Front Panel = RigVMCollapseNode(ExecuteContext=())
  ]}
 RigVMFunction_Sequence_1_ContainedGraph {nodes [
   n1 = RigVMFunctionEntryNode()
   RigVMFunction_Sequence = &RigVMFunction_Sequence(ExecuteContext=@n1.ExecuteContext)
   n2 = &RigVMFunction_Sequence(ExecuteContext=@RigVMFunction_Sequence.B)
   n3 = &RigVMFunction_Sequence(ExecuteContext=@n2.B)
   n4 = &RigVMFunction_Sequence(ExecuteContext=@n3.B)
   RigVMFunctionReturnNode(A=@RigVMFunction_Sequence.A, B=@n2.A, C=@n3.A, D=@n4.A, E=@n4.B)
  ]}
 RigVMFunction_Sequence_1_ContainedGraph {nodes [
   n1 = RigVMFunctionEntryNode()
   RigVMFunction_Sequence = &RigVMFunction_Sequence(ExecuteContext=@n1.ExecuteContext)
   n2 = &RigVMFunction_Sequence(ExecuteContext=@RigVMFunction_Sequence.B)
   n3 = &RigVMFunction_Sequence(ExecuteContext=@n2.B)
   RigVMFunctionReturnNode(A=@RigVMFunction_Sequence.A, B=@n2.A, C=@n3.A, D=@n3.B)
  ]}
 RigVMFunction_Sequence_1_ContainedGraph {nodes [
   n1 = RigVMFunctionEntryNode()
   RigVMFunction_Sequence = &RigVMFunction_Sequence(ExecuteContext=@n1.ExecuteContext)
   n2 = &RigVMFunction_Sequence(ExecuteContext=@RigVMFunction_Sequence.B)
   n3 = &RigVMFunction_Sequence(ExecuteContext=@n2.B)
   RigVMFunctionReturnNode(A=@RigVMFunction_Sequence.A, B=@n2.A, C=@n3.A, D=@n3.B)
  ]}
 RigVMFunction_Sequence_1_ContainedGraph {nodes [
   n1 = RigVMFunctionEntryNode()
   RigVMFunction_Sequence = &RigVMFunction_Sequence(ExecuteContext=@n1.ExecuteContext)
   n2 = &RigVMFunction_Sequence(ExecuteContext=@RigVMFunction_Sequence.B)
   RigVMFunctionReturnNode(A=@RigVMFunction_Sequence.A, B=@n2.A, C=@n2.B)
  ]}
 RigVMModel {nodes [
   RigUnit_BeginExecution = &RigUnit_BeginExecution()
   n1 = &RigVMFunction_ControlFlowBranch(
    Condition=RigVMDispatchNode(B=1, A=RigVMVariableNode() {variable=IsCustomLobbyIdle}.Value) {template="DISPATCH_RigVMDispatch_CoreEquals(in A,in B,out Result)"}.Result
    ExecuteContext=@RigUnit_BeginExecution.ExecutePin
   )
   RigVMVariableNode(Value=1, ExecuteContext=@n1.False) {variable=RobeOffsetWeight}
   RigVMVariableNode(Value=0.5, ExecuteContext=@n1.True) {variable=RobeOffsetWeight}
   RigVMFunction_Sequence = RigVMAggregateNode(ExecuteContext=@n1.Completed)
   Modify Transforms with Mirror_1 = RigVMFunctionReferenceNode(
    Base Modify Data="((Item=(Type=Bone,Name=\"dyn_lowerarm_l_sleeve_twist_01_MSTR\"),Transform=(Rotation=(X=0,Y=0,Z=0,W=1),Translation=(X=0,Y=0,Z=0.1),Scale3D=(X=1,Y=1,Z=1))),(Item=(Type=Bone,Name=\"dyn_lowerarm_l_sleeve_twist_02_MSTR\"),Transform=(Rotation=(X=0,Y=0,Z=0,W=1),Trans ...(truncated)"
    Mirror Modify Data="((Item=(Type=Bone,Name=\"dyn_lowerarm_r_sleeve_twist_01_MSTR\"),Transform=(Rotation=(X=0,Y=0,Z=0,W=1),Translation=(X=0,Y=0,Z=0),Scale3D=(X=1,Y=1,Z=1))),(Item=(Type=Bone,Name=\"dyn_lowerarm_r_sleeve_twist_02_MSTR\"),Transform=(Rotation=(X=0,Y=0,Z=0,W=1),Transla ...(truncated)"
    "Mirror Weight "=1
    Weight=1
    ExecuteContext=@RigVMFunction_Sequence.B
   )
   RigVMFunction_ControlFlowBranch = &RigVMFunction_ControlFlowBranch(ExecuteContext=@RigVMFunction_Sequence.C, Condition=RigVMVariableNode() {variable=IsHostFull}.Value)
   Constrain Front Panel = RigVMFunctionReferenceNode(ExecuteContext=@RigVMFunction_ControlFlowBranch.True)
   > Pose Front Panel = RigVMFunctionReferenceNode()
   n2 = &RigVMFunction_ControlFlowBranch(ExecuteContext=@RigVMFunction_Sequence.D, Condition=RigVMVariableNode() {variable=IsBaseballSliding}.Value)
   &RigUnit_ModifyTransforms(
    ItemToModify="((Item=(Type=Bone,Name=\"dyn_side_skirt_r_MSTR\"),Transform=(Rotation=(X=0,Y=0,Z=-0.1736,W=0.9848),Translation=(X=0,Y=0,Z=0),Scale3D=(X=1,Y=1,Z=1))),(Item=(Type=Bone,Name=\"dyn_side_skirt_r_02\"),Transform=(Rotation=(X=0,Y=0,Z=0.342,W=0.9397),Translation=(X=0, ...(truncated)"
    Weight=1
    WeightMinimum=0
    ExecutePin=@n2.True
   )
   n3 = &RigUnit_SphericalPoseReader(
    ActiveRegionScaleFactors=(PositiveWidth=3,NegativeWidth=1,PositiveHeight=1.2,NegativeHeight=1)
    ActiveRegionSize=0.1
    Debug=(bDrawDebug=false,bDraw2D=False,bDrawLocalAxes=False,DebugScale=25,DebugSegments=20,DebugThickness=0.25)
    DriverAxis=(X=-1,Y=0,Z=0)
    DriverItem=(Type=Bone,Name="thigh_l")
    FalloffRegionScaleFactors=(PositiveWidth=1,NegativeWidth=1,PositiveHeight=0.75,NegativeHeight=0.75)
    FalloffSize=0.58
    FlipHeightScaling=False
    FlipWidthScaling=False
    OptionalParentItem=(Type=Bone,Name="None")
    RotationOffset=(X=90,Y=0,Z=-15)
    ExecutePin=@RigVMFunction_Sequence.E
   )
   > &RigUnit_ModifyTransforms(
    ItemToModify="((Item=(Type=Bone,Name=\"dyn_front_skirt_MSTR\"),Transform=(Rotation=(X=0,Y=0,Z=-0.0262,W=0.9997),Translation=(X=0,Y=0,Z=0),Scale3D=(X=1,Y=1,Z=1))),(Item=(Type=Bone,Name=\"dyn_front_skirt_l_MSTR\"),Transform=(Rotation=(X=-0.0038,Y=0.0435,Z=0.0871,W=0.9952),Tra ...(truncated)"
    WeightMinimum=0
    Weight=@n3.OutputParam
   )
   > n4 = &RigUnit_SphericalPoseReader(
    ActiveRegionScaleFactors=(PositiveWidth=3,NegativeWidth=1,PositiveHeight=1.2,NegativeHeight=1)
    ActiveRegionSize=0.1
    Debug=(bDrawDebug=false,bDraw2D=False,bDrawLocalAxes=False,DebugScale=25,DebugSegments=20,DebugThickness=0.25)
    DriverAxis=(X=1,Y=0,Z=0)
    DriverItem=(Type=Bone,Name="thigh_r")
    FalloffRegionScaleFactors=(PositiveWidth=1,NegativeWidth=1,PositiveHeight=0.75,NegativeHeight=0.75)
    FalloffSize=0.58
    FlipHeightScaling=true
    FlipWidthScaling=true
    OptionalParentItem=(Type=Bone,Name="None")
    RotationOffset=(X=-90,Y=0,Z=-15)
   )
   > &RigUnit_ModifyTransforms(
    ItemToModify=((Item=(Type=Bone,Name="dyn_front_skirt_MSTR"),Transform=(Rotation=(X=0,Y=0,Z=-0.0262,W=0.9997),Translation=(X=0,Y=0,Z=0),Scale3D=(X=1,Y=1,Z=1))))
    WeightMinimum=0
    Weight=@n4.OutputParam
   )
   > n5 = &RigUnit_SphericalPoseReader(
    ActiveRegionScaleFactors=(PositiveWidth=3,NegativeWidth=0.5,PositiveHeight=5,NegativeHeight=5)
    ActiveRegionSize=0.1
    Debug=(bDrawDebug=false,bDraw2D=False,bDrawLocalAxes=False,DebugScale=25,DebugSegments=20,DebugThickness=0.25)
    DriverAxis=(X=-1,Y=0,Z=0)
    DriverItem=(Type=Bone,Name="calf_l")
    FalloffRegionScaleFactors=(PositiveWidth=0.2,NegativeWidth=0.92,PositiveHeight=1,NegativeHeight=1)
    FalloffSize=0.7
    FlipHeightScaling=False
    FlipWidthScaling=False
    OptionalParentItem=(Type=Bone,Name="thigh_l")
    RotationOffset=(X=-90,Y=0,Z=27)
   )
   > &RigUnit_ModifyTransforms(
    ItemToModify="((Item=(Type=Bone,Name=\"dyn_side_skirt_l_02\"),Transform=(Rotation=(X=0,Y=0.0872,Z=0,W=0.9962),Translation=(X=0,Y=0,Z=0),Scale3D=(X=1,Y=1,Z=1))),(Item=(Type=Bone,Name=\"dyn_side_skirt_l_MSTR\"),Transform=(Rotation=(X=0.0151,Y=0.173,Z=-0.0858,W=0.9811),Transla ...(truncated)"
    WeightMinimum=0
    Weight=@n5.OutputParam
   )
   > n6 = &RigUnit_SphericalPoseReader(
    ActiveRegionScaleFactors=(PositiveWidth=3,NegativeWidth=0.5,PositiveHeight=5,NegativeHeight=5)
    ActiveRegionSize=0.1
    Debug=(bDrawDebug=false,bDraw2D=False,bDrawLocalAxes=False,DebugScale=25,DebugSegments=20,DebugThickness=0.25)
    DriverAxis=(X=1,Y=0,Z=0)
    DriverItem=(Type=Bone,Name="calf_r")
    FalloffRegionScaleFactors=(PositiveWidth=0.2,NegativeWidth=0.92,PositiveHeight=1,NegativeHeight=1)
    FalloffSize=0.7
    FlipHeightScaling=true
    FlipWidthScaling=true
    OptionalParentItem=(Type=Bone,Name="thigh_r")
    RotationOffset=(X=90,Y=0,Z=27)
   )
   > &RigUnit_ModifyTransforms(
    ItemToModify="((Item=(Type=Bone,Name=\"dyn_side_skirt_r_02\"),Transform=(Rotation=(X=0,Y=0.0872,Z=0,W=0.9962),Translation=(X=0,Y=0,Z=0),Scale3D=(X=1,Y=1,Z=1))),(Item=(Type=Bone,Name=\"dyn_side_skirt_r_MSTR\"),Transform=(Rotation=(X=0.0151,Y=0.173,Z=-0.0858,W=0.9811),Transla ...(truncated)"
    WeightMinimum=0
    Weight=@n6.OutputParam
   )
   n7 = RigVMVariableNode() {variable=RobeOffsetWeight}
   Modify Transforms with Mirror = RigVMFunctionReferenceNode(
    Base Modify Data="((Item=(Type=Bone,Name=\"dyn_front_skirt_l_01\"),Transform=(Rotation=(X=0,Y=0,Z=0,W=1),Translation=(X=0,Y=-1,Z=-1),Scale3D=(X=1,Y=1,Z=1))),(Item=(Type=Bone,Name=\"dyn_side_skirt_l_01\"),Transform=(Rotation=(X=0,Y=0,Z=-0.0175,W=0.9998),Translation=(X=0,Y=-1,Z=0 ...(truncated)"
    Mirror Modify Data="((Item=(Type=Bone,Name=\"dyn_front_skirt_r_01\"),Transform=(Rotation=(X=0,Y=0,Z=0,W=1),Translation=(X=0,Y=0,Z=0),Scale3D=(X=1,Y=1,Z=1))),(Item=(Type=Bone,Name=\"dyn_side_skirt_r_01\"),Transform=(Rotation=(X=0,Y=0,Z=0,W=1),Translation=(X=0,Y=0,Z=0),Scale3D=(X=1 ...(truncated)"
    "Mirror Weight "=@n7.Value
    Weight=@n7.Value
    ExecuteContext=@RigVMFunction_Sequence.A
   )
   > Deform Post Sim Skirt Curves = RigVMFunctionReferenceNode()
   > Constrain Mid Sim Chains = RigVMFunctionReferenceNode()
  ]}
 RigVMModel Construc {nodes [
   n1 = &RigUnit_PrepareForExecution()
   RigUnit_GetTransform = &RigUnit_GetTransform(Item=(Type=Bone,Name="thigh_l"), Space=LocalSpace, bInitial=False)
   RigVMVariableNode(ExecuteContext=@n1.ExecutePin, Value=@RigUnit_GetTransform.Transform.Rotation) {variable=prev thigh l quaternion}
   > RigVMVariableNode(Value=&RigUnit_GetTransform(Item=(Type=Bone,Name="thigh_r"), Space=LocalSpace, bInitial=False).Transform.Rotation) {variable=prev thigh r quaternion}
  ]}
]
hierarchy [
 root {loc=<0 0 0>, rot=<0 0 0 1>}
 pelvis {parent=root, loc=<0 -0.5301 98.0571>, rot=<0.0127 0.707 -0.0127 -0.707>}
 spine_01 {parent=pelvis, loc=<6.1208 -0.9328 0>, rot=<0 0 0.034 -0.9994>}
 spine_02 {parent=spine_01, loc=<6.5996 -0.2095 0>, rot=<0 0 0.0238 -0.9997>}
 spine_03 {parent=spine_02, loc=<6.4914 -0.1237 0>, rot=<0 0 0.0149 -0.9999>}
 spine_04 {parent=spine_03, loc=<6.9598 0.0794 0>, rot=<0 0 -0.0132 -0.9999>}
 spine_05 {parent=spine_04, loc=<6.9987 0 0>, rot=<0 0 -0.0015 -1>}
 neck_01 {parent=spine_05, loc=<11.3138 0.1529 0>, rot=<0 0 -0.2003 -0.9797>}
 neck_02 {parent=neck_01, loc=<4.5627 0 0>, rot=<0 0 0.0504 -0.9987>}
 head {parent=neck_02, loc=<5.3914 0 0>, rot=<0 0 0.115 -0.9934>}
 attach_cape {parent=spine_05, loc=<9.7215 -9.4575 0>, rot=<0.1328 -0.6945 -0.1328 -0.6945>}
 cape {parent=attach_cape, type=Null, loc=<0 2.401 -4.7663>, rot=<0.0872 0 0 0.9962>}
 attach_backpack {parent=spine_05, loc=<-2.006 -14.6941 0>, rot=<0.0353 -0.7062 -0.0353 -0.7062>}
 Backpack_BR {parent=attach_backpack, type=Null, loc=<0 4.5 0>, rot=<0.061 0 0 0.9981>}
 clavicle_l {parent=spine_05, loc=<5.4831 -0.1053 -3.2946>, rot=<0.0465 -0.7289 -0.009 -0.6829>}
 upperarm_l {parent=clavicle_l, loc=<9.8658 -1.2136 -0.4105>, rot=<0.0164 -0.3245 -0.0269 -0.9454>}
 lowerarm_l {parent=upperarm_l, loc=<26.465 -0.8345 0.259>, rot=<0.0036 0.0161 -0.2298 -0.9731>}
 hand_l {parent=lowerarm_l, loc=<22.7355 -0.3203 -0.6105>, rot=<-0.6784 -0.0223 0.0735 -0.7307>}
 thumb_01_l {parent=hand_l, loc=<1.3118 -0.3191 -2.459>, rot=<0.6546 -0.4042 0.0664 -0.6354>}
 thumb_02_l {parent=thumb_01_l, loc=<5.9336 -0.0448 0.0401>, rot=<-0.0037 0.0213 0.298 -0.9543>}
 thumb_03_l {parent=thumb_02_l, loc=<2.2223 0.3682 -0.0907>, rot=<0.0118 -0.013 0.0066 -0.9998>}
 index_metacarpal_l {parent=hand_l, loc=<1.6535 0.0183 -2.96>, rot=<0.0158 -0.0124 -0.0173 -0.9996>}
 index_01_l {parent=index_metacarpal_l, loc=<8.3008 0.0357 -0.0524>, rot=<-0.0044 0.0485 0.0924 -0.9945>}
 index_02_l {parent=index_01_l, loc=<3.3587 -0.3004 0.0452>, rot=<-0.002 0.0027 0.3225 -0.9466>}
 index_03_l {parent=index_02_l, loc=<2.5578 -0.123 -0.0252>, rot=<-0.002 0.0067 0.2558 -0.9667>}
 middle_metacarpal_l {parent=hand_l, loc=<2.1856 -0.1725 -0.5984>, rot=<0.006 0.0168 -0.0329 -0.9993>}
 middle_01_l {parent=middle_metacarpal_l, loc=<7.3275 -0.2447 0.0129>, rot=<-0.0208 0.0042 0.0858 -0.9961>}
 middle_02_l {parent=middle_01_l, loc=<4.0123 0.1401 0.1175>, rot=<-0.0218 0.0258 0.3512 -0.9357>}
 middle_03_l {parent=middle_02_l, loc=<2.6077 0.0534 0.1373>, rot=<0.0082 0.0188 0.1708 -0.9851>}
 ring_metacarpal_l {parent=hand_l, loc=<2.2074 -0.3625 1.3638>, rot=<-0.0284 0.0722 -0.0458 -0.9959>}
 ring_01_l {parent=ring_metacarpal_l, loc=<7.0295 -0.6197 -0.1446>, rot=<-0.0489 -0.0252 0.1892 -0.9804>}
 ring_02_l {parent=ring_01_l, loc=<3.6174 0.4647 -0.1027>, rot=<0 0 0.2806 -0.9598>}
 ring_03_l {parent=ring_02_l, loc=<2.4152 0.114 0.0199>, rot=<0 0 0.1189 -0.9929>}
 pinky_metacarpal_l {parent=hand_l, loc=<1.7616 -0.9753 3.3814>, rot=<-0.0397 0.0814 -0.069 -0.9935>}
 pinky_01_l {parent=pinky_metacarpal_l, loc=<7.1709 -0.4831 -0.1094>, rot=<-0.0654 -0.0035 0.2242 -0.9723>}
 pinky_02_l {parent=pinky_01_l, loc=<2.8274 0.0893 -0.1043>, rot=<0.0037 0.0011 0.2854 -0.9584>}
 pinky_03_l {parent=pinky_02_l, loc=<2.2046 0.0307 -0.0078>, rot=<-0.0002 0.0098 0.1445 -0.9895>}
 hand_attach_l {parent=hand_l, loc=<8.5 -3 0>, rot=<0.7071 0.7071 0 0>}
 weapon_l {parent=hand_l, loc=<-0.1974 -3.4111 0.3811>, rot=<0 0 0 1>}
 lowerarm_twist_02_l {parent=lowerarm_l, loc=<10.6707 0 0>, rot=<0 0 0 1>}
 lowerarm_l_front {parent=lowerarm_twist_02_l, loc=<-10.6707 0.5246 0>, rot=<0 0 0 -1>}
 lowerarm_l_back {parent=lowerarm_twist_02_l, loc=<-10.6707 -0.4766 0>, rot=<0 0 0 -1>}
 lowerarm_twist_01_l {parent=lowerarm_l, loc=<18.316 0 0>, rot=<0 0 0 1>}
 dyn_Cloth_lowerarm_l {parent=lowerarm_twist_01_l, loc=<-13.2236 0.085 2.6897>, rot=<0.2399 -0.6171 0.6392 0.3913>}
 dyn_Cloth_lowerarm_l_A {parent=dyn_Cloth_lowerarm_l, loc=<-5.2302 -9.6435 -1.3163>, rot=<0 0 0 1>}
 dyn_Cloth_lowerarm_l_B01 {parent=dyn_Cloth_lowerarm_l, loc=<-3.5403 0.5991 -4.5567>, rot=<-0.1767 0 0 -0.9843>}
 dyn_Cloth_lowerarm_l_B02 {parent=dyn_Cloth_lowerarm_l_B01, loc=<-5.2759 -8.2336 0>, rot=<0 0 0 1>}
 dyn_Cloth_lowerarm_l_C01 {parent=dyn_Cloth_lowerarm_l, loc=<-9.4954 0.7927 0.3135>, rot=<-0.1767 0 0 -0.9843>}
 dyn_Cloth_lowerarm_l_C02 {parent=dyn_Cloth_lowerarm_l_C01, loc=<-6.9026 -6.9304 1.7741>, rot=<0 0 0 1>}
 dyn_Cloth_lowerarm_l_D01 {parent=dyn_Cloth_lowerarm_l, loc=<-4.0352 -0.5173 3.8454>, rot=<-0.1767 0 0 -0.9843>}
 dyn_Cloth_lowerarm_l_D02 {parent=dyn_Cloth_lowerarm_l_D01, loc=<-5.5161 -7.4602 1.7741>, rot=<0 0 0 1>}
 lowerarm_twist_01_l_scale {parent=lowerarm_twist_01_l, loc=<1 0 0>, rot=<0 0 0 1>}
 upperarm_twist_01_l {parent=upperarm_l, loc=<7.2135 -0.1134 0>, rot=<0 0 0 1>}
 flont_uppelalm_l {parent=upperarm_twist_01_l, loc=<-7.2177 1 0.0001>, rot=<0 0 0 1>}
 up_uppelalm_l {parent=upperarm_twist_01_l, loc=<-7.2179 0.113 1>, rot=<0 0 0 1>}
 back_uppelalm_l {parent=upperarm_twist_01_l, loc=<-7.2177 -1 0.0001>, rot=<0 0 0 1>}
 upperarm_twist_02_l {parent=upperarm_l, loc=<19.7998 -0.1134 0>, rot=<0 0 0 1>}
 clavicle_r {parent=spine_05, loc=<5.4831 -0.1053 3.2946>, rot=<0.7289 0.0465 -0.6829 0.009>}
 upperarm_r {parent=clavicle_r, loc=<-9.8658 1.2136 0.4105>, rot=<0.0164 -0.3245 -0.0269 -0.9454>}
 lowerarm_r {parent=upperarm_r, loc=<-26.465 0.8345 -0.259>, rot=<0.0036 0.0161 -0.2298 -0.9731>}
 hand_r {parent=lowerarm_r, loc=<-22.7355 0.3203 0.6106>, rot=<-0.6784 -0.0223 0.0735 -0.7307>}
 thumb_01_r {parent=hand_r, loc=<-1.3118 0.3191 2.459>, rot=<0.6546 -0.4042 0.0664 -0.6354>}
 thumb_02_r {parent=thumb_01_r, loc=<-5.9336 0.0448 -0.0401>, rot=<-0.0037 0.0213 0.298 -0.9543>}
 thumb_03_r {parent=thumb_02_r, loc=<-2.2223 -0.3682 0.0907>, rot=<0.0118 -0.013 0.0066 -0.9998>}
 index_metacarpal_r {parent=hand_r, loc=<-1.6535 -0.0183 2.96>, rot=<0.0158 -0.0124 -0.0173 -0.9996>}
 index_01_r {parent=index_metacarpal_r, loc=<-8.3008 -0.0357 0.0524>, rot=<-0.0044 0.0485 0.0924 -0.9945>}
 index_02_r {parent=index_01_r, loc=<-3.3587 0.3004 -0.0452>, rot=<-0.002 0.0027 0.3225 -0.9466>}
 index_03_r {parent=index_02_r, loc=<-2.5578 0.123 0.0252>, rot=<-0.002 0.0067 0.2558 -0.9667>}
 middle_metacarpal_r {parent=hand_r, loc=<-2.1856 0.1725 0.5984>, rot=<0.006 0.0168 -0.0329 -0.9993>}
 middle_01_r {parent=middle_metacarpal_r, loc=<-7.3275 0.2447 -0.0129>, rot=<-0.0208 0.0042 0.0858 -0.9961>}
 middle_02_r {parent=middle_01_r, loc=<-4.0123 -0.1401 -0.1175>, rot=<-0.0218 0.0258 0.3512 -0.9357>}
 middle_03_r {parent=middle_02_r, loc=<-2.6077 -0.0534 -0.1373>, rot=<0.0082 0.0188 0.1708 -0.9851>}
 ring_metacarpal_r {parent=hand_r, loc=<-2.2074 0.3625 -1.3638>, rot=<-0.0284 0.0722 -0.0458 -0.9959>}
 ring_01_r {parent=ring_metacarpal_r, loc=<-7.0295 0.6197 0.1446>, rot=<-0.0489 -0.0252 0.1892 -0.9804>}
 ring_02_r {parent=ring_01_r, loc=<-3.6174 -0.4647 0.1027>, rot=<0 0 0.2806 -0.9598>}
 ring_03_r {parent=ring_02_r, loc=<-2.4152 -0.114 -0.0199>, rot=<0 0 0.1189 -0.9929>}
 pinky_metacarpal_r {parent=hand_r, loc=<-1.7616 0.9753 -3.3814>, rot=<-0.0397 0.0814 -0.069 -0.9935>}
 pinky_01_r {parent=pinky_metacarpal_r, loc=<-7.1709 0.4831 0.1094>, rot=<-0.0654 -0.0035 0.2242 -0.9723>}
 pinky_02_r {parent=pinky_01_r, loc=<-2.8274 -0.0893 0.1043>, rot=<0.0037 0.0011 0.2854 -0.9584>}
 pinky_03_r {parent=pinky_02_r, loc=<-2.2046 -0.0307 0.0078>, rot=<-0.0002 0.0098 0.1445 -0.9895>}
 weapon_r {parent=hand_r, loc=<0.1974 3.4111 -0.3811>, rot=<0 0 0 1>}
 hand_attach_r {parent=hand_r, loc=<-8.5001 2.9999 0>, rot=<0 0 -0.7071 -0.7071>}
 lowerarm_twist_02_r {parent=lowerarm_r, loc=<-10.6707 0 0>, rot=<0 0 0 -1>}
 lowerarm_r_front {parent=lowerarm_twist_02_r, loc=<10.6707 -0.4766 -0.0001>, rot=<0 0 0 1>}
 lowerarm_r_back {parent=lowerarm_twist_02_r, loc=<10.6707 0.5246 -0.0001>, rot=<0 0 0 1>}
 lowerarm_twist_01_r {parent=lowerarm_r, loc=<-18.316 0 0>, rot=<0 0 0 -1>}
 dyn_Cloth_lowerarm_r {parent=lowerarm_twist_01_r, loc=<13.2235 -0.0849 -2.6896>, rot=<-0.6245 -0.2199 0.3706 -0.6514>}
 dyn_Cloth_lowerarm_r_A {parent=dyn_Cloth_lowerarm_r, loc=<-5.838 -9.2878 1.3164>, rot=<0 0 0 1>}
 dyn_Cloth_lowerarm_r_B01 {parent=dyn_Cloth_lowerarm_r, loc=<-3.4941 0.8251 4.5567>, rot=<0.1763 -0.0113 0 -0.9843>}
 dyn_Cloth_lowerarm_r_B02 {parent=dyn_Cloth_lowerarm_r_B01, loc=<-5.7945 -7.8778 -0.0001>, rot=<0 0 0 1>}
 dyn_Cloth_lowerarm_r_C01 {parent=dyn_Cloth_lowerarm_r, loc=<-9.425 1.4007 -0.3135>, rot=<0.1763 -0.0113 0 -0.9843>}
 dyn_Cloth_lowerarm_r_C02 {parent=dyn_Cloth_lowerarm_r_C01, loc=<-7.333 -6.4729 -1.7741>, rot=<0 0 0 1>}
 dyn_Cloth_lowerarm_r_D01 {parent=dyn_Cloth_lowerarm_r, loc=<-4.0599 -0.2572 -3.8454>, rot=<0.1763 -0.0113 0 -0.9843>}
 dyn_Cloth_lowerarm_r_D02 {parent=dyn_Cloth_lowerarm_r_D01, loc=<-5.9841 -7.0906 -1.7741>, rot=<0 0 0.0321 -0.9995>}
 lowerarm_twist_01_r_scale {parent=lowerarm_twist_01_r, loc=<-1 0 0>, rot=<0 0 0 1>}
 upperarm_twist_01_r {parent=upperarm_r, loc=<-7.2135 0.1134 0>, rot=<0 0 0 -1>}
 front_upperarm_r {parent=upperarm_twist_01_r, loc=<7.2136 -1 -0.0001>, rot=<0 0 0 1>}
 up_upperarm_r {parent=upperarm_twist_01_r, loc=<7.2136 -0.113 -1.0001>, rot=<0 0 0 1>}
 back_upperarm_r {parent=upperarm_twist_01_r, loc=<7.2136 1 -0.0001>, rot=<0 0 0 1>}
 upperarm_twist_02_r {parent=upperarm_r, loc=<-19.7998 0.1134 0>, rot=<0 0 0 -1>}
 dyn_Hair_BackL_01 {parent=spine_05, loc=<10.5734 -4.0321 -6.3713>, rot=<0.999 0.0401 -0.0008 -0.0201>}
 dyn_Hair_BackL_02 {parent=dyn_Hair_BackL_01, loc=<-8.263 4.2069 -0.1352>, rot=<0 0 0 1>}
 dyn_Hair_BackL_03 {parent=dyn_Hair_BackL_02, loc=<-7.0233 0.8274 -0.0333>, rot=<0 0 0 1>}
 dyn_Hair_BackL_04 {parent=dyn_Hair_BackL_03, loc=<-7.3857 0.1475 -0.0059>, rot=<0 0 0 1>}
 dyn_Hair_BackL_05 {parent=dyn_Hair_BackL_04, loc=<-8.1116 0.6002 -0.0242>, rot=<0 0 0 1>}
 dyn_Hair_BackL_06 {parent=dyn_Hair_BackL_05, loc=<-6.6208 -0.4716 -0.2398>, rot=<0 0 0 1>}
 dyn_Hair_BackM_01 {parent=spine_05, loc=<10.5734 -4.0321 -0.0376>, rot=<0.999 0.0401 -0.0008 -0.0201>}
 dyn_Hair_BackM_02 {parent=dyn_Hair_BackM_01, loc=<-8.263 4.2069 -0.1352>, rot=<0 0 0 1>}
 dyn_Hair_BackM_03 {parent=dyn_Hair_BackM_02, loc=<-7.0233 0.8274 -0.0333>, rot=<0 0 0 1>}
 dyn_Hair_BackM_04 {parent=dyn_Hair_BackM_03, loc=<-7.3857 0.1475 -0.0059>, rot=<0 0 0 1>}
 dyn_Hair_BackM_05 {parent=dyn_Hair_BackM_04, loc=<-8.1116 0.6002 -0.0242>, rot=<0 0 0 1>}
 dyn_Hair_BackM_06 {parent=dyn_Hair_BackM_05, loc=<-6.6208 -0.4716 -0.2398>, rot=<0 0 0 1>}
 dyn_Hair_BackR_01 {parent=spine_05, loc=<10.5734 -4.0321 5.6531>, rot=<0.999 0.0401 -0.0008 -0.0201>}
 dyn_Hair_BackR_02 {parent=dyn_Hair_BackR_01, loc=<-8.263 4.2069 -0.1352>, rot=<0 0 0 1>}
 dyn_Hair_BackR_03 {parent=dyn_Hair_BackR_02, loc=<-7.0233 0.8274 -0.0333>, rot=<0 0 0 1>}
 dyn_Hair_BackR_04 {parent=dyn_Hair_BackR_03, loc=<-7.3857 0.1475 -0.0059>, rot=<0 0 0 1>}
 dyn_Hair_BackR_05 {parent=dyn_Hair_BackR_04, loc=<-8.1116 0.6002 -0.0242>, rot=<0 0 0 1>}
 dyn_Hair_BackR_06 {parent=dyn_Hair_BackR_05, loc=<-6.6208 -0.4716 -0.2398>, rot=<0 0 0 1>}
 dyn_skirt_MSTR {parent=spine_01, loc=<0 0 0>, rot=<0 0 0 1>}
 dyn_front_skirt_l_MSTR {parent=dyn_skirt_MSTR, loc=<-4.3901 7.6198 -11.2383>, rot=<-0.0248 0.0328 0.9992 0.0008>}
 dyn_front_skirt_l_01 {parent=dyn_front_skirt_l_MSTR, loc=<0 0 0>, rot=<0 0 0 1>}
 dyn_front_skirt_l_02 {parent=dyn_front_skirt_l_01, loc=<22 0 0>, rot=<0 0 0 1>}
 dyn_front_skirt_l_03 {parent=dyn_front_skirt_l_02, loc=<16 0 0>, rot=<0 0 0 1>}
 dyn_front_skirt_l_04 {parent=dyn_front_skirt_l_03, loc=<15 0 0>, rot=<0 0 0 1>}
 dyn_front_skirt_l_05 {parent=dyn_front_skirt_l_04, loc=<13 0 0>, rot=<0 0 0 1>}
 dyn_front_skirt_l_06 {parent=dyn_front_skirt_l_05, loc=<14 0 0>, rot=<0 0 0 1>}
 dyn_side_skirt_l_MSTR {parent=dyn_skirt_MSTR, loc=<-4.8705 -1.7711 -14.5383>, rot=<0.0446 -0.7517 -0.6575 -0.0241>}
 dyn_side_skirt_l_01 {parent=dyn_side_skirt_l_MSTR, loc=<0 0 0>, rot=<0 0 0 1>}
 dyn_side_skirt_l_02 {parent=dyn_side_skirt_l_01, loc=<22 0 0>, rot=<0 0 0 1>}
 dyn_side_skirt_l_03 {parent=dyn_side_skirt_l_02, loc=<16 0 0>, rot=<0 0 0.0076 -1>}
 dyn_side_skirt_l_04 {parent=dyn_side_skirt_l_03, loc=<15 0 0>, rot=<0 0 0 1>}
 dyn_side_skirt_l_05 {parent=dyn_side_skirt_l_04, loc=<13 0 0>, rot=<0 0 0 1>}
 dyn_side_skirt_l_06 {parent=dyn_side_skirt_l_05, loc=<14 0 0>, rot=<0 0 0 1>}
 dyn_rear_skirt_l_MSTR {parent=dyn_skirt_MSTR, loc=<-3.9603 -9.2321 -10.8897>, rot=<0.0356 -0.999 0.0067 -0.0281>}
 dyn_rear_skirt_l_01 {parent=dyn_rear_skirt_l_MSTR, loc=<0 0 0>, rot=<0 0 0 1>}
 dyn_rear_skirt_l_02 {parent=dyn_rear_skirt_l_01, loc=<22 0 0>, rot=<0 0 0 1>}
 dyn_rear_skirt_l_03 {parent=dyn_rear_skirt_l_02, loc=<16 0 0>, rot=<0 0 0 1>}
 dyn_rear_skirt_l_04 {parent=dyn_rear_skirt_l_03, loc=<15 0 0>, rot=<0 0 0 1>}
 dyn_rear_skirt_l_05 {parent=dyn_rear_skirt_l_04, loc=<13 0 0>, rot=<0 0 0 1>}
 dyn_rear_skirt_l_06 {parent=dyn_rear_skirt_l_05, loc=<14 0 0>, rot=<0 0 0 1>}
 dyn_rear_skirt_MSTR {parent=dyn_skirt_MSTR, loc=<-4.3819 -11.2881 0>, rot=<0.034 -0.9994 0 0>}
 dyn_rear_skirt_01 {parent=dyn_rear_skirt_MSTR, loc=<0 0 0>, rot=<0 0 0 1>}
 dyn_rear_skirt_02 {parent=dyn_rear_skirt_01, loc=<22 0 0>, rot=<0 0 0 1>}
 dyn_rear_skirt_03 {parent=dyn_rear_skirt_02, loc=<16 0 0>, rot=<0 0 0 1>}
 dyn_rear_skirt_04 {parent=dyn_rear_skirt_03, loc=<15 0 0>, rot=<0 0 0 1>}
 dyn_rear_skirt_05 {parent=dyn_rear_skirt_04, loc=<13 0 0>, rot=<0 0 0 1>}
 dyn_rear_skirt_06 {parent=dyn_rear_skirt_05, loc=<14 0 0>, rot=<0 0 0 1>}
 dyn_front_skirt_MSTR {parent=dyn_skirt_MSTR, loc=<-4.8798 9.5825 0>, rot=<0 0 1 -0.006>}
 dyn_front_skirt_01 {parent=dyn_front_skirt_MSTR, loc=<0 0 0>, rot=<0 0 0 1>}
 dyn_front_skirt_02 {parent=dyn_front_skirt_01, loc=<22 0 0>, rot=<0 0 0 1>}
 dyn_front_skirt_03 {parent=dyn_front_skirt_02, loc=<16 0 0>, rot=<0 0 0 1>}
 dyn_front_skirt_04 {parent=dyn_front_skirt_03, loc=<15 0 0>, rot=<0 0 0 1>}
 dyn_front_skirt_05 {parent=dyn_front_skirt_04, loc=<13 0 0>, rot=<0 0 0 1>}
 dyn_front_skirt_06 {parent=dyn_front_skirt_05, loc=<14 0 0>, rot=<0 0 0 1>}
 dyn_front_skirt_r_MSTR {parent=dyn_skirt_MSTR, loc=<-4.3901 7.6198 11.2383>, rot=<-0.0332 -0.0244 -0.0118 -0.9991>}
 dyn_front_skirt_r_01 {parent=dyn_front_skirt_r_MSTR, loc=<0 0 0>, rot=<0 0 0 1>}
 dyn_front_skirt_r_02 {parent=dyn_front_skirt_r_01, loc=<-22.0003 0 0>, rot=<0 0 0 1>}
 dyn_front_skirt_r_03 {parent=dyn_front_skirt_r_02, loc=<-16 0 0>, rot=<0 0 0 1>}
 dyn_front_skirt_r_04 {parent=dyn_front_skirt_r_03, loc=<-15 0 0>, rot=<0 0 0 1>}
 dyn_front_skirt_r_05 {parent=dyn_front_skirt_r_04, loc=<-13 0 0>, rot=<0 0 0 1>}
 dyn_front_skirt_r_06 {parent=dyn_front_skirt_r_05, loc=<-14.0001 0 0>, rot=<0 0 0 1>}
 dyn_side_skirt_r_MSTR {parent=dyn_skirt_MSTR, loc=<-4.8705 -1.7711 14.5383>, rot=<-0.7516 -0.0467 0.026 -0.6575>}
 dyn_side_skirt_r_01 {parent=dyn_side_skirt_r_MSTR, loc=<0 0 0>, rot=<0 0 0 1>}
 dyn_side_skirt_r_02 {parent=dyn_side_skirt_r_01, loc=<-22 0 0>, rot=<0 0 0 1>}
 dyn_side_skirt_r_03 {parent=dyn_side_skirt_r_02, loc=<-16 -0.0001 0>, rot=<0 0 0.0076 -1>}
 dyn_side_skirt_r_04 {parent=dyn_side_skirt_r_03, loc=<-15 0.0001 0>, rot=<0 0 0 1>}
 dyn_side_skirt_r_05 {parent=dyn_side_skirt_r_04, loc=<-13 0 0>, rot=<0 0 0 1>}
 dyn_side_skirt_r_06 {parent=dyn_side_skirt_r_05, loc=<-14.0001 0 0>, rot=<0 0 0 1>}
 dyn_rear_skirt_r_MSTR {parent=dyn_skirt_MSTR, loc=<-3.9603 -9.2321 10.8897>, rot=<0.999 0.0356 -0.0281 -0.0067>}
 dyn_rear_skirt_r_01 {parent=dyn_rear_skirt_r_MSTR, loc=<0 0 0>, rot=<0 0 0 1>}
 dyn_rear_skirt_r_02 {parent=dyn_rear_skirt_r_01, loc=<-22.0001 0 0>, rot=<0 0 0 1>}
 dyn_rear_skirt_r_03 {parent=dyn_rear_skirt_r_02, loc=<-16 -0.0001 0>, rot=<0 0 0 -1>}
 dyn_rear_skirt_r_04 {parent=dyn_rear_skirt_r_03, loc=<-15 0 0>, rot=<0 0 0 1>}
 dyn_rear_skirt_r_05 {parent=dyn_rear_skirt_r_04, loc=<-13 0 0>, rot=<0 0 0 1>}
 dyn_rear_skirt_r_06 {parent=dyn_rear_skirt_r_05, loc=<-14 0 0>, rot=<0 0 0 1>}
 dyn_skirt_SIMSPACE {parent=spine_01, loc=<0 0 0>, rot=<0 0 0 1>}
 thigh_l {parent=pelvis, loc=<-1.0855 -0.2109 -8.7053>, rot=<0.0009 0.0355 0.0155 -0.9993>}
 calf_l {parent=thigh_l, loc=<-42.5284 -0.0019 -0.7297>, rot=<0.0006 -0.0161 -0.0391 -0.9991>}
 foot_l {parent=calf_l, loc=<-46.7401 -2.5014 -0.592>, rot=<0.0427 -0.0238 0.0353 -0.9982>}
 ball_l {parent=foot_l, loc=<-3.5658 10.8066 -0.1544>, rot=<0.0001 -0.0001 -0.7063 -0.7079>}
 calf_twist_02_l {parent=calf_l, loc=<-19.1546 -1.7943 -0.2638>, rot=<0.0237 0.0095 -0.0035 -0.9997>}
 calf_twist_01_l {parent=calf_l, loc=<-35.8483 -1.9197 -0.5794>, rot=<0.0237 0.0095 -0.0035 -0.9997>}
 thigh_twist_01_l {parent=thigh_l, loc=<-27.952 0 0>, rot=<0.0013 0 0 -1>}
 thigh_volume_fn_l {parent=thigh_twist_01_l, loc=<27.952 1.5 0.0038>, rot=<0.9998 0 -0.0217 -0.0013>}
 thigh_volume_lt_l1 {parent=thigh_twist_01_l, loc=<27.952 0.0038 -1.5>, rot=<-0.7078 -0.0154 0.0153 -0.706>}
 thigh_volume_bk_l {parent=thigh_twist_01_l, loc=<27.952 -1.5 -0.0038>, rot=<-0.0013 0 0 -1>}
 deform_thigh_notwist_l {parent=thigh_l, loc=<0 0 0>, rot=<0 0.0217 0 -0.9998>}
 thigh_r {parent=pelvis, loc=<-1.0855 -0.2109 8.7053>, rot=<0.0355 -0.0009 0.9993 0.0155>}
 calf_r {parent=thigh_r, loc=<42.5284 0.0019 0.7297>, rot=<0.0006 -0.0161 -0.0391 -0.9991>}
 foot_r {parent=calf_r, loc=<46.7401 2.5014 0.592>, rot=<0.0427 -0.0238 0.0353 -0.9982>}
 ball_r {parent=foot_r, loc=<3.5658 -10.8066 0.1544>, rot=<0.0001 -0.0001 -0.7063 -0.7079>}
 calf_twist_02_r {parent=calf_r, loc=<19.1546 1.7943 0.2638>, rot=<0.0237 0.0095 -0.0035 -0.9997>}
 calf_twist_01_r {parent=calf_r, loc=<35.8483 1.9197 0.5794>, rot=<0.0237 0.0095 -0.0035 -0.9997>}
 thigh_twist_01_r {parent=thigh_r, loc=<27.952 0 0>, rot=<0.0013 0 0 -1>}
 thigh_volume_fn_r {parent=thigh_twist_01_r, loc=<-27.9522 -1.5 -0.0038>, rot=<-0.0013 0 0 -1>}
 thigh_volume_rt_r {parent=thigh_twist_01_r, loc=<-27.9522 -0.0038 1.5>, rot=<0.7062 0 0 -0.708>}
 thigh_volume_bk_r {parent=thigh_twist_01_r, loc=<-27.9522 1.5 0.0038>, rot=<1 0 0 -0.0013>}
 deform_thigh_notwist_r {parent=thigh_r, loc=<0 0 0>, rot=<0 0.0217 0 -0.9998>}
 dyn_Cloth_volume_fn {parent=pelvis, loc=<-6.5428 8.5934 0.1611>, rot=<0.9998 -0.0179 0 0>}
 dyn_Cloth_volume_fn01 {parent=dyn_Cloth_volume_fn, loc=<-14.7676 -1.5293 0>, rot=<0 0 0 1>}
 dyn_Cloth_volume_fn02 {parent=dyn_Cloth_volume_fn01, loc=<-22.6974 -1.0417 0.0916>, rot=<0.024 0 0 -0.9997>}
 dyn_Cloth_volume_fn03 {parent=dyn_Cloth_volume_fn02, loc=<-16.5606 0 0>, rot=<0 0 0 1>}
 dyn_Cloth_volume_fn04 {parent=dyn_Cloth_volume_fn03, loc=<-17.0687 0 0>, rot=<0 0 0 1>}
 dyn_Cloth_volume_bk {parent=pelvis, loc=<-6.8808 -8.964 0.2721>, rot=<0 0 0.0179 -0.9998>}
 dyn_Cloth_volume_bk01 {parent=dyn_Cloth_volume_bk, loc=<-14.7845 -1.9699 0>, rot=<0 0 0 1>}
 dyn_Cloth_volume_bk02 {parent=dyn_Cloth_volume_bk01, loc=<-22.9717 -1.6111 -0.2026>, rot=<0.024 0 0 -0.9997>}
 dyn_Cloth_volume_bk03 {parent=dyn_Cloth_volume_bk02, loc=<-16.5606 0 0>, rot=<0 0 0 1>}
 dyn_Cloth_volume_bk04 {parent=dyn_Cloth_volume_bk03, loc=<-17.0687 0 0>, rot=<0 0 0 1>}
 dyn_Cloth_volume_l_A {parent=pelvis, loc=<-6.0375 5.718 -9.1939>, rot=<0.9888 -0.0177 -0.0027 0.148>}
 dyn_Cloth_volume_l_A01 {parent=dyn_Cloth_volume_l_A, loc=<-16.8081 -1.2971 0.6872>, rot=<0 0 0 1>}
 dyn_Cloth_volume_l_A02 {parent=dyn_Cloth_volume_l_A01, loc=<-20.3614 -1.2837 0.4445>, rot=<0 0 0 1>}
 dyn_Cloth_volume_l_A03 {parent=dyn_Cloth_volume_l_A02, loc=<-16.8081 0 0>, rot=<0 0 0 1>}
 dyn_Cloth_volume_l_A04 {parent=dyn_Cloth_volume_l_A03, loc=<-17.0247 0 0>, rot=<0 0 0 1>}
 dyn_Cloth_volume_l_B {parent=pelvis, loc=<-6.161 -1.397 -13.2219>, rot=<-0.7102 0.0127 0.0126 -0.7037>}
 dyn_Cloth_volume_l_B01 {parent=dyn_Cloth_volume_l_B, loc=<-17.06 -1.8458 0>, rot=<0 0 0 1>}
 dyn_Cloth_volume_l_B02 {parent=dyn_Cloth_volume_l_B01, loc=<-20.1095 -2.0721 0>, rot=<0 0 0 1>}
 dyn_Cloth_volume_l_B03 {parent=dyn_Cloth_volume_l_B02, loc=<-17.06 -1.4023 0>, rot=<0 0 0 1>}
 dyn_Cloth_volume_l_B04 {parent=dyn_Cloth_volume_l_B03, loc=<-16.9858 -1.239 0>, rot=<0 0 0 1>}
 dyn_Cloth_volume_l_C {parent=pelvis, loc=<-6.4944 -7.029 -8.659>, rot=<-0.1379 0.0025 0.0177 -0.9903>}
 dyn_Cloth_volume_l_C01 {parent=dyn_Cloth_volume_l_C, loc=<-16.8081 -1.5667 -0.6841>, rot=<0 0 0 1>}
 dyn_Cloth_volume_l_C02 {parent=dyn_Cloth_volume_l_C01, loc=<-20.3614 -1.8288 -0.8944>, rot=<0 0 0 1>}
 dyn_Cloth_volume_l_C03 {parent=dyn_Cloth_volume_l_C02, loc=<-16.8081 -0.907 0>, rot=<0 0 0 1>}
 dyn_Cloth_volume_l_C04 {parent=dyn_Cloth_volume_l_C03, loc=<-17.0247 -0.4758 0>, rot=<0 0 0 1>}
 dyn_Cloth_volume_r_A {parent=pelvis, loc=<-6.0375 5.718 9.1939>, rot=<0.9888 -0.0177 0.0027 -0.148>}
 dyn_Cloth_volume_r_A01 {parent=dyn_Cloth_volume_r_A, loc=<-16.8081 -1.2971 -0.6872>, rot=<0 0 0.0286 -0.9996>}
 dyn_Cloth_volume_r_A02 {parent=dyn_Cloth_volume_r_A01, loc=<-20.2545 -2.4471 -0.4444>, rot=<0 0 0 1>}
 dyn_Cloth_volume_r_A03 {parent=dyn_Cloth_volume_r_A02, loc=<-16.7805 -0.9621 0>, rot=<0 0 0 1>}
 dyn_Cloth_volume_r_A04 {parent=dyn_Cloth_volume_r_A03, loc=<-16.9968 -0.9745 0>, rot=<0 0 -0.0286 -0.9996>}
 dyn_Cloth_volume_r_B {parent=pelvis, loc=<-6.161 -1.397 13.2219>, rot=<0.7101 -0.0188 0.0066 -0.7038>}
 dyn_Cloth_volume_r_B01 {parent=dyn_Cloth_volume_r_B, loc=<-17.089 -1.5542 0>, rot=<0 0 0 1>}
 dyn_Cloth_volume_r_B02 {parent=dyn_Cloth_volume_r_B01, loc=<-20.142 -1.7284 0>, rot=<0 0 0 1>}
 dyn_Cloth_volume_r_B03 {parent=dyn_Cloth_volume_r_B02, loc=<-17.0815 -1.1107 0>, rot=<0 0 0 1>}
 dyn_Cloth_volume_r_B04 {parent=dyn_Cloth_volume_r_B03, loc=<-17.0044 -0.9487 0>, rot=<0 0 0.0085 -1>}
 dyn_Cloth_volume_r_C {parent=pelvis, loc=<-6.4944 -7.029 8.6589>, rot=<0.1379 -0.0042 0.005 -0.9904>}
 dyn_Cloth_volume_r_C01 {parent=dyn_Cloth_volume_r_C, loc=<-16.8429 -1.133 0.6841>, rot=<0 0 0 1>}
 dyn_Cloth_volume_r_C02 {parent=dyn_Cloth_volume_r_C01, loc=<-20.4018 -1.3035 0.8944>, rot=<0 0 0 1>}
 dyn_Cloth_volume_r_C03 {parent=dyn_Cloth_volume_r_C02, loc=<-16.8259 -0.4737 0>, rot=<0 0 0 1>}
 dyn_Cloth_volume_r_C04 {parent=dyn_Cloth_volume_r_C03, loc=<-17.0313 -0.037 0>, rot=<0 0 0.0129 -0.9999>}
 dyn_skirt_ground_collision {parent=root, loc=<0 0 0>, rot=<0 0 0 1>}
 collision {parent=root, loc=<0 0 0>, rot=<0 0 0 1>}
 collision_01 {parent=root, loc=<0 0 0>, rot=<0 0 0 1>}
 camera_root {parent=root, loc=<0 0 0>, rot=<0 0 0 1>}
 camera {parent=camera_root, loc=<0 0 0>, rot=<0 0 0 1>}
 CR Rear Knee Bend L {type=Curve, loc=<0 0 0>, rot=<0 0 0 1>}
 CR Rear Knee Bend R {type=Curve, loc=<0 0 0>, rot=<0 0 0 1>}
 CR Foot Near Pelvis L {type=Curve, loc=<0 0 0>, rot=<0 0 0 1>}
 CR Foot Near Pelvis R {type=Curve, loc=<0 0 0>, rot=<0 0 0 1>}
 CR Thigh Raised Front L {type=Curve, loc=<0 0 0>, rot=<0 0 0 1>}
 CR Thigh Raised Front R {type=Curve, loc=<0 0 0>, rot=<0 0 0 1>}
 CR Thigh L Across R {type=Curve, loc=<0 0 0>, rot=<0 0 0 1>}
 CR Thigh R Across L {type=Curve, loc=<0 0 0>, rot=<0 0 0 1>}
]
variables [
 IsBaseballSliding {type=bool, default=False}
 IsCustomLobbyIdle {type=double, default=0}
 IsHostFull {type=bool, default=False}
 IsInFrontend {type=bool, default=False}
 IsPlayingEmote {type=bool, default=False}
 RobeOffsetWeight {type=double, default=0}
 front panel l {
  type="TArray<FRigElementKey>"
  default=((Type=Bone,Name="dyn_front_panel_l_01"),(Type=Bone,Name="dyn_front_panel_l_02"),(Type=Bone,Name="dyn_front_panel_l_03"),(Type=Bone,Name="dyn_front_panel_l_04"),(Type=Bone,Name="dyn_front_panel_l_05"))
 }
 front panel mid {
  type="TArray<FRigElementKey>"
  default=((Type=Bone,Name="dyn_front_panel_01"),(Type=Bone,Name="dyn_front_panel_02"),(Type=Bone,Name="dyn_front_panel_03"),(Type=Bone,Name="dyn_front_panel_04"),(Type=Bone,Name="dyn_front_panel_05"))
 }
 front panel r {
  type="TArray<FRigElementKey>"
  default=((Type=Bone,Name="dyn_front_panel_r_01"),(Type=Bone,Name="dyn_front_panel_r_02"),(Type=Bone,Name="dyn_front_panel_r_03"),(Type=Bone,Name="dyn_front_panel_r_04"),(Type=Bone,Name="dyn_front_panel_r_05"))
 }
 front skirt l {
  type="TArray<FRigElementKey>"
  default=((Type=Bone,Name="dyn_front_skirt_l_01"),(Type=Bone,Name="dyn_front_skirt_l_02"),(Type=Bone,Name="dyn_front_skirt_l_03"),(Type=Bone,Name="dyn_front_skirt_l_04"),(Type=Bone,Name="dyn_front_skirt_l_05"),(Type=Bone,Name="dyn_front_skirt_l_06"))
 }
 front skirt mid {
  type="TArray<FRigElementKey>"
  default=((Type=Bone,Name="dyn_front_skirt_01"),(Type=Bone,Name="dyn_front_skirt_02"),(Type=Bone,Name="dyn_front_skirt_03"),(Type=Bone,Name="dyn_front_skirt_04"),(Type=Bone,Name="dyn_front_skirt_05"),(Type=Bone,Name="dyn_front_skirt_06"))
 }
 front skirt r {
  type="TArray<FRigElementKey>"
  default=((Type=Bone,Name="dyn_front_skirt_r_01"),(Type=Bone,Name="dyn_front_skirt_r_02"),(Type=Bone,Name="dyn_front_skirt_r_03"),(Type=Bone,Name="dyn_front_skirt_r_04"),(Type=Bone,Name="dyn_front_skirt_r_05"),(Type=Bone,Name="dyn_front_skirt_r_06"))
 }
 player speed {type=double, default=0}
 prev thigh l quaternion {type=FQuat, default=(X=0,Y=0,Z=0,W=1)}
 prev thigh r quaternion {type=FQuat, default=(X=0,Y=0,Z=0,W=1)}
 rear_skirt mid {
  type="TArray<FRigElementKey>"
  default=((Type=Bone,Name="dyn_rear_skirt_01"),(Type=Bone,Name="dyn_rear_skirt_02"),(Type=Bone,Name="dyn_rear_skirt_03"),(Type=Bone,Name="dyn_rear_skirt_04"),(Type=Bone,Name="dyn_rear_skirt_05"),(Type=Bone,Name="dyn_rear_skirt_06"))
 }
 rear_skirt_l {
  type="TArray<FRigElementKey>"
  default=((Type=Bone,Name="dyn_rear_skirt_l_01"),(Type=Bone,Name="dyn_rear_skirt_l_02"),(Type=Bone,Name="dyn_rear_skirt_l_03"),(Type=Bone,Name="dyn_rear_skirt_l_04"),(Type=Bone,Name="dyn_rear_skirt_l_05"),(Type=Bone,Name="dyn_rear_skirt_l_06"))
 }
 rear_skirt_r {
  type="TArray<FRigElementKey>"
  default=((Type=Bone,Name="dyn_rear_skirt_r_01"),(Type=Bone,Name="dyn_rear_skirt_r_02"),(Type=Bone,Name="dyn_rear_skirt_r_03"),(Type=Bone,Name="dyn_rear_skirt_r_04"),(Type=Bone,Name="dyn_rear_skirt_r_05"),(Type=Bone,Name="dyn_rear_skirt_r_06"))
 }
 rotation speed {type=double, default=0}
]
