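# Physics asset for the left robe panel of the female medium DistantEchoCastle player body, set up for skirt bone simulation.
# Front, side and rear left skirt bone chains use capsules, linked segment to segment by tight swing limits with locked twist.
# Capsules on the pelvis, left thigh and calf, and both feet provide body colliders, alongside a large box on a dedicated ground-collision bone.
# Tiny capsules sit on the chain master bones, which have no constraints.
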PhysicsAsset /BRCosmetics/Characters/Player/Female/Medium/Bodies/F_MED_DistantEchoCastle/Meshes/F_MED_DistantEchoCastle_RobeLeft_Physics
bodies {
 calf_l [{shapeType=Capsule, length=45.1837, radius=7.2139, center=<-21.4132 -1.2591 -0.0544>, rotation=<rot -0.1402 -90.666 -89.5616>}]
 dyn_front_skirt_l_01 [{
   shapeType=Capsule
   length=15.0362
   radius=7.125
   shapeName=dyn_front_skirt_l_01_capsule
   center=<15 -7.0435 -2.5185>
   rotation=<rot 23.5464 89.9995 0>
  }]
 dyn_front_skirt_l_02 [{
   shapeType=Capsule
   length=13.5872
   radius=9.2867
   shapeName=dyn_front_skirt_l_02_capsule
   center=<7 -8.7686 -2.5054>
   rotation=<rot 23.5464 89.9995 0>
  }]
 dyn_front_skirt_l_03 [{
   shapeType=Capsule
   length=21.1835
   radius=9.6544
   shapeName=dyn_front_skirt_l_03_capsule
   center=<8.1821 -7.3554 -3.8795>
   rotation=<rot 23.5464 89.9995 0>
  }]
 dyn_front_skirt_l_04 [{
   shapeType=Capsule
   length=25.2806
   radius=10.2785
   shapeName=dyn_front_skirt_l_04_capsule
   center=<8.1821 -5.3379 -3.7194>
   rotation=<rot 23.5464 89.9995 0>
  }]
 dyn_front_skirt_l_05 [{
   shapeType=Capsule
   length=24.6733
   radius=10.3389
   shapeName=dyn_front_skirt_l_05_capsule
   center=<6.5033 -4.5407 -3.372>
   rotation=<rot 23.5464 89.9995 0>
  }]
 dyn_front_skirt_l_06 [{
   shapeType=Capsule
   length=26.0721
   radius=9.8571
   shapeName=dyn_front_skirt_l_06_capsule
   center=<5.5821 -3.9891 -3.0744>
   rotation=<rot 23.5464 89.9995 0>
  }]
 dyn_front_skirt_l_MSTR [{shapeType=Capsule, length=0.505, radius=0.505, shapeName=dyn_front_skirt_l_MSTR_capsule, rotation=<rot 0 -90 0>}]
 dyn_rear_skirt_l_01 [{
   shapeType=Capsule
   length=10
   radius=6
   shapeName=dyn_rear_skirt_r_01_capsule
   center=<14.9123 -5.1684 0>
   rotation=<rot 20.2399 -89.9997 -180>
  }]
 dyn_rear_skirt_l_02 [{
   shapeType=Capsule
   length=18.125
   radius=8.1875
   shapeName=dyn_rear_skirt_l_02_capsule
   center=<7 -8.2358 0.218>
   rotation=<rot -25.8086 90.0005 0>
  }]
 dyn_rear_skirt_l_03 [{
   shapeType=Capsule
   length=20.0394
   radius=8.8125
   shapeName=dyn_rear_skirt_l_03_capsule
   center=<7 -7.4493 0.5344>
   rotation=<rot -20.2399 90.0003 0>
  }]
 dyn_rear_skirt_l_04 [{
   shapeType=Capsule
   length=25.6437
   radius=8.6085
   shapeName=dyn_rear_skirt_l_04_capsule
   center=<7 -7.5177 0.5596>
   rotation=<rot -10.4317 90.0008 0>
  }]
 dyn_rear_skirt_l_05 [{
   shapeType=Capsule
   length=28.0864
   radius=8.7395
   shapeName=dyn_rear_skirt_l_05_capsule
   center=<8.6209 -8.3036 0.3895>
   rotation=<rot -3.7976 90.0018 0>
  }]
 dyn_rear_skirt_l_06 [{
   shapeType=Capsule
   length=29.149
   radius=8.625
   shapeName=dyn_rear_skirt_l_06_capsule
   center=<6.7036 -8.5423 0.4306>
   rotation=<rot -3.7976 90.0018 0>
  }]
 dyn_rear_skirt_l_MSTR [{shapeType=Capsule, length=0.505, radius=0.505, shapeName=dyn_rear_skirt_l_MSTR_capsule, rotation=<rot 0 -90 0>}]
 dyn_side_skirt_l_01 [{shapeType=Capsule, length=10, radius=6, shapeName=dyn_side_skirt_l_01_capsule, center=<10.3933 -8.2214 0>, rotation=<rot 0 90 0>}]
 dyn_side_skirt_l_02 [{shapeType=Capsule, length=17.75, radius=8.9534, shapeName=dyn_side_skirt_l_02_capsule, center=<7.6895 -10.7413 0>, rotation=<rot 0 90 0>}]
 dyn_side_skirt_l_03 [{shapeType=Capsule, length=23.0816, radius=8.8125, shapeName=dyn_side_skirt_l_03_capsule, center=<7.6895 -9.8399 0>, rotation=<rot 0 90 0>}]
 dyn_side_skirt_l_04 [{shapeType=Capsule, length=26.1351, radius=8.625, shapeName=dyn_side_skirt_l_04_capsule, center=<7.6895 -8.6505 0>, rotation=<rot 0 90 0>}]
 dyn_side_skirt_l_05 [{shapeType=Capsule, length=31.3982, radius=8.5888, shapeName=dyn_side_skirt_l_05_capsule, center=<7.6895 -7.8676 0>, rotation=<rot 0 90 0>}]
 dyn_side_skirt_l_06 [{shapeType=Capsule, length=31.9607, radius=9.125, shapeName=dyn_side_skirt_l_06_capsule, center=<6.895 -7.5117 0>, rotation=<rot 0 90 0>}]
 dyn_side_skirt_l_MSTR [{shapeType=Capsule, length=0.505, radius=0.505, shapeName=dyn_side_skirt_l_MSTR_capsule, rotation=<rot 0 -90 0>}]
 dyn_skirt_MSTR [{shapeType=Capsule, length=1, radius=1, rotation=<rot 0 -90 0>}]
 dyn_skirt_ground_collision [{shapeType=Box, extentX=303, extentY=300, extentZ=100, shapeName=dyn_skirt_ground_collision_box, center=<0 0 -46.4179>}]
 foot_l [{
   shapeType=Capsule
   length=13.9992
   radius=12.0663
   shapeName=foot_l_capsule
   center=<-11.3855 8.2245 -0.2438>
   rotation=<rot -1.7789 33.9391 -87.778>
  }]
 foot_r [{
   shapeType=Capsule
   length=13.9992
   radius=12.0663
   shapeName=foot_r_capsule
   center=<10.8878 -7.6388 -2.5222>
   rotation=<rot -1.7789 33.9391 92.222>
  }]
 pelvis [{shapeType=Capsule, length=9, radius=13, center=<-5.0212 0.9394 0>, rotation=<rot 0 -90 0>}]
 thigh_l [{shapeType=Capsule, length=26.6691, radius=9.1467, center=<-24.0471 -0.8818 0.582>, rotation=<rot 0.3314 -90.4107 -90.0983>}]
}
constraints [
 {
  bone1Name=dyn_front_skirt_l_02
  bone2Name=dyn_front_skirt_l_01
  swing1LimitDegrees=12
  swing2LimitDegrees=2
  twistLimitDegrees=2
  twistMotion=Locked
 }
 {
  bone1Name=dyn_front_skirt_l_03
  bone2Name=dyn_front_skirt_l_02
  swing1LimitDegrees=12
  swing2LimitDegrees=2
  twistLimitDegrees=2
  twistMotion=Locked
 }
 {
  bone1Name=dyn_front_skirt_l_04
  bone2Name=dyn_front_skirt_l_03
  swing1LimitDegrees=20
  swing2LimitDegrees=8
  twistLimitDegrees=8
  twistMotion=Locked
 }
 {
  bone1Name=dyn_front_skirt_l_05
  bone2Name=dyn_front_skirt_l_04
  swing1LimitDegrees=20
  swing2LimitDegrees=8
  twistLimitDegrees=8
  twistMotion=Locked
 }
 {
  bone1Name=dyn_front_skirt_l_06
  bone2Name=dyn_front_skirt_l_05
  swing1LimitDegrees=20
  swing2LimitDegrees=8
  twistLimitDegrees=8
  twistMotion=Locked
 }
 {
  bone1Name=dyn_rear_skirt_l_02
  bone2Name=dyn_rear_skirt_l_01
  swing1LimitDegrees=12
  swing2LimitDegrees=8
  twistLimitDegrees=8
  twistMotion=Locked
 }
 {
  bone1Name=dyn_rear_skirt_l_03
  bone2Name=dyn_rear_skirt_l_02
  swing1LimitDegrees=14
  swing2LimitDegrees=8
  twistLimitDegrees=8
  twistMotion=Locked
 }
 {
  bone1Name=dyn_rear_skirt_l_04
  bone2Name=dyn_rear_skirt_l_03
  swing1LimitDegrees=14
  swing2LimitDegrees=8
  twistLimitDegrees=8
  twistMotion=Locked
 }
 {
  bone1Name=dyn_rear_skirt_l_05
  bone2Name=dyn_rear_skirt_l_04
  swing1LimitDegrees=14
  swing2LimitDegrees=8
  twistLimitDegrees=8
  twistMotion=Locked
 }
 {
  bone1Name=dyn_rear_skirt_l_06
  bone2Name=dyn_rear_skirt_l_05
  swing1LimitDegrees=14
  swing2LimitDegrees=8
  twistLimitDegrees=8
  twistMotion=Locked
 }
 {
  bone1Name=dyn_side_skirt_l_02
  bone2Name=dyn_side_skirt_l_01
  swing1LimitDegrees=12
  swing2LimitDegrees=4
  twistLimitDegrees=4
  twistMotion=Locked
 }
 {
  bone1Name=dyn_side_skirt_l_03
  bone2Name=dyn_side_skirt_l_02
  swing1LimitDegrees=14
  swing2LimitDegrees=4
  twistLimitDegrees=4
  twistMotion=Locked
 }
 {
  bone1Name=dyn_side_skirt_l_04
  bone2Name=dyn_side_skirt_l_03
  swing1LimitDegrees=14
  swing2LimitDegrees=8
  twistLimitDegrees=8
  twistMotion=Locked
 }
 {
  bone1Name=dyn_side_skirt_l_05
  bone2Name=dyn_side_skirt_l_04
  swing1LimitDegrees=14
  swing2LimitDegrees=8
  twistLimitDegrees=8
  twistMotion=Locked
 }
 {
  bone1Name=dyn_side_skirt_l_06
  bone2Name=dyn_side_skirt_l_05
  swing1LimitDegrees=14
  swing2LimitDegrees=4
  twistLimitDegrees=4
  twistMotion=Locked
 }
]
