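# Physics asset that drives the skirt of the female medium DistantEchoCastle robe.
# Six capsule chains (front, side and rear, left and right) hang from locked master bodies under a pelvis-linked skirt root, with narrow per-segment angular limits.
# Pelvis, leg and foot capsules plus a large ground box provide collision shapes.
# Angularly free cross-links, some duplicated, tie lower front segments together, and front to side on the right only.
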PhysicsAsset /BRCosmetics/Characters/Player/Female/Medium/Bodies/F_MED_DistantEchoCastle/Meshes/F_MED_DistantEchoCastle_Robe_Physics
bodies {
 calf_l [{shapeType=Capsule, length=45.1837, radius=12.661, center=<-21.4003 -0.1542 -0.0517>, rotation=<rot -0.1046 -95.2647 -89.5518>}]
 calf_r [{shapeType=Capsule, length=45.1837, radius=12.661, center=<21.4003 0.1542 0.0517>, rotation=<rot -0.1046 -95.2647 90.4482>}]
 dyn_front_skirt_l_01 [{shapeType=Capsule, length=21.643, radius=7.125, shapeName=dyn_front_skirt_l_01_capsule, center=<15 -10.0631 0>, rotation=<rot 0 90 0>}]
 dyn_front_skirt_l_02 [{shapeType=Capsule, length=25.5899, radius=9.2867, shapeName=dyn_front_skirt_l_02_capsule, center=<7 -11.9562 0>, rotation=<rot 0 90 0>}]
 dyn_front_skirt_l_03 [{shapeType=Capsule, length=39.916, radius=9.6465, shapeName=dyn_front_skirt_l_03_capsule, center=<8.1821 -4.5411 0>, rotation=<rot 0 90 0>}]
 dyn_front_skirt_l_04 [{shapeType=Capsule, length=49.875, radius=8.625, shapeName=dyn_front_skirt_l_04_capsule, center=<8.1821 -4.5411 0>, rotation=<rot 0 90 0>}]
 dyn_front_skirt_l_05 [{shapeType=Capsule, length=57.5645, radius=8.6854, shapeName=dyn_front_skirt_l_05_capsule, center=<8.1821 -4.5411 0>, rotation=<rot 0 90 0>}]
 dyn_front_skirt_l_06 [{shapeType=Capsule, length=58.9633, radius=9.8492, shapeName=dyn_front_skirt_l_06_capsule, center=<8.1821 -4.5411 0>, rotation=<rot 0 90 0>}]
 dyn_front_skirt_l_MSTR [{shapeType=Capsule, length=0.505, radius=0.505, shapeName=dyn_front_skirt_l_MSTR_capsule, rotation=<rot 0 -90 0>}]
 dyn_front_skirt_r_01 [{shapeType=Capsule, length=21.643, radius=7.125, shapeName=dyn_front_skirt_r_01_capsule, center=<-15 10.565 0>, rotation=<rot 0 90 -180>}]
 dyn_front_skirt_r_02 [{shapeType=Capsule, length=25.3077, radius=9.5689, shapeName=dyn_front_skirt_r_02_capsule, center=<-7 12.9345 0>, rotation=<rot 0 90 -180>}]
 dyn_front_skirt_r_03 [{shapeType=Capsule, length=39.916, radius=9.6465, shapeName=dyn_front_skirt_r_03_capsule, center=<-4 7 0>, rotation=<rot 0 90 -180>}]
 dyn_front_skirt_r_04 [{shapeType=Capsule, length=50.0391, radius=8.4609, shapeName=dyn_front_skirt_r_04_capsule, center=<-7 6 0>, rotation=<rot 0 90 -180>}]
 dyn_front_skirt_r_05 [{shapeType=Capsule, length=57.9713, radius=8.2787, shapeName=dyn_front_skirt_r_05_capsule, center=<-7 6 0>, rotation=<rot 0 90 -180>}]
 dyn_front_skirt_r_06 [{shapeType=Capsule, length=57.2532, radius=11.5593, shapeName=dyn_front_skirt_r_06_capsule, center=<-7 6 0>, rotation=<rot 0 90 -180>}]
 dyn_front_skirt_r_MSTR [{shapeType=Capsule, length=0.505, radius=0.505, shapeName=dyn_front_skirt_r_MSTR_capsule, rotation=<rot 0 -90 0>}]
 dyn_rear_skirt_l_01 [{shapeType=Capsule, length=10, radius=6, shapeName=dyn_rear_skirt_l_01_capsule, center=<1.5316 -4.4228 0>, rotation=<rot 0 90 0>}]
 dyn_rear_skirt_l_02 [{
   shapeType=Capsule
   length=18.125
   radius=8.1875
   shapeName=dyn_rear_skirt_l_02_capsule
   center=<7 -12.2199 0.6064>
   rotation=<rot -5.5687 90.0002 0>
  }]
 dyn_rear_skirt_l_03 [{shapeType=Capsule, length=50.3234, radius=8.8125, shapeName=dyn_rear_skirt_l_03_capsule, center=<7 -10.4003 0>, rotation=<rot 0 90 0>}]
 dyn_rear_skirt_l_04 [{shapeType=Capsule, length=59.9484, radius=8.125, shapeName=dyn_rear_skirt_l_04_capsule, center=<7 -6 0>, rotation=<rot 0 90 0>}]
 dyn_rear_skirt_l_05 [{shapeType=Capsule, length=68.6984, radius=7.125, shapeName=dyn_rear_skirt_l_05_capsule, center=<7 -6 0>, rotation=<rot 0 90 0>}]
 dyn_rear_skirt_l_06 [{shapeType=Capsule, length=69.7609, radius=8.625, shapeName=dyn_rear_skirt_l_06_capsule, center=<3 -6 0>, rotation=<rot 0 90 0>}]
 dyn_rear_skirt_l_MSTR [{shapeType=Capsule, length=0.505, radius=0.505, shapeName=dyn_rear_skirt_l_MSTR_capsule, rotation=<rot 0 -90 0>}]
 dyn_rear_skirt_r_01 [{shapeType=Capsule, length=10, radius=6, shapeName=dyn_rear_skirt_l_01_capsule, center=<-0.8215 5.1685 0>, rotation=<rot 0 -90 0>}]
 dyn_rear_skirt_r_02 [{
   shapeType=Capsule
   length=18.125
   radius=8.1875
   shapeName=dyn_rear_skirt_r_02_capsule
   center=<-7 8.2358 -0.218>
   rotation=<rot -5.5687 90.0002 180>
  }]
 dyn_rear_skirt_r_03 [{shapeType=Capsule, length=50.3234, radius=8.8125, shapeName=dyn_rear_skirt_r_03_capsule, center=<-7 6 0>, rotation=<rot 0 90 -180>}]
 dyn_rear_skirt_r_04 [{shapeType=Capsule, length=59.9484, radius=8.125, shapeName=dyn_rear_skirt_r_04_capsule, center=<-7 6 0>, rotation=<rot 0 90 -180>}]
 dyn_rear_skirt_r_05 [{shapeType=Capsule, length=68.6984, radius=7.125, shapeName=dyn_rear_skirt_r_05_capsule, center=<-7 6 0>, rotation=<rot 0 90 -180>}]
 dyn_rear_skirt_r_06 [{shapeType=Capsule, length=69.7609, radius=8.625, shapeName=dyn_rear_skirt_r_06_capsule, center=<-3 6 0>, rotation=<rot 0 90 -180>}]
 dyn_rear_skirt_r_MSTR [{shapeType=Capsule, length=0.505, radius=0.505, shapeName=dyn_rear_skirt_r_MSTR_capsule, rotation=<rot 0 -90 0>}]
 dyn_side_skirt_l_01 [{shapeType=Capsule, length=10, radius=6, shapeName=dyn_side_skirt_l_01_capsule, center=<4.5234 -8.2214 0>, rotation=<rot 0 90 0>}]
 dyn_side_skirt_l_02 [{shapeType=Capsule, length=17.75, radius=8.5625, shapeName=dyn_side_skirt_l_02_capsule, center=<7.6895 -5.4064 0>, rotation=<rot 0 90 0>}]
 dyn_side_skirt_l_03 [{shapeType=Capsule, length=40.75, radius=8.8125, shapeName=dyn_side_skirt_l_03_capsule, center=<7.6895 -5.4064 0>, rotation=<rot 0 90 0>}]
 dyn_side_skirt_l_04 [{shapeType=Capsule, length=49.875, radius=8.625, shapeName=dyn_side_skirt_l_04_capsule, center=<7.6895 -5.4064 0>, rotation=<rot 0 90 0>}]
 dyn_side_skirt_l_05 [{shapeType=Capsule, length=59.125, radius=7.125, shapeName=dyn_side_skirt_l_05_capsule, center=<7.6895 -5.4064 0>, rotation=<rot 0 90 0>}]
 dyn_side_skirt_l_06 [{shapeType=Capsule, length=59.6875, radius=9.125, shapeName=dyn_side_skirt_l_06_capsule, center=<7.6895 -5.4064 0>, rotation=<rot 0 90 0>}]
 dyn_side_skirt_l_MSTR [{shapeType=Capsule, length=0.505, radius=0.505, shapeName=dyn_side_skirt_l_MSTR_capsule, rotation=<rot 0 -90 0>}]
 dyn_side_skirt_r_01 [{shapeType=Capsule, length=10, radius=6, shapeName=dyn_side_skirt_r_01_capsule, center=<-7 8.1827 0>, rotation=<rot 0 90 -180>}]
 dyn_side_skirt_r_02 [{shapeType=Capsule, length=17.75, radius=8.5625, shapeName=dyn_side_skirt_r_02_capsule, center=<-7 9 0>, rotation=<rot 0 90 -180>}]
 dyn_side_skirt_r_03 [{shapeType=Capsule, length=40.75, radius=8.8125, shapeName=dyn_side_skirt_r_03_capsule, center=<-7 9.5 0>, rotation=<rot 0 90 -180>}]
 dyn_side_skirt_r_04 [{shapeType=Capsule, length=49.875, radius=8.625, shapeName=dyn_side_skirt_r_04_capsule, center=<-7 10 0>, rotation=<rot 0 90 -180>}]
 dyn_side_skirt_r_05 [{shapeType=Capsule, length=59.125, radius=7.125, shapeName=dyn_side_skirt_r_05_capsule, center=<-7 7 0>, rotation=<rot 0 90 -180>}]
 dyn_side_skirt_r_06 [{shapeType=Capsule, length=58.6875, radius=10.125, shapeName=dyn_side_skirt_r_06_capsule, center=<-7 8 0>, rotation=<rot 0 90 -180>}]
 dyn_side_skirt_r_MSTR [{shapeType=Capsule, length=0.505, radius=0.505, shapeName=dyn_side_skirt_r_MSTR_capsule, rotation=<rot 0 -90 0>}]
 dyn_skirt_MSTR [{shapeType=Capsule, length=1, radius=1, rotation=<rot 0 -90 0>}]
 dyn_skirt_ground_collision [{shapeType=Box, extentX=303, extentY=300, extentZ=100, shapeName=dyn_skirt_ground_collision_box, center=<0 0 -45.9991>}]
 foot_l [{
   shapeType=Capsule
   length=13.9992
   radius=12.0663
   shapeName=foot_l_capsule
   center=<-9.7134 10.0758 -0.3423>
   rotation=<rot -2.0667 41.8204 -88.0429>
  }]
 foot_r [{
   shapeType=Capsule
   length=13.9992
   radius=12.0663
   shapeName=foot_r_capsule
   center=<9.8161 -9.9839 0.3374>
   rotation=<rot -2.0667 41.8204 91.9571>
  }]
 pelvis [{shapeType=Capsule, length=9, radius=13, center=<-5.0212 0.9394 0>, rotation=<rot 0 -90 0>}]
 thigh_l [{shapeType=Capsule, length=26.6691, radius=12.6089, center=<-20.6042 -2.423 0.2667>, rotation=<rot 0.4507 -91.6836 -84.6494>}]
 thigh_r [{shapeType=Capsule, length=26.6691, radius=12.6089, center=<20.5925 2.0262 -0.2636>, rotation=<rot 0.3313 -90.4104 95.3593>}]
}
constraints [
 {bone1Name=dyn_skirt_MSTR, bone2Name=pelvis, swing1LimitDegrees=45, swing1Motion=Locked, swing2LimitDegrees=45, twistLimitDegrees=45}
 {bone1Name=thigh_l, bone2Name=pelvis, swing1LimitDegrees=45, swing1Motion=Locked, swing2LimitDegrees=45, twistLimitDegrees=45}
 {bone1Name=calf_l, bone2Name=thigh_l, swing1LimitDegrees=45, swing1Motion=Locked, swing2LimitDegrees=45, twistLimitDegrees=45}
 {bone1Name=thigh_r, bone2Name=pelvis, swing1LimitDegrees=45, swing1Motion=Locked, swing2LimitDegrees=45, twistLimitDegrees=45}
 {bone1Name=calf_r, bone2Name=thigh_r, swing1LimitDegrees=45, swing1Motion=Locked, swing2LimitDegrees=45, twistLimitDegrees=45}
 {
  bone1Name=dyn_front_skirt_l_MSTR
  bone2Name=dyn_skirt_MSTR
  swing1LimitDegrees=24
  swing1Motion=Locked
  swing2LimitDegrees=24
  swing2Motion=Locked
  twistLimitDegrees=24
  twistMotion=Locked
 }
 {bone1Name=dyn_front_skirt_l_01, bone2Name=dyn_front_skirt_l_MSTR, swing1LimitDegrees=8, swing2LimitDegrees=2, twistLimitDegrees=2}
 {bone1Name=dyn_front_skirt_l_02, bone2Name=dyn_front_skirt_l_01, swing1LimitDegrees=24, swing2LimitDegrees=2, twistLimitDegrees=2}
 {bone1Name=dyn_front_skirt_l_03, bone2Name=dyn_front_skirt_l_02, swing1LimitDegrees=24, swing2LimitDegrees=2, twistLimitDegrees=2}
 {bone1Name=dyn_front_skirt_l_04, bone2Name=dyn_front_skirt_l_03, swing1LimitDegrees=24, swing2LimitDegrees=8, twistLimitDegrees=8}
 {bone1Name=dyn_front_skirt_l_05, bone2Name=dyn_front_skirt_l_04, swing1LimitDegrees=24, swing2LimitDegrees=8, twistLimitDegrees=8}
 {bone1Name=dyn_front_skirt_l_06, bone2Name=dyn_front_skirt_l_05, swing1LimitDegrees=24, swing2LimitDegrees=8, twistLimitDegrees=8}
 {
  bone1Name=dyn_front_skirt_r_MSTR
  bone2Name=dyn_skirt_MSTR
  swing1LimitDegrees=45
  swing1Motion=Locked
  swing2LimitDegrees=45
  swing2Motion=Locked
  twistLimitDegrees=45
  twistMotion=Locked
 }
 {bone1Name=dyn_front_skirt_r_01, bone2Name=dyn_front_skirt_r_MSTR, swing1LimitDegrees=8, swing2LimitDegrees=2, twistLimitDegrees=2}
 {bone1Name=dyn_front_skirt_r_02, bone2Name=dyn_front_skirt_r_01, swing1LimitDegrees=24, swing2LimitDegrees=2, twistLimitDegrees=2}
 {bone1Name=dyn_front_skirt_r_03, bone2Name=dyn_front_skirt_r_02, swing1LimitDegrees=24, swing2LimitDegrees=2, twistLimitDegrees=2}
 {bone1Name=dyn_front_skirt_r_04, bone2Name=dyn_front_skirt_r_03, swing1LimitDegrees=24, swing2LimitDegrees=8, twistLimitDegrees=8}
 {bone1Name=dyn_front_skirt_r_05, bone2Name=dyn_front_skirt_r_04, swing1LimitDegrees=24, swing2LimitDegrees=8, twistLimitDegrees=8}
 {bone1Name=dyn_front_skirt_r_06, bone2Name=dyn_front_skirt_r_05, swing1LimitDegrees=24, swing2LimitDegrees=8, twistLimitDegrees=8}
 {
  bone1Name=dyn_rear_skirt_l_MSTR
  bone2Name=dyn_skirt_MSTR
  swing1LimitDegrees=24
  swing1Motion=Locked
  swing2LimitDegrees=24
  swing2Motion=Locked
  twistLimitDegrees=24
  twistMotion=Locked
 }
 {bone1Name=dyn_rear_skirt_l_01, bone2Name=dyn_rear_skirt_l_MSTR, swing1LimitDegrees=12, swing2LimitDegrees=0, twistLimitDegrees=8}
 {bone1Name=dyn_rear_skirt_l_02, bone2Name=dyn_rear_skirt_l_01, swing1LimitDegrees=12, swing2LimitDegrees=8, twistLimitDegrees=8}
 {bone1Name=dyn_rear_skirt_l_03, bone2Name=dyn_rear_skirt_l_02, swing1LimitDegrees=14, swing2LimitDegrees=8, twistLimitDegrees=8}
 {bone1Name=dyn_rear_skirt_l_04, bone2Name=dyn_rear_skirt_l_03, swing1LimitDegrees=14, swing2LimitDegrees=8, twistLimitDegrees=8}
 {bone1Name=dyn_rear_skirt_l_05, bone2Name=dyn_rear_skirt_l_04, swing1LimitDegrees=14, swing2LimitDegrees=8, twistLimitDegrees=8}
 {bone1Name=dyn_rear_skirt_l_06, bone2Name=dyn_rear_skirt_l_05, swing1LimitDegrees=14, swing2LimitDegrees=8, twistLimitDegrees=8}
 {
  bone1Name=dyn_rear_skirt_r_MSTR
  bone2Name=dyn_skirt_MSTR
  swing1LimitDegrees=45
  swing1Motion=Locked
  swing2LimitDegrees=45
  swing2Motion=Locked
  twistLimitDegrees=45
  twistMotion=Locked
 }
 {bone1Name=dyn_rear_skirt_r_01, bone2Name=dyn_rear_skirt_r_MSTR, swing1LimitDegrees=12, swing2LimitDegrees=0, twistLimitDegrees=8}
 {bone1Name=dyn_rear_skirt_r_02, bone2Name=dyn_rear_skirt_r_01, swing1LimitDegrees=12, swing2LimitDegrees=8, twistLimitDegrees=8}
 {bone1Name=dyn_rear_skirt_r_03, bone2Name=dyn_rear_skirt_r_02, swing1LimitDegrees=14, swing2LimitDegrees=8, twistLimitDegrees=8}
 {bone1Name=dyn_rear_skirt_r_04, bone2Name=dyn_rear_skirt_r_03, swing1LimitDegrees=14, swing2LimitDegrees=8, twistLimitDegrees=8}
 {bone1Name=dyn_rear_skirt_r_05, bone2Name=dyn_rear_skirt_r_04, swing1LimitDegrees=14, swing2LimitDegrees=8, twistLimitDegrees=8}
 {bone1Name=dyn_rear_skirt_r_06, bone2Name=dyn_rear_skirt_r_05, swing1LimitDegrees=14, swing2LimitDegrees=8, twistLimitDegrees=8}
 {
  bone1Name=dyn_side_skirt_l_MSTR
  bone2Name=dyn_skirt_MSTR
  swing1LimitDegrees=24
  swing1Motion=Locked
  swing2LimitDegrees=24
  swing2Motion=Locked
  twistLimitDegrees=24
  twistMotion=Locked
 }
 {bone1Name=dyn_side_skirt_l_01, bone2Name=dyn_side_skirt_l_MSTR, swing1LimitDegrees=12, swing2LimitDegrees=4, twistLimitDegrees=4}
 {bone1Name=dyn_side_skirt_l_02, bone2Name=dyn_side_skirt_l_01, swing1LimitDegrees=12, swing2LimitDegrees=4, twistLimitDegrees=4}
 {bone1Name=dyn_side_skirt_l_03, bone2Name=dyn_side_skirt_l_02, swing1LimitDegrees=14, swing2LimitDegrees=4, twistLimitDegrees=4}
 {bone1Name=dyn_side_skirt_l_04, bone2Name=dyn_side_skirt_l_03, swing1LimitDegrees=14, swing2LimitDegrees=8, twistLimitDegrees=8}
 {bone1Name=dyn_side_skirt_l_05, bone2Name=dyn_side_skirt_l_04, swing1LimitDegrees=14, swing2LimitDegrees=8, twistLimitDegrees=8}
 {bone1Name=dyn_side_skirt_l_06, bone2Name=dyn_side_skirt_l_05, swing1LimitDegrees=14, swing2LimitDegrees=4, twistLimitDegrees=4}
 {
  bone1Name=dyn_side_skirt_r_MSTR
  bone2Name=dyn_skirt_MSTR
  swing1LimitDegrees=45
  swing1Motion=Locked
  swing2LimitDegrees=45
  swing2Motion=Locked
  twistLimitDegrees=45
  twistMotion=Locked
 }
 {bone1Name=dyn_side_skirt_r_01, bone2Name=dyn_side_skirt_r_MSTR, swing1LimitDegrees=12, swing2LimitDegrees=4, twistLimitDegrees=4}
 {bone1Name=dyn_side_skirt_r_02, bone2Name=dyn_side_skirt_r_01, swing1LimitDegrees=12, swing2LimitDegrees=4, twistLimitDegrees=4}
 {bone1Name=dyn_side_skirt_r_03, bone2Name=dyn_side_skirt_r_02, swing1LimitDegrees=14, swing2LimitDegrees=4, twistLimitDegrees=4}
 {bone1Name=dyn_side_skirt_r_04, bone2Name=dyn_side_skirt_r_03, swing1LimitDegrees=14, swing2LimitDegrees=8, twistLimitDegrees=8}
 {bone1Name=dyn_side_skirt_r_05, bone2Name=dyn_side_skirt_r_04, swing1LimitDegrees=14, swing2LimitDegrees=8, twistLimitDegrees=8}
 {bone1Name=dyn_side_skirt_r_06, bone2Name=dyn_side_skirt_r_05, swing1LimitDegrees=14, swing2LimitDegrees=4, twistLimitDegrees=4}
 {
  bone1Name=dyn_front_skirt_r_06
  bone2Name=dyn_front_skirt_l_06
  swing1LimitDegrees=45
  swing1Motion=Free
  swing2LimitDegrees=45
  swing2Motion=Free
  twistLimitDegrees=45
  twistMotion=Free
 }
 {
  bone1Name=dyn_front_skirt_l_06
  bone2Name=dyn_front_skirt_r_06
  swing1LimitDegrees=45
  swing1Motion=Free
  swing2LimitDegrees=45
  swing2Motion=Free
  twistLimitDegrees=45
  twistMotion=Free
 }
 {
  bone1Name=foot_l
  bone2Name=calf_l
  swing1LimitDegrees=45
  swing1Motion=Free
  swing2LimitDegrees=45
  swing2Motion=Free
  twistLimitDegrees=45
  twistMotion=Free
 }
 {
  bone1Name=foot_r
  bone2Name=calf_r
  swing1LimitDegrees=45
  swing1Motion=Free
  swing2LimitDegrees=45
  swing2Motion=Free
  twistLimitDegrees=45
  twistMotion=Free
 }
 {
  bone1Name=dyn_front_skirt_r_05
  bone2Name=dyn_side_skirt_r_05
  swing1LimitDegrees=45
  swing1Motion=Free
  swing2LimitDegrees=45
  swing2Motion=Free
  twistLimitDegrees=45
  twistMotion=Free
 }
 {
  bone1Name=dyn_front_skirt_r_06
  bone2Name=dyn_front_skirt_l_06
  swing1LimitDegrees=45
  swing1Motion=Free
  swing2LimitDegrees=45
  swing2Motion=Free
  twistLimitDegrees=45
  twistMotion=Free
 }
 {
  bone1Name=dyn_front_skirt_r_05
  bone2Name=dyn_front_skirt_l_05
  swing1LimitDegrees=45
  swing1Motion=Free
  swing2LimitDegrees=45
  swing2Motion=Free
  twistLimitDegrees=45
  twistMotion=Free
 }
]
